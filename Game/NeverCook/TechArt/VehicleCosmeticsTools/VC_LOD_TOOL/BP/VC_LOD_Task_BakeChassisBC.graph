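# Editor utility task in the vehicle cosmetics LOD tool that bakes a chassis base color (BC) texture for LODs.
# After a delay and cancel check, it copies settings from a newly constructed LOD tool object, then Geometry Script projects the high chassis texture onto the low mesh.
# It saves and configures the texture asset before finishing the task.
# The optional composite branch calls mask-bake and composite functions, both empty here.
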EditorUtilityBlueprint /Game/NeverCook/TechArt/VehicleCosmeticsTools/VC_LOD_TOOL/BP/VC_LOD_Task_BakeChassisBC
graphs [
 BakeCompositeMask_Initialize {nodes [BakeCompositeMask_Initialize() {is=title K2Node_FunctionEntry}]}
 CreateCompositChassisBC {nodes [CreateCompositChassisBC() {is=title K2Node_FunctionEntry}]}
 EventGraph {nodes [
   Event BeginExecution()
   > SetTaskNotificationText!(execute, self:Editor Utility Task Object Reference, Text:"Text (by ref)"=MakeLiteralText(Value="LOCTEXT(\"Baking chassis BC\")")) -> (then)
   > PrintString(InString=BakeChassisBC)
   > Delay!(execute, WorldContextObject:Object Reference, Duration=3) -> (then)
   > n1 = Branch(Condition=Get CancelPending() -> (CancelPending:bool))
   n2 = Construct Object from Class(
    execute=@n1.else
    Class=/Game/NeverCook/TechArt/VehicleCosmeticsTools/VC_LOD_TOOL/Versions/VehicleCosmetics_LODTOOL_AsyncTesting.VehicleCosmetics_LODTOOL_AsyncTesting_C
   ) -> (ReturnValue:Vehicle Cosmetics LODTOOL Async Testing Object Reference)
   > Set DynamicMeshActor(DynamicMeshActor:Vehicle Cosmetics LODBP5 Object Reference=Get DynamicMeshActor!(self:Vehicle Cosmetics LODTOOL Async Testing Object Reference=@n2) -> (DynamicMeshActor:Vehicle Cosmetics LODBP5 Object Reference)) -> (Output_Get:Vehicle Cosmetics LODBP5 Object Reference)
   > Set ChassisBC_Source(ChassisBC_Source:Texture 2D Object Reference=Get ChassisBC_Source!(self:Vehicle Cosmetics LODTOOL Async Testing Object Reference=@n2) -> (ChassisBC_Source:Texture 2D Object Reference)) -> (Output_Get:Texture 2D Object Reference)
   > Set Resolution(Resolution:EGeometryScriptBakeResolution Enum=Get Resolution!(self:Vehicle Cosmetics LODTOOL Async Testing Object Reference=@n2) -> (Resolution:EGeometryScriptBakeResolution Enum=Resolution16)) -> (Output_Get:EGeometryScriptBakeResolution Enum=Resolution16)
   > Set Projection Distance(Projection Distance:float=Get Projection Distance!(self:Vehicle Cosmetics LODTOOL Async Testing Object Reference=@n2) -> (Projection Distance:float)) -> (Output_Get:float)
   > Set Samples Per Pixel(Samples Per Pixel:EGeometryScriptBakeSamplesPerPixel Enum=Get Samples Per Pixel!(self:Vehicle Cosmetics LODTOOL Async Testing Object Reference=@n2) -> (Samples Per Pixel:EGeometryScriptBakeSamplesPerPixel Enum=Sample1)) -> (Output_Get:EGeometryScriptBakeSamplesPerPixel Enum=Sample1)
   > Set SavePath(SavePath:String=Get SavePath!(self:Vehicle Cosmetics LODTOOL Async Testing Object Reference=@n2) -> (SavePath:String)) -> (Output_Get:String)
   > Set VehicleName(VehicleName:String=Get VehicleName!(self:Vehicle Cosmetics LODTOOL Async Testing Object Reference=@n2) -> (VehicleName:String)) -> (Output_Get:String)
   > Set ChassisCompositeMask_LOD(ChassisCompositeMask_LOD:Texture 2D Object Reference=Get ChassisCompositeMask_LOD!(self:Vehicle Cosmetics LODTOOL Async Testing Object Reference=@n2) -> (ChassisCompositeMask_LOD:Texture 2D Object Reference)) -> (Output_Get:Texture 2D Object Reference)
   > Set CompositeChassisBC(CompositeChassisBC:bool=Get CompositeChassisBC!(self:Vehicle Cosmetics LODTOOL Async Testing Object Reference=@n2) -> (CompositeChassisBC:bool)) -> (Output_Get:bool)
   n3 = Get ChassisBC_Source!() -> (ChassisBC_Source)
   n4 = Get Target Transform() -> (Target Transform:Transform)
   n5 = Get DynamicMeshActor!() -> (DynamicMeshActor)
   > n6 = BakeTexture!(
    execute
    TargetMesh:Dynamic Mesh Object Reference=Get DynamicMeshLowChassis!(self:Vehicle Cosmetics LODBP5 Object Reference=@n5) -> (DynamicMeshLowChassis:Dynamic Mesh Object Reference)
    TargetTransform:Transform=@n4
    TargetOptions:Geometry Script Bake Target Mesh Options Structure=Make Geometry Script Bake Target Mesh Options!(TargetUVLayer:int=0) -> (GeometryScriptBakeTargetMeshOptions:Geometry Script Bake Target Mesh Options Structure) {is=title K2Node_MakeStruct}
    SourceMesh:Dynamic Mesh Object Reference=Get DynamicMeshHighChassis!(self:Vehicle Cosmetics LODBP5 Object Reference=@n5) -> (DynamicMeshHighChassis:Dynamic Mesh Object Reference)
    SourceTransform:Transform=@n4
    SourceOptions:Geometry Script Bake Source Mesh Options Structure=Make Geometry Script Bake Source Mesh Options!(SourceNormalMap:Texture 2D Object Reference, SourceNormalUVLayer:int=0, SourceNormalSpace:EGeometryScriptBakeNormalSpace Enum=Tangent) -> (GeometryScriptBakeSourceMeshOptions:Geometry Script Bake Source Mesh Options Structure) {is=title K2Node_MakeStruct}
    BakeTypes:Array of Geometry Script Bake Type Options Structures=Make Array("[0]":Geometry Script Bake Type Options Structure=MakeBakeTypeTexture!(SourceTexture:Texture 2D Object Reference=@n3, SourceUVLayer:int=0) -> (ReturnValue:Geometry Script Bake Type Options Structure)) -> (Array:Array of Geometry Script Bake Type Options Structures)
    BakeOptions:Geometry Script Bake Texture Options Structure=Make Geometry Script Bake Texture Options!(
     Resolution:EGeometryScriptBakeResolution Enum=Get Resolution!() -> (Resolution=Resolution16)
     BitDepth:EGeometryScriptBakeBitDepth Enum=ChannelBits8
     GutterSize:int=4
     SamplesPerPixel:EGeometryScriptBakeSamplesPerPixel Enum=Get Samples Per Pixel!() -> (Samples Per Pixel=Sample1)
     SampleFilterMask:Texture 2D Object Reference
     FilteringType:EGeometryScriptBakeFilteringType Enum=BSpline
     ProjectionDistance:float=Get Projection Distance!() -> (Projection Distance)
     bProjectionInWorldSpace:bool=False
    ) -> (GeometryScriptBakeTextureOptions:Geometry Script Bake Texture Options Structure) {is=title K2Node_MakeStruct}
   ) -> (then, ReturnValue:Array of Texture 2D Object References)
   n7 = "Get (a copy)"(Array:Array of Texture 2D Object References=@n6) -> (Output:Texture 2D Object Reference)
   > n8 = Branch(Condition=IsValid(Object=@n7))
   n9 = Get SavePath!() -> (SavePath)
   > n10 = CreateNewTexture2DAsset!(
    execute
    FromTexture:Texture 2D Object Reference=@n7
    AssetPathAndName:String=Append!(A=Append!(A=@n9, B=/) -> (ReturnValue), B=T_, C:String=Get VehicleName!() -> (VehicleName), D:String=_, E:String=Chassis_BC, F:String=_LOD) -> (ReturnValue)
    Options:Geometry Script Create New Texture 2DAsset Options Structure=Make Geometry Script Create New Texture 2DAsset Options!(bOverwriteIfExists:bool=true) -> (GeometryScriptCreateNewTexture2DAssetOptions:Geometry Script Create New Texture 2DAsset Options Structure) {is=title K2Node_MakeStruct}
   ) -> (Failure:Exec, Success:Exec, ReturnValue:Texture 2D Object Reference)
   LoadAsset!(
    execute=@n10.Success
    self:Editor Asset Subsystem Object Reference=Get EditorAssetSubsystem() -> (ReturnValue:Editor Asset Subsystem Object Reference) {is=title K2Node_GetEditorSubsystem}
    AssetPath=GetSystemPath!(Object:Object Reference=@n10) -> (ReturnValue:String)
   ) -> (then, ReturnValue)
   > Set Chassis_BC_LOD(Chassis_BC_LOD:Texture 2D Object Reference=@n10) -> (Output_Get:Texture 2D Object Reference)
   > n11 = Branch(Condition=Get CompositeChassisBC!() -> (CompositeChassisBC))
   > PrintString(InString=Compositing Chassis BC, bPrintToScreen=false)
   > n12 = Branch(Condition=IsValid(Object=Get ChassisCompositeMask_LOD!() -> (ChassisCompositeMask_LOD)))
   Delay(execute=[@n15.then, @n12.then], Duration=2)
   > CreateCompositChassisBC() {is=title K2Node_CallFunction}
   > Delay(Duration=2)
   > Set CompressionSettings(CompressionSettings=TC_Default, self=Get Chassis_BC_LOD() -> (Chassis_BC_LOD:Texture 2D Object Reference))
   n13 = Get Chassis_BC_LOD()
   > Set SRGB(SRGB=true, self=@n13)
   > Set CompressionNoAlpha!(execute, CompressionNoAlpha:bool=true, self:Texture Object Reference=@n13) -> (then, Output_Get:bool)
   > n14 = Set LODGroup!(execute, LODGroup:TextureGroup Enum=TEXTUREGROUP_Project20, self:Texture Object Reference=@n13) -> (then, Output_Get:TextureGroup Enum)
   PrintString(execute=[@n17.then, @n14.then], InString=Finishing Task..)
   > FinishExecutingTask()
   n15 = BakeCompositeMask_Initialize(execute=@n12.else) {is=title K2Node_CallFunction}
   Delay(execute=@n11.else, Duration=1)
   > Set CompressionSettings(CompressionSettings=TC_Default, self=Get Chassis_BC_LOD())
   n16 = Get Chassis_BC_LOD()
   > Set SRGB(SRGB=true, self=@n16)
   > Set CompressionNoAlpha(CompressionNoAlpha=true, self=@n16)
   > n17 = Set LODGroup(LODGroup=TEXTUREGROUP_Project20, self=@n16)
   PrintString(execute=@n8.else, InString=Fail)
   Event CancelRequested() {is=title K2Node_Event}
   > PrintString(InString=Task Cancelled)
   > Set CancelPending(CancelPending:bool=true) -> (Output_Get:bool)
  ]}
]
variables [
 CancelPending {type=boolean}
 Chassis_BC_LOD {type=Texture2D}
 ChassisBC_Source {type=Texture2D}
 ChassisCompositeMask_LOD {type=Texture2D}
 CompositeChassisBC {type=boolean}
 DynamicMeshActor {type=VehicleCosmetics_LODBP5_C, default=null}
 Projection Distance {type=number}
 Resolution {type=EGeometryScriptBakeResolution, default=Resolution16}
 Samples Per Pixel {type=EGeometryScriptBakeSamplesPerPixel, default=Sample1}
 SavePath {type=string, default=""}
 Target Transform {type=Transform, default {location=<0 0 0>, rotation=<rot 0 0 0>, scale=<1 1 1>}}
 VehicleName {type=string, default=""}
]
