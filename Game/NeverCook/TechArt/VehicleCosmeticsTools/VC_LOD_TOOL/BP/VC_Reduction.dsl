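; VC_Reduction is a vehicle cosmetics LOD editor tool that uses Geometry Script to create a reduced static mesh asset.
; Reduce copies LowSM into a compute mesh, simplifies it to Vertex Count, and writes it to LOD_Test at a hardcoded Sourdough path.
; It also creates an unfilled "Reduction" asset at Package Path, then releases compute meshes.
; SetVertexUV, uncalled here, moves a vertex's UV by tolerance-matching corners in connected triangles.
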(event Custom|Reduce
  (Utilities|FlowControl|Sequence
    (:then_0
      (bind _returnvalue (DynamicMeshActor|AllocateComputeMesh))
      (bind _returnvalue_2 (GeometryScript|StaticMesh|CopyMeshFromStaticMeshwithSectionMaterials (Variables|Default|GetLowSM) _returnvalue)
        (:Success
          (Variables|Default|SetDynamic%20Mesh%20Low _returnvalue_2))))
    (:then_1
      (GeometryScript|Simplification|ApplySimplifytoVertexCount (Variables|Default|GetDynamic%20Mesh%20Low) (Variables|Default|GetVertex%20Count))
      (bind _returnvalue_3 (EditorScripting|AssetTools|GetAssetTools))
      (bind _returnvalue_4 (EditorScripting|AssetTools|CreateAsset _returnvalue_3 "LOD_Test" "/VehicleCosmetics/Sourdough/Meshes/" "/Script/Engine.StaticMesh" 0 "None" false))
      (bind _asstatic_mesh (Utilities|Casting|CastToStaticMesh _returnvalue_4))
      (bind _output_get (Variables|Default|SetNewAsset _asstatic_mesh))
      (GeometryScript|StaticMesh|CopyMeshtoStaticMesh (Variables|Default|GetDynamic%20Mesh%20Low) _output_get))
    (:then_2
      (bind _returnvalue_5 (EditorScripting|AssetTools|GetAssetTools))
      (bind _returnvalue_6 (EditorScripting|AssetTools|CreateAsset _returnvalue_5 "Reduction" (Variables|Default|GetPackage%20Path) "/Script/Engine.StaticMesh" 0 "None" false))
      (Utilities|Casting|CastToStaticMesh _returnvalue_6))
    (:then_3
      (DynamicMeshActor|ReleaseAllComputeMeshes))))

(fn SetVertexUV
  (params (TargetMesh :type Dynamic%20Mesh%20Object%20Reference) (VertexID :type Integer) (MoveToLocationUV :type Vector%202D%20Structure))
  (bind _target (LocalVariables|GetTarget))
  (bind _vertexuv (LocalVariables|GetVertexUV))
  (LocalVariables|SetTarget TargetMesh)
  (LocalVariables|SetVertex VertexID)
  (LocalVariables|SetMoveLocation MoveToLocationUV)
  (bind _uvlist (GeometryScript|UVs|GetMeshPerVertexUVs (LocalVariables|GetTarget)))
  (bind _uvarray (GeometryScript|ListUtils|ConvertUVListtoArray _uvlist))
  (LocalVariables|SetVertexUV (Utilities|Array|Get(acopy) _uvarray (LocalVariables|GetVertex)))
  (bind _triangles (GeometryScript|MeshQueries|GetVertexConnectedTriangles (LocalVariables|GetTarget) (LocalVariables|GetVertex)))
  (for _array_element _triangles
    (if (Utilities|Operators|Equal(==) (GeometryScript|MeshQueries|GetTriangleUVs _target 0 _array_element) _vertexuv 0.001)
      (GeometryScript|UVs|SetMeshTriangleUVs (LocalVariables|GetTarget) 0 _array_element (Utilities|Struct|MakeGeometryScriptUVTriangle (LocalVariables|GetMoveLocation) (.UV2 (GeometryScript|MeshQueries|GetTriangleUVs _target 0 _array_element)) (.UV3 (GeometryScript|MeshQueries|GetTriangleUVs _target 0 _array_element))) false)
      (elif (Utilities|Operators|Equal(==) (.UV2 (GeometryScript|MeshQueries|GetTriangleUVs _target 0 _array_element)) _vertexuv 0.001)
        (GeometryScript|UVs|SetMeshTriangleUVs (LocalVariables|GetTarget) 0 _array_element (Utilities|Struct|MakeGeometryScriptUVTriangle (GeometryScript|MeshQueries|GetTriangleUVs _target 0 _array_element) (LocalVariables|GetMoveLocation) (.UV3 (GeometryScript|MeshQueries|GetTriangleUVs _target 0 _array_element))) false)
        (elif (Utilities|Operators|Equal(==) _vertexuv (.UV3 (GeometryScript|MeshQueries|GetTriangleUVs _target 0 _array_element)) 0.001)
          (GeometryScript|UVs|SetMeshTriangleUVs (LocalVariables|GetTarget) 0 _array_element (Utilities|Struct|MakeGeometryScriptUVTriangle (GeometryScript|MeshQueries|GetTriangleUVs _target 0 _array_element) (.UV2 (GeometryScript|MeshQueries|GetTriangleUVs _target 0 _array_element)) (LocalVariables|GetMoveLocation)) false)
          (else
            _)))))
  (return))

(fn ConstructionScript)
