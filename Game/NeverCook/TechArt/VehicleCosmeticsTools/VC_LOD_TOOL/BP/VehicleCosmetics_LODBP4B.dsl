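; Vehicle cosmetics LOD tool Blueprint that uses Geometry Script to split source (high) and target (low) static meshes by material slot.
; It resolves body, chassis and trim slot IDs, plus optional carbon fiber and plastic, then extracts per-slot dynamic meshes from allocated compute meshes.
; Plastic is appended into trim, then trim, chassis and carbon fiber merge into a high chassis composite.
; Helpers look up skeletal slot indices and move vertex UVs.
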(event Custom|CheckForTeriarty
  ;[0.3897 1 0.7209] Slot IDs
  (bind _array_element (Utilities|Array|ForEachLoopwithBreak (Variables|/Script/Engine.StaticMesh|GetStaticMaterials (Variables|Default|GetSourceMesh)))
    (:LoopBody
      (if (== (.MaterialSlotName (Utilities|Struct|BreakStaticMaterial _array_element)) (Variables|Default|GetCarbonFiberSlotName))
        (Variables|Default|SetHasCarbonFiberSlot true))
      (Variables|Default|SetHasCarbonFiberSlot false))
    (:Completed
      _))
  (bind _array_element_2 (Utilities|Array|ForEachLoopwithBreak (Variables|/Script/Engine.StaticMesh|GetStaticMaterials (Variables|Default|GetSourceMesh)))
    (:LoopBody
      (if (== (.MaterialSlotName (Utilities|Struct|BreakStaticMaterial _array_element_2)) (Variables|Default|GetPlasticSlotName))
        (Variables|Default|SetHasPlasticSlot true))
      (Variables|Default|SetHasPlasticSlot false))))

(event Custom|SetSlotIDs
  (bind _trimslotname (Variables|Default|GetTrimSlotName))
  (bind _sourcemesh (Variables|Default|GetSourceMesh))
  (bind _returnvalue (StaticMesh|GetMaterialIndex _sourcemesh _trimslotname))
  (bind _bodyslotname (Variables|Default|GetBodySlotName))
  (bind _sourcemesh_2 (Variables|Default|GetSourceMesh))
  (bind _returnvalue_2 (StaticMesh|GetMaterialIndex _sourcemesh_2 _bodyslotname))
  (bind _bodyslotname_2 (Variables|Default|GetBodySlotName))
  (bind _targetmesh (Variables|Default|GetTargetMesh))
  (bind _returnvalue_3 (StaticMesh|GetMaterialIndex _targetmesh _bodyslotname_2))
  (bind _chassisslotname (Variables|Default|GetChassisSlotName))
  (bind _targetmesh_2 (Variables|Default|GetTargetMesh))
  (bind _returnvalue_4 (StaticMesh|GetMaterialIndex _targetmesh_2 _chassisslotname))
  (bind _chassisslotname_2 (Variables|Default|GetChassisSlotName))
  (bind _sourcemesh_3 (Variables|Default|GetSourceMesh))
  (bind _returnvalue_5 (StaticMesh|GetMaterialIndex _sourcemesh_3 _chassisslotname_2))
  (bind _hascarbonfiberslot (Variables|Default|GetHasCarbonFiberSlot))
  (bind _hasplasticslot (Variables|Default|GetHasPlasticSlot))
  ;[0.3897 1 0.7209] Slot IDs
  (CallFunction|CheckforTeriarty)
  (Variables|Default|SetBodySlotInvalid false)
  (Variables|Default|SetChassisSlotInvalid false)
  (Utilities|FlowControl|Sequence
    (:then_0
      ;;[1 1 1] Body
      (Variables|Default|SetSource_BodySlotID _returnvalue_2)
      (Development|PrintString (Utilities|String|Append "BodyIDSLOT for Source Mesh " (Utilities|String|ToString(Integer) _returnvalue_2)))
      (Variables|Default|SetTarget_BodySlotID _returnvalue_3)
      (Development|PrintString (Utilities|String|Append "BodyIDSLOT for Target Mesh " (Utilities|String|ToString(Integer) _returnvalue_3))))
    (:then_1
      ;;[1 1 1] Chassis
      (Variables|Default|SetSource_ChassisSlotID _returnvalue_5)
      (Development|PrintString (Utilities|String|Append "ChassisIDSLOT for Source Mesh " (Utilities|String|ToString(Integer) _returnvalue_5)))
      (Variables|Default|SetTarget_ChassisSlotID _returnvalue_4)
      (Development|PrintString (Utilities|String|Append "ChassisIDSLOT for Target Mesh " (Utilities|String|ToString(Integer) _returnvalue_4))))
    (:then_2
      ;;[1 1 1] Trim
      (Variables|Default|SetSource_TrimSlotID _returnvalue)
      (Development|PrintString (Utilities|String|Append "TrimIDSLOT for Source Mesh " (Utilities|String|ToString(Integer) _returnvalue))))
    (:then_3
      (Development|PrintString (Utilities|String|Append "Has Tertiary: " (Utilities|String|ToString(Boolean) _hascarbonfiberslot)))
      (if _hascarbonfiberslot
        ;;[1 1 1] Tertiary
        (Variables|Default|SetSource_CarbonFiberSlotID (StaticMesh|GetMaterialIndex (Variables|Default|GetSourceMesh) (Variables|Default|GetCarbonFiberSlotName)))))
    (:then_4
      (Development|PrintString (Utilities|String|Append "Has Plastic: " (Utilities|String|ToString(Boolean) _hasplasticslot)))
      (if _hasplasticslot
        ;;[1 1 1] Plastic
        (Variables|Default|SetSource_PlasticSlotID (StaticMesh|GetMaterialIndex (Variables|Default|GetSourceMesh) (Variables|Default|GetPlasticSlotName)))))))

(event Custom|SetDynamicMeshes
  (Utilities|FlowControl|Sequence
    (:then_0
      (DynamicMeshActor|ReleaseAllComputeMeshes)
      (bind _returnvalue (GeometryScript|Utility|CreateDynamicMeshPool))
      (Variables|Default|SetDynamic%20Mesh%20Pool _returnvalue))
    (:then_1
      (bind _returnvalue_2 (DynamicMeshActor|AllocateComputeMesh))
      (bind _returnvalue_3 (GeometryScript|StaticMesh|CopyMeshFromStaticMeshwithSectionMaterials (Variables|Default|GetTargetMesh) _returnvalue_2)
        (:Success
          (Variables|Default|SetDynamicMeshLow _returnvalue_3)
          (bind _returnvalue_4 (DynamicMeshActor|AllocateComputeMesh))
          (bind _returnvalue_5 (GeometryScript|StaticMesh|CopyMeshFromStaticMeshwithSectionMaterials (Variables|Default|GetSourceMesh) _returnvalue_4)
            (:Success
              (Variables|Default|SetDynamicMeshHigh _returnvalue_5))))))
    (:then_2
      ;[1 1 1] Split IDs Create Dynamic Meshes
      (bind _componentmeshes (GeometryScript|Decomposition|SplitMeshbyMaterialIDs (Variables|Default|GetDynamicMeshHigh) (Variables|Default|GetDynamic%20Mesh%20Pool)))
      (Variables|Default|SetDynamicMeshHighBody (Utilities|Array|Get(acopy) _componentmeshes (Variables|Default|GetSource_BodySlotID)))
      (bind _componentmeshes_2 (GeometryScript|Decomposition|SplitMeshbyMaterialIDs (Variables|Default|GetDynamicMeshLow) (Variables|Default|GetDynamic%20Mesh%20Pool)))
      (Variables|Default|SetDynamicMeshLowBody (Utilities|Array|Get(acopy) _componentmeshes_2 (Variables|Default|GetTarget_BodySlotID)))
      (bind _componentmeshes_3 (GeometryScript|Decomposition|SplitMeshbyMaterialIDs (Variables|Default|GetDynamicMeshHigh) (Variables|Default|GetDynamic%20Mesh%20Pool)))
      (Variables|Default|SetDynamicMeshHighChassis (Utilities|Array|Get(acopy) _componentmeshes_3 (Variables|Default|GetSource_ChassisSlotID)))
      (bind _componentmeshes_4 (GeometryScript|Decomposition|SplitMeshbyMaterialIDs (Variables|Default|GetDynamicMeshLow) (Variables|Default|GetDynamic%20Mesh%20Pool)))
      (Variables|Default|SetDynamicMeshLowChassis (Utilities|Array|Get(acopy) _componentmeshes_4 (Variables|Default|GetTarget_ChassisSlotID)))
      (bind _componentmeshes_5 (GeometryScript|Decomposition|SplitMeshbyMaterialIDs (Variables|Default|GetDynamicMeshHigh) (Variables|Default|GetDynamic%20Mesh%20Pool)))
      (Variables|Default|SetDynamicMeshHighTrim (Utilities|Array|Get(acopy) _componentmeshes_5 (Variables|Default|GetSource_TrimSlotID))))
    (:then_3
      (if (Variables|Default|GetHasCarbonFiberSlot)
        (bind _componentmeshes_6 (GeometryScript|Decomposition|SplitMeshbyMaterialIDs (Variables|Default|GetDynamicMeshHigh) (Variables|Default|GetDynamic%20Mesh%20Pool)))
        (Variables|Default|SetDynamicMeshHighCarbonFiber (Utilities|Array|Get(acopy) _componentmeshes_6 (Variables|Default|GetSource_CarbonFiberSlotID)))))
    (:then_4
      (if (Variables|Default|GetHasPlasticSlot)
        (bind _componentmeshes_7 (GeometryScript|Decomposition|SplitMeshbyMaterialIDs (Variables|Default|GetDynamicMeshHigh) (Variables|Default|GetDynamic%20Mesh%20Pool)))
        (bind _output_get (Variables|Default|SetDynamicMeshHighPlastic (Utilities|Array|Get(acopy) _componentmeshes_7 (Variables|Default|GetSource_PlasticSlotID))))
        (GeometryScript|MeshEdits|AppendMesh :TargetMesh (Variables|Default|GetDynamicMeshHighTrim) :AppendMesh _output_get :bDeferChangeNotifications false :AppendOptions "(CombineMode=EnableAllMatching)")))
    (:then_5
      (CallFunction|MakeCompositeDynamicMesh))))

(event Custom|ReleaseComputeMeshes
  (DynamicMeshActor|ReleaseAllComputeMeshes))

(event Custom|MakeCompositeDynamicMesh
  ;[1 1 1] Split IDs Create Dynamic Meshes
  (Utilities|FlowControl|Sequence
    (:then_0
      (bind _returnvalue (DynamicMeshActor|AllocateComputeMesh))
      (bind _returnvalue_2 (GeometryScript|MeshEdits|AppendMesh :TargetMesh _returnvalue :AppendMesh (Variables|Default|GetDynamicMeshHighTrim) :bDeferChangeNotifications false :AppendOptions "(CombineMode=EnableAllMatching)"))
      (bind _returnvalue_3 (GeometryScript|MeshEdits|AppendMesh :TargetMesh _returnvalue_2 :AppendMesh (Variables|Default|GetDynamicMeshHighChassis) :bDeferChangeNotifications false :AppendOptions "(CombineMode=EnableAllMatching)"))
      (Variables|Default|SetDynamicMeshHighChassisTarget _returnvalue_3)
      (if (Variables|Default|GetHasCarbonFiberSlot)
        (bind _returnvalue_4 (GeometryScript|MeshEdits|AppendMesh :TargetMesh (Variables|Default|GetDynamicMeshHighChassisTarget) :AppendMesh (Variables|Default|GetDynamicMeshHighCarbonFiber) :bDeferChangeNotifications false :AppendOptions "(CombineMode=EnableAllMatching)"))
        (Variables|Default|SetDynamicMeshHighChassisTarget _returnvalue_4)))
    (:then_1
      (Variables|Default|SetDynamicMeshHighChassisComposite (Variables|Default|GetDynamicMeshHighChassisTarget)))))

(fn GetSlotID
  (params (SK :type Skeletal%20Mesh%20Object%20Reference) (SlotName :type Name))
  (bind (_array_element _array_index) (Utilities|Array|ForEachLoopwithBreak (Variables|/Script/Engine.SkeletalMesh|GetMaterials SK))
    (:LoopBody
      (if (== (.MaterialSlotName (Utilities|Struct|BreakSkeletalMaterial _array_element)) SlotName)
        (Variables|Default|SetChassisSlotInvalid false)
        (return _array_index)))
    (:Completed
      (return 0))))

(fn SetVertexUV
  (params (TargetMesh :type Dynamic%20Mesh%20Object%20Reference) (VertexID :type Integer) (MoveToLocationUV :type Vector%202D%20Structure))
  (bind _target (LocalVariables|GetTarget))
  (bind _vertexuv (LocalVariables|GetVertexUV))
  (LocalVariables|SetTarget TargetMesh)
  (LocalVariables|SetVertex VertexID)
  (LocalVariables|SetMoveLocation MoveToLocationUV)
  (bind _uvlist (GeometryScript|UVs|GetMeshPerVertexUVs (LocalVariables|GetTarget)))
  (bind _uvarray (GeometryScript|ListUtils|ConvertUVListtoArray _uvlist))
  (LocalVariables|SetVertexUV (Utilities|Array|Get(acopy) _uvarray (LocalVariables|GetVertex)))
  (bind _triangles (GeometryScript|MeshQueries|GetVertexConnectedTriangles (LocalVariables|GetTarget) (LocalVariables|GetVertex)))
  (for _array_element _triangles
    (if (Utilities|Operators|Equal(==) (GeometryScript|MeshQueries|GetTriangleUVs _target 0 _array_element) _vertexuv 0.001)
      (GeometryScript|UVs|SetMeshTriangleUVs (LocalVariables|GetTarget) 0 _array_element (Utilities|Struct|MakeGeometryScriptUVTriangle (LocalVariables|GetMoveLocation) (.UV2 (GeometryScript|MeshQueries|GetTriangleUVs _target 0 _array_element)) (.UV3 (GeometryScript|MeshQueries|GetTriangleUVs _target 0 _array_element))) false)
      (elif (Utilities|Operators|Equal(==) (.UV2 (GeometryScript|MeshQueries|GetTriangleUVs _target 0 _array_element)) _vertexuv 0.001)
        (GeometryScript|UVs|SetMeshTriangleUVs (LocalVariables|GetTarget) 0 _array_element (Utilities|Struct|MakeGeometryScriptUVTriangle (GeometryScript|MeshQueries|GetTriangleUVs _target 0 _array_element) (LocalVariables|GetMoveLocation) (.UV3 (GeometryScript|MeshQueries|GetTriangleUVs _target 0 _array_element))) false)
        (elif (Utilities|Operators|Equal(==) _vertexuv (.UV3 (GeometryScript|MeshQueries|GetTriangleUVs _target 0 _array_element)) 0.001)
          (GeometryScript|UVs|SetMeshTriangleUVs (LocalVariables|GetTarget) 0 _array_element (Utilities|Struct|MakeGeometryScriptUVTriangle (GeometryScript|MeshQueries|GetTriangleUVs _target 0 _array_element) (.UV2 (GeometryScript|MeshQueries|GetTriangleUVs _target 0 _array_element)) (LocalVariables|GetMoveLocation)) false)
          (else
            _)))))
  (return))

(fn ConstructionScript)
