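# Material function for the vehicle cosmetics LOD tool that outputs carbon fiber material attributes.
# Tiled wrap textures supply darkened base color, with roughness derived from the normal map's green channel.
# The flattened normal feeds only the clear coat bottom-normal output, while scalar parameters drive both custom data inputs.
# A procedural sine-based weave height branch exists but does not reach the function output.
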MaterialFunction /Game/NeverCook/TechArt/VehicleCosmeticsTools/VC_LOD_TOOL/Materials/MF_CarbonFiberTool
graph {nodes [
  n1 = TextureCoordinate() {uTiling=16, vTiling=16}
  n2 = TextureSample(UVs=@n1) {samplerType=SAMPLERTYPE_Normal, texture=/Game/Weapons/WeaponSkins/Wraps/Textures/T_Wrap_CarbonFiber_N}
  ClearCoatNormalCustomOutput!(None=MaterialFunctionCall(Normal=@n2.RGB, Flatness=ScalarParameter() {defaultValue=0.25, parameterName=Flatten}) {materialFunction=&FlattenNormal}.Result)
  n3 = SetMaterialAttributes!(
   MaterialAttributes
   BaseColor=Multiply(A=TextureSample(UVs=@n1) {texture=/Game/Weapons/WeaponSkins/Wraps/Textures/T_Wrap_CarbonFiber_D}.RGB, B=Constant() {r=0.15})
   Metallic=Constant() {r=0.1}
   Specular
   Roughness=Add(A=@n2.G, B=Constant() {r=0.2})
   Anisotropy
   EmissiveColor
   Opacity
   OpacityMask
   Normal
   Tangent
   WorldPositionOffset
   Displacement
   SubsurfaceColor
   CustomData0=ScalarParameter() {defaultValue=1}
   CustomData1=ScalarParameter() {defaultValue=0.1, parameterName=Param_1}
  ) {attributeSetTypes [
    #1
    #2
    #3
    #4
    #5
    #6
    #7
    #8
    #9
    #10
    #11
    #12
    #13
    #14
    #15
   ]}
  FunctionOutput(None=@n3)
  Multiply(A=@n2.RGB)
  n4 = Multiply(A=TextureCoordinate() {vTiling=2}, B=Constant() {r=32})
  n5 = Multiply(A=@n4) {constB=2}
  n6 = Multiply(A=ComponentMask(None=Floor(None=@n5)) {g=true}) {constB=0.125}
  n7 = Multiply(A=@n4) {constB=0.5}
  n8 = ComponentMask(None=Sine(None=Add(A=@n6, B=@n7)) {period=0.5}) {r=true}
  n9 = Abs(None=@n8)
  NamedRerouteDeclaration(None=LinearInterpolate(A=@n9, B=Multiply(A=@n9) {constB=0.1}, Alpha=Ceil(None=@n8))) {name=CarbonFiberHeight, nodeColor=<rgba 1 0 0.0937 1>}
  n10 = Multiply(B=NamedRerouteUsage()) {constA=0.1}
  SetMaterialAttributes!(MaterialAttributes, BaseColor=@n10, Metallic=Constant(), Roughness=@n10) {attributeSetTypes [#1, #2, #4]}
  VectorParameter() {defaultValue=<rgba 0.026 0.026 0.026 0>}
 ]}
