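# Unlit material in the vehicle cosmetics LOD tool that composites a chassis base color, apparently for texture baking.
# It starts from the Chassis_BC texture parameter and uses the Mask texture's channels to layer surfaces over it.
# Green blends in a dark constant color, red adds carbon fiber base color from MF_CarbonFiberTool, and blue applies the same dark color last as plastic.
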Material /Game/NeverCook/TechArt/VehicleCosmeticsTools/VC_LOD_TOOL/Materials/MAT_BakeComposie_Chassis_BaseColor
graph {
 properties {pixelDepthOffsetMode=PDOM_Legacy, shadingModel=MSM_Unlit, usedShadingModels=Unlit}
 nodes [
  n1 = TextureSampleParameter2D() {parameterName=Chassis_BC, texture=/Game/NeverCook/TechArt/VehicleCosmeticsTools/VC_LOD_TOOL/Textures/T_Black}
  n2 = Constant3Vector() {constant=<rgba 0.02 0.02 0.02 1>}
  n3 = TextureSampleParameter2D() {parameterName=Mask, texture=&T_Black}
  n4 = MaterialFunctionCall() {materialFunction=&/MF_CarbonFiberTool}
  LinearInterpolate(
   A=LinearInterpolate(
    A=LinearInterpolate(A=@n1.RGB, B=@n2, Alpha=@n3.G)
    B=GetMaterialAttributes(None=@n4.Result) {attributeGetTypes [#1]}.Base Color
    Alpha=@n3.R
   ) {desc=CF}
   B=@n2
   Alpha=@n3.B
  ) {desc=Plastic}
 ]
}
