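# An unlit material in the vehicle cosmetics LOD tool that computes a composite mask, apparently for baking, stored under NeverCook as a tool asset.
# It samples two texture parameters, Chassis and Trim, subtracts the Trim RGB from the Chassis RGB, and clamps the result.
# Both parameters default to the same white composite mask texture, so the result is black until different textures are supplied.
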Material /Game/NeverCook/TechArt/VehicleCosmeticsTools/VC_LOD_TOOL/Materials/MAT_BakeCompositeMask
graph {
 properties {pixelDepthOffsetMode=PDOM_Legacy, shadingModel=MSM_Unlit, usedShadingModels=Unlit}
 nodes [Clamp(None=Subtract(
    A=TextureSampleParameter2D() {parameterName=Chassis, texture=/Game/NeverCook/TechArt/VehicleCosmeticsTools/VC_LOD_TOOL/Textures/T_WhiteCompositMask}.RGB
    B=TextureSampleParameter2D() {parameterName=Trim, texture=&T_WhiteCompositMask}.RGB
   ))]
}
