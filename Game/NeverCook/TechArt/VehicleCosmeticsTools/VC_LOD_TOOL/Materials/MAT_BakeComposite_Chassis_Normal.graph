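# Unlit material in the vehicle cosmetics LOD tool that, by its name, composites the chassis normal map for baking.
# It blends normal inputs A and B by the mask's green channel.
# The red channel (CF) adds only the XY of the normal from MF_CarbonFiberTool onto that blend.
# The blue channel (Plastic) then blends back to plain B, overriding the carbon fiber detail.
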Material /Game/NeverCook/TechArt/VehicleCosmeticsTools/VC_LOD_TOOL/Materials/MAT_BakeComposite_Chassis_Normal
graph {
 properties {pixelDepthOffsetMode=PDOM_Legacy, shadingModel=MSM_Unlit, usedShadingModels=Unlit}
 nodes [
  n1 = TextureSampleParameter2D() {parameterName=B, texture=/Game/NeverCook/TechArt/VehicleCosmeticsTools/VC_LOD_TOOL/Textures/T_Black}
  n2 = TextureSampleParameter2D() {parameterName=Mask, texture=&T_Black}
  n3 = LinearInterpolate(
   A=TextureSampleParameter2D() {parameterName=A, texture=&T_Black}.RGBA
   B=@n1.RGBA
   Alpha=@n2.G
  )
  n4 = MaterialFunctionCall() {materialFunction=&/MF_CarbonFiberTool}
  n5 = Multiply(
   A=Constant4Vector() {constant=<rgba 1 1 0 0>}.RGBA
   B=AppendVector(A=GetMaterialAttributes(None=@n4.Result) {attributeGetTypes [#1]}.Normal, B=Constant())
  )
  LinearInterpolate(A=LinearInterpolate(A=@n3, B=Add(A=@n3, B=@n5), Alpha=@n2.R) {desc=CF}, B=@n1.RGBA, Alpha=@n2.B) {desc=Plastic}
 ]
}
