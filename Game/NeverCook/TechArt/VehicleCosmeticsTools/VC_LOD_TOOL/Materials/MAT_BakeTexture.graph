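# MAT_BakeTexture is an unlit material in the vehicle cosmetics LOD tool, under NeverCook TechArt, named for texture baking.
# Its only node is a 2D texture sample parameter named Texture, which defaults to the tool's T_Black texture.
# The graph shows no math or other nodes, and no link from the sample to a material output.
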Material /Game/NeverCook/TechArt/VehicleCosmeticsTools/VC_LOD_TOOL/Materials/MAT_BakeTexture
graph {
 properties {pixelDepthOffsetMode=PDOM_Legacy, shadingModel=MSM_Unlit, usedShadingModels=Unlit}
 nodes [TextureSampleParameter2D() {parameterName=Texture, texture=/Game/NeverCook/TechArt/VehicleCosmeticsTools/VC_LOD_TOOL/Textures/T_Black}]
}
