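# Unlit alpha-composite material in the vehicle cosmetics LOD tool that bakes chassis textures from LOD0 chassis and body inputs.
# A composite mask layers body values (G), carbon fiber from MF_CarbonFiberTool (R) and plastic (B) over the chassis texture.
# The color chain blends body red with metallic values.
# A parallel chain blends chassis alpha with roughness values, then saturates and inverts it, consistent with baking into alpha.
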Material /Game/NeverCook/TechArt/VehicleCosmeticsTools/VC_LOD_TOOL/Materials/MAT_Bake_ChassisTextures
graph {
 properties {blendMode=BLEND_AlphaComposite, pixelDepthOffsetMode=PDOM_Legacy, shadingModel=MSM_Unlit, usedShadingModels=Unlit}
 nodes [
  n1 = TextureSampleParameter2D() {desc=ChassisTextures LOD0, parameterName=A, texture=/Game/NeverCook/TechArt/VehicleCosmeticsTools/VC_LOD_TOOL/Textures/T_Black}
  n2 = TextureSampleParameter2D() {desc=BodyTextures LOD0, parameterName=B, texture=&T_Black}
  n3 = MaterialFunctionCall(R=@n2.R, G=@n2.R, B=ScalarParameter() {parameterName=Metallic}) {materialFunction=&AppendMany}
  n4 = TextureSampleParameter2D() {desc=CompositMask, parameterName=Mask, texture=&T_Black}
  n5 = GetMaterialAttributes(None=MaterialFunctionCall() {materialFunction=&/MF_CarbonFiberTool}.Result) {attributeGetTypes [#1, #2]}
  LinearInterpolate(A=LinearInterpolate(A=LinearInterpolate(A=@n1.RGB, B=@n3.RGB, Alpha=@n4.G), B=@n5.Metallic, Alpha=@n4.R) {desc=CF}, B=@n3.RGB, Alpha=@n4.B) {constB=0, desc=Plastic}
  n6 = ScalarParameter() {defaultValue=0.65, parameterName=Roughness}
  n7 = LinearInterpolate(A=LinearInterpolate(A=@n1.A, B=@n6, Alpha=@n4.G), B=@n5.Roughness, Alpha=@n4.R) {desc=CF}
  OneMinus(None=Saturate(None=LinearInterpolate(A=@n7, B=@n6, Alpha=@n4.B) {constB=0.65, desc=Plastic}))
 ]
}
