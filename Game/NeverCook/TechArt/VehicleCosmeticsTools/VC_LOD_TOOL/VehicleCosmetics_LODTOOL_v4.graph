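# Editor utility widget that bakes LOD textures for vehicle cosmetics (body normal and chassis base color, normal, textures, E and M maps) from a source mesh onto a target mesh.
# Each bake button finds or spawns a VehicleCosmetics_LODBP5 actor to supply per-slot dynamic meshes, then runs Geometry Script bakes saved as texture assets.
# A material-slot mask bake drives optional compositing of chassis maps through materials drawn to render targets.
# Delays separate baking, compositing and texture-setting steps.
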EditorUtilityWidgetBlueprint /Game/NeverCook/TechArt/VehicleCosmeticsTools/VC_LOD_TOOL/VehicleCosmetics_LODTOOL_v4
graphs [
 BakeTexture {nodes [
   n1 = CreateDynamicMaterialInstance(Parent=/Game/NeverCook/TechArt/VehicleCosmeticsTools/VC_LOD_TOOL/Materials/MAT_BakeTexture, OptionalName=None, CreationFlags=None)
   > n2 = Get Texture Parameter Value!(execute, self:Material Instance Dynamic Object Reference=@n1, ParameterName:Name=Texture) -> (then, ReturnValue:Texture Object Reference) {is=title K2Node_CallFunction}
   > Cast To Texture2D(Object=@n2)
   n3 = BakeTexture!() -> (then, Target:Dynamic Mesh Object Reference, Source:Dynamic Mesh Object Reference, SourceTexture:Texture 2D Object Reference) {is=title K2Node_FunctionEntry}
   n4 = Get Target Transform() -> (Target Transform:Transform)
   > n5 = BakeTexture!(
    execute
    TargetMesh:Dynamic Mesh Object Reference=@n3
    TargetTransform:Transform=@n4
    TargetOptions:Geometry Script Bake Target Mesh Options Structure=Make Geometry Script Bake Target Mesh Options!(TargetUVLayer:int=0) -> (GeometryScriptBakeTargetMeshOptions:Geometry Script Bake Target Mesh Options Structure) {is=title K2Node_MakeStruct}
    SourceMesh:Dynamic Mesh Object Reference=@n3.Source
    SourceTransform:Transform=@n4
    SourceOptions:Geometry Script Bake Source Mesh Options Structure=Make Geometry Script Bake Source Mesh Options!(SourceNormalMap:Texture 2D Object Reference, SourceNormalUVLayer:int=0, SourceNormalSpace:EGeometryScriptBakeNormalSpace Enum=Tangent) -> (GeometryScriptBakeSourceMeshOptions:Geometry Script Bake Source Mesh Options Structure) {is=title K2Node_MakeStruct}
    BakeTypes:Array of Geometry Script Bake Type Options Structures=Make Array("[0]":Geometry Script Bake Type Options Structure=MakeBakeTypeTexture!(SourceTexture:Texture 2D Object Reference=@n3.SourceTexture, SourceUVLayer:int=0) -> (ReturnValue:Geometry Script Bake Type Options Structure)) -> (Array:Array of Geometry Script Bake Type Options Structures)
    BakeOptions:Geometry Script Bake Texture Options Structure=Make Geometry Script Bake Texture Options!(
     Resolution:EGeometryScriptBakeResolution Enum=Get Resolution() -> (Resolution:EGeometryScriptBakeResolution Enum=Resolution16)
     BitDepth:EGeometryScriptBakeBitDepth Enum=ChannelBits8
     GutterSize:int=4
     SamplesPerPixel:EGeometryScriptBakeSamplesPerPixel Enum=Get Samples Per Pixel() -> (Samples Per Pixel:EGeometryScriptBakeSamplesPerPixel Enum=Sample1)
     SampleFilterMask:Texture 2D Object Reference
     FilteringType:EGeometryScriptBakeFilteringType Enum=BSpline
     ProjectionDistance:float=Get Projection Distance() -> (Projection Distance:float)
     bProjectionInWorldSpace:bool=False
    ) -> (GeometryScriptBakeTextureOptions:Geometry Script Bake Texture Options Structure) {is=title K2Node_MakeStruct}
   ) -> (then, ReturnValue:Array of Texture 2D Object References) {is=title K2Node_CallFunction}
   n6 = "Get (a copy)"(Array:Array of Texture 2D Object References=@n5) -> (Output:Texture 2D Object Reference)
   > Branch(Condition=IsValid(Object=@n6))
   > Return Node!(execute, OutTexture:Texture 2D Object Reference=@n6)
   SetTextureParameterValue(self=@n1, ParameterName=Texture)
  ]}
 CreateCompositChassisBC {nodes [
   CreateCompositChassisBC() {is=title K2Node_FunctionEntry}
   > n1 = CreateDynamicMaterialInstance(
    Parent=/Game/NeverCook/TechArt/VehicleCosmeticsTools/VC_LOD_TOOL/Materials/MAT_BakeComposie_Chassis_BaseColor
    OptionalName=None
    CreationFlags=None
   )
   > Set DMIC(DMIC:Material Instance Dynamic Object Reference=@n1) -> (Output_Get:Material Instance Dynamic Object Reference)
   > n2 = Get!(execute:Exec) -> (then:Exec, else:Exec, Chassis_BC_LOD:Texture 2D Object Reference)
   > SetTextureParameterValue(self=Get DMIC() -> (DMIC:Material Instance Dynamic Object Reference), ParameterName=Chassis_BC, Value=@n2)
   > SetTextureParameterValue(self=Get DMIC(), ParameterName=Mask, Value=Get ChassisCompositeMask_LOD() -> (ChassisCompositeMask_LOD:Texture 2D Object Reference))
   > ClearRenderTarget2D!(
    execute
    WorldContextObject:Object Reference
    TextureRenderTarget:Texture Render Target 2D Object Reference=Get TextureRenderTarget16() -> (TextureRenderTarget16:Texture Render Target 2D Object Reference)
    ClearColor:Linear Color Structure=(R=0,G=0,B=0,A=1)
   ) -> (then)
   > DrawMaterialToRenderTarget!(
    execute
    WorldContextObject:Object Reference
    TextureRenderTarget:Texture Render Target 2D Object Reference=Get TextureRenderTarget16()
    Material:Material Interface Object Reference=Get DMIC()
   ) -> (then)
   > ConvertRenderTargetToTexture2DEditorOnly(RenderTarget=Get TextureRenderTarget16(), Texture=Get Chassis_BC_LOD() -> (Chassis_BC_LOD:Texture 2D Object Reference))
   Render Target 2D Create Static Texture 2D Editor Only!(
    execute
    RenderTarget:Texture Render Target 2D Object Reference=/Game/NeverCook/TechArt/VehicleCosmeticsTools/VC_LOD_TOOL/Textures/RT_CompositA
    Name:String=T_Chassis_BC_LOD
    CompressionSettings:TextureCompressionSettings Enum=TC_Default
    MipSettings:TextureMipGenSettings Enum=TMGS_FromTextureGroup
   ) -> (then, ReturnValue:Texture 2D Object Reference)
   Set SRGB(SRGB=true)
  ]}
 CreateCompositChassisMask {nodes [
   n1 = CreateCompositChassisMask!() -> (then, A:Texture 2D Object Reference, B:Texture 2D Object Reference, OverrideTexture:Texture 2D Object Reference) {is=title K2Node_FunctionEntry}
   > n2 = CreateDynamicMaterialInstance(Parent=/Game/NeverCook/TechArt/VehicleCosmeticsTools/VC_LOD_TOOL/Materials/MAT_BakeCompositeMask, OptionalName=None, CreationFlags=None)
   > Set DMIC(DMIC=@n2)
   > SetTextureParameterValue(self=Get DMIC(), ParameterName=Chassis, Value=@n1)
   > SetTextureParameterValue(self=Get DMIC(), ParameterName=Trim, Value=@n1.B)
   > ClearRenderTarget2D(
    TextureRenderTarget=Get TextureRenderTarget8() -> (TextureRenderTarget8:Texture Render Target 2D Object Reference)
    ClearColor=(R=0,G=0,B=0,A=1)
   )
   > DrawMaterialToRenderTarget(TextureRenderTarget=Get TextureRenderTarget8(), Material=Get DMIC())
   > ConvertRenderTargetToTexture2DEditorOnly(RenderTarget=Get TextureRenderTarget8(), Texture=@n1.OverrideTexture)
   Set SRGB(SRGB=true)
  ]}
 CreateCompositChassisNormal {nodes [
   n1 = CreateCompositChassisNormal!() -> (then, A:Texture 2D Object Reference, B:Texture 2D Object Reference, OverrideTexture:Texture 2D Object Reference) {is=title K2Node_FunctionEntry}
   > n2 = CreateDynamicMaterialInstance(
    Parent=/Game/NeverCook/TechArt/VehicleCosmeticsTools/VC_LOD_TOOL/Materials/MAT_BakeComposite_Chassis_Normal
    OptionalName=None
    CreationFlags=None
   )
   > Set DMIC(DMIC=@n2)
   > SetTextureParameterValue(self=Get DMIC(), ParameterName=A, Value=@n1)
   > SetTextureParameterValue(self=Get DMIC(), ParameterName=B, Value=@n1.B)
   > SetTextureParameterValue(self=Get DMIC(), ParameterName=Mask, Value=Get ChassisCompositeMask_LOD())
   > ClearRenderTarget2D(TextureRenderTarget=Get TextureRenderTarget8(), ClearColor=(R=0,G=0,B=0,A=1))
   > DrawMaterialToRenderTarget(TextureRenderTarget=Get TextureRenderTarget8(), Material=Get DMIC())
   > ConvertRenderTargetToTexture2DEditorOnly(RenderTarget=Get TextureRenderTarget8(), Texture=@n1.OverrideTexture)
   Set SRGB(SRGB=true)
  ]}
 CreateCompositChassisTextures {nodes [
   n1 = CreateCompositChassisTextures!() -> (then, A:Texture 2D Object Reference, B:Texture 2D Object Reference, OverrideTexture:Texture 2D Object Reference) {is=title K2Node_FunctionEntry}
   > n2 = CreateDynamicMaterialInstance(Parent=/Game/NeverCook/TechArt/VehicleCosmeticsTools/VC_LOD_TOOL/Materials/MAT_Bake_ChassisTextures, OptionalName=None, CreationFlags=None)
   > Set DMIC(DMIC=@n2)
   > SetTextureParameterValue(self=Get DMIC(), ParameterName=A, Value=@n1)
   > SetTextureParameterValue(self=Get DMIC(), ParameterName=B, Value=@n1.B)
   > SetTextureParameterValue(self=Get DMIC(), ParameterName=Mask, Value=Get ChassisCompositeMask_LOD())
   > ClearRenderTarget2D(TextureRenderTarget=Get TextureRenderTarget8(), ClearColor=(R=0,G=0,B=0,A=1))
   > DrawMaterialToRenderTarget(TextureRenderTarget=Get TextureRenderTarget8(), Material=Get DMIC())
   > ConvertRenderTargetToTexture2DEditorOnly(RenderTarget=Get TextureRenderTarget8(), Texture=@n1.OverrideTexture)
   Set SRGB(SRGB=true)
  ]}
 EventGraph {nodes [
   Clear(TargetArray=Get Assets to Export() -> (Assets to Export:Array of Strings))
   > K2Node_Composite_0 = ExportEvents!(execute) -> (Out:Exec) {is=title K2Node_Composite}
   > n1 = GetAssetTools()
   n2 = Get Assets to Export()
   > ExportAssetsWithDialog!(execute, self:Asset Tools Interface=@n1, AssetsToExport:Array of Strings=@n2, bPromptForIndividualFilenames:bool=true) -> (then)
   n3 = CreateDynamicMaterialInstance(Parent=&MAT_BakeCompositeMask, OptionalName=None, CreationFlags=None)
   > n4 = Get Texture Parameter Value(self=@n3, ParameterName=White)
   > Cast To Texture2D(Object=@n4)
   "On Clicked (EditorUtilityButton_BakeBodyNormal)"() {is=title K2Node_ComponentBoundEvent}
   > n5 = Sequence!(execute) -> (then_0, then_1, then_2:Exec, then_3:Exec)
   > GetDynamicActor()
   InitializeDynamicActor(execute=@n5.then_1)
   BakeBodyNormal(execute=@n5.then_2)
   "On Clicked (EditorUtilityButton_BakeChassis_E)"() {is=title K2Node_ComponentBoundEvent}
   > n6 = Sequence()
   > GetDynamicActor()
   InitializeDynamicActor(execute=@n6.then_1)
   BakeChassis_E(execute=@n6.then_2)
   "On Clicked (SetPathButton)"() {is=title K2Node_ComponentBoundEvent}
   > n7 = GetSelectedAssets()
   n8 = CreateAssetData(InAsset="Get (a copy)"(Array:Array of Object References=@n7) -> (Output:Object Reference), bAllowBlueprintClass=false)
   > Set SavePath(SavePath:String="To String (Name)"(InName=Break Asset Data(AssetData=@n8).PackagePath)) -> (Output_Get:String)
   "On Clicked (EditorUtilityButton_BakeCompositeMask)"() {is=title K2Node_ComponentBoundEvent}
   > n9 = Sequence()
   > GetDynamicActor()
   InitializeDynamicActor(execute=@n9.then_1)
   BakeCompositeMask_Initialize(execute=@n9.then_2)
   n10 = "On Check State Changed (EditorUtilityCheckBox_N_Compute)"!() -> (OutputDelegate, then, bIsChecked:bool) {is=title K2Node_ComponentBoundEvent}
   > Set ComputeNormals(ComputeNormals:bool=@n10.bIsChecked) -> (Output_Get:bool)
   "On Clicked (EditorUtilityButton_BakeAll)"() {is=title K2Node_ComponentBoundEvent}
   > n11 = Sequence()
   > GetDynamicActor()
   InitializeDynamicActor(execute=@n11.then_1)
   BakeAll(execute=@n11.then_2)
   "On Clicked (EditorUtilityButton_BakeChassisBC)"() {is=title K2Node_ComponentBoundEvent}
   > n12 = Sequence()
   > GetDynamicActor()
   InitializeDynamicActor(execute=@n12.then_1)
   BakeChassisBC(execute=@n12.then_2)
   "On Clicked (EditorUtilityButton_BakeChassisNormal)"() {is=title K2Node_ComponentBoundEvent}
   > n13 = Sequence()
   > GetDynamicActor()
   InitializeDynamicActor(execute=@n13.then_1)
   BakeChassisNormal(execute=@n13.then_2)
   "On Clicked (EditorUtilityButton_BakeChassisTextures)"() {is=title K2Node_ComponentBoundEvent}
   > n14 = Sequence()
   > GetDynamicActor()
   InitializeDynamicActor(execute=@n14.then_1)
   BakeChassisTextures(execute=@n14.then_2)
   "On Clicked (EditorUtilityButton_BakeChassis_M)"() {is=title K2Node_ComponentBoundEvent}
   > n15 = Sequence()
   > GetDynamicActor()
   InitializeDynamicActor(execute=@n15.then_1)
   BakeChassis_M(execute=@n15.then_2)
   n16 = "On Check State Changed (EditorUtilityCheckBox_BC_Composit)"!() -> (OutputDelegate, then, bIsChecked:bool) {is=title K2Node_ComponentBoundEvent}
   > Set CompositeChassisBC(CompositeChassisBC:bool=@n16.bIsChecked) -> (Output_Get:bool)
   n17 = "On Check State Changed (EditorUtilityCheckBox_CTextures_Composit)"!() -> (OutputDelegate, then, bIsChecked:bool) {is=title K2Node_ComponentBoundEvent}
   > Set CompositeChassisTextures(CompositeChassisTextures:bool=@n17.bIsChecked) -> (Output_Get:bool)
   n18 = "On Check State Changed (EditorUtilityCheckBox_N_Composit)"!() -> (OutputDelegate, then, bIsChecked:bool) {is=title K2Node_ComponentBoundEvent}
   > Set CompositeChassisN(CompositeChassisN:bool=@n18.bIsChecked) -> (Output_Get:bool)
   BakeChassis_E() {is=title K2Node_CustomEvent}
   n19 = Get DynamicMeshActor() -> (DynamicMeshActor:Vehicle Cosmetics LODBP5 Object Reference)
   > n20 = BakeTexture!(
    execute
    self:Self Object Reference
    Target:Dynamic Mesh Object Reference=Get DynamicMeshLowChassis!(self:Vehicle Cosmetics LODBP5 Object Reference=@n19) -> (DynamicMeshLowChassis:Dynamic Mesh Object Reference)
    Source:Dynamic Mesh Object Reference=Get DynamicMeshHighChassis!(self:Vehicle Cosmetics LODBP5 Object Reference=@n19) -> (DynamicMeshHighChassis:Dynamic Mesh Object Reference)
    SourceTexture:Texture 2D Object Reference=Get Chassis_E_Source() -> (Chassis_E_Source:Texture 2D Object Reference)
   ) -> (then, OutTexture:Texture 2D Object Reference)
   n21 = Get SavePath() -> (SavePath:String)
   > n22 = CreateNewTexture2DAsset!(
    execute
    FromTexture:Texture 2D Object Reference=@n20
    AssetPathAndName:String=Append!(
     A=Append!(A=@n21, B=/) -> (ReturnValue)
     B=T_
     C:String=Get VehicleName() -> (VehicleName:String)
     D:String=_
     E:String=Chassis_E
     F:String=_LOD
    ) -> (ReturnValue)
    Options:Geometry Script Create New Texture 2DAsset Options Structure=Make Geometry Script Create New Texture 2DAsset Options!(bOverwriteIfExists:bool=true) -> (GeometryScriptCreateNewTexture2DAssetOptions:Geometry Script Create New Texture 2DAsset Options Structure) {is=title K2Node_MakeStruct}
   ) -> (Failure:Exec, Success:Exec, ReturnValue:Texture 2D Object Reference)
   Set Chassis_E_LOD(execute=@n22.Success, Chassis_E_LOD:Texture 2D Object Reference=@n22) -> (Output_Get:Texture 2D Object Reference)
   > Delay!(execute, WorldContextObject:Object Reference, Duration=1) -> (then)
   > Set CompressionSettings(CompressionSettings=TC_Default, self=Get Chassis_E_LOD() -> (Chassis_E_LOD:Texture 2D Object Reference))
   n23 = Get Chassis_E_LOD()
   > Set CompressionNoAlpha!(execute, CompressionNoAlpha:bool=true, self:Texture Object Reference=@n23) -> (then, Output_Get:bool)
   > Set LODGroup!(execute, LODGroup:TextureGroup Enum=TEXTUREGROUP_Project22, self:Texture Object Reference=@n23) -> (then, Output_Get:TextureGroup Enum)
   InitializeDynamicActor() {is=title K2Node_CustomEvent}
   > Set TargetMesh!(
    execute
    TargetMesh:Static Mesh Object Reference=Get TargetMesh() -> (TargetMesh:Static Mesh Object Reference)
    self:Vehicle Cosmetics LODBP5 Object Reference=Get DynamicMeshActor()
   ) -> (then, Output_Get:Static Mesh Object Reference)
   > Set SourceMesh!(
    execute
    SourceMesh:Static Mesh Object Reference=Get SourceMesh() -> (SourceMesh:Static Mesh Object Reference)
    self:Vehicle Cosmetics LODBP5 Object Reference=Get DynamicMeshActor()
   ) -> (then, Output_Get:Static Mesh Object Reference)
   > SetSlotIDs(self:Vehicle Cosmetics LODBP5 Object Reference=Get DynamicMeshActor())
   > SetDynamicMeshes(self:Vehicle Cosmetics LODBP5 Object Reference=Get DynamicMeshActor())
   BakeCompositeMask() {is=title K2Node_CustomEvent}
   n24 = Get CompositeArray() -> (CompositeArray:Array of Texture 2D Object References)
   n25 = Get DynamicMeshActor()
   n26 = Get Target Transform()
   > n27 = BakeTexture!(
    execute
    TargetMesh:Dynamic Mesh Object Reference=Get DynamicMeshLowChassis(self=@n25)
    TargetTransform:Transform=@n26
    TargetOptions:Geometry Script Bake Target Mesh Options Structure=Make Geometry Script Bake Target Mesh Options(TargetUVLayer=0)
    SourceMesh:Dynamic Mesh Object Reference=Get DynamicMeshHighChassisComposite!(self:Vehicle Cosmetics LODBP5 Object Reference=@n25) -> (DynamicMeshHighChassisComposite:Dynamic Mesh Object Reference)
    SourceTransform:Transform=@n26
    SourceOptions:Geometry Script Bake Source Mesh Options Structure=Make Geometry Script Bake Source Mesh Options(SourceNormalUVLayer=0, SourceNormalSpace=Tangent)
    BakeTypes:Array of Geometry Script Bake Type Options Structures=Make Array("[0]"=MakeBakeTypeMultiTexture!(MaterialIDSourceTextures:Array of Texture 2D Object References=@n24, SourceUVLayer:int=0) -> (ReturnValue:Geometry Script Bake Type Options Structure))
    BakeOptions:Geometry Script Bake Texture Options Structure=Make Geometry Script Bake Texture Options(
     Resolution=Get Resolution() -> (Resolution=Resolution16)
     BitDepth=ChannelBits8
     GutterSize=4
     SamplesPerPixel=Get Samples Per Pixel() -> (Samples Per Pixel=Sample1)
     FilteringType=BSpline
     ProjectionDistance=Get Projection Distance()
     bProjectionInWorldSpace=False
    )
   ) -> (then, ReturnValue:Array of Texture 2D Object References)
   n28 = "Get (a copy)"(Array:Array of Texture 2D Object References=@n27) -> (Output:Texture 2D Object Reference)
   > Branch(Condition=IsValid(Object=@n28))
   n29 = Get SavePath()
   > n30 = CreateNewTexture2DAsset(
    FromTexture=@n28
    AssetPathAndName=Append!(A=Append!(A=@n29, B=/) -> (ReturnValue), B=T_, C:String=Get VehicleName(), D:String=_, E:String=CompositeMask) -> (ReturnValue)
    Options=Make Geometry Script Create New Texture 2DAsset Options(bOverwriteIfExists=true)
   )
   Set ChassisCompositeMask_LOD(execute=@n30.Success, ChassisCompositeMask_LOD:Texture 2D Object Reference=@n30) -> (Output_Get:Texture 2D Object Reference)
   > Delay(Duration=1)
   n31 = Get ChassisCompositeMask_LOD()
   > Set CompressionSettings(CompressionSettings=TC_Default, self=@n31)
   > Set LODGroup(LODGroup=TEXTUREGROUP_Project22, self=@n31)
   > Set SRGB(SRGB=false, self=@n31)
   > Set CompressionNoAlpha(CompressionNoAlpha=true, self=@n31)
   BakeChassisTextures() {is=title K2Node_CustomEvent}
   n32 = Get DynamicMeshActor()
   > n33 = BakeTexture!(
    execute
    self:Self Object Reference
    Target:Dynamic Mesh Object Reference=Get DynamicMeshLowChassis(self=@n32)
    Source:Dynamic Mesh Object Reference=Get DynamicMeshHighChassis(self=@n32)
    SourceTexture:Texture 2D Object Reference=Get ChassisTextures_Source() -> (ChassisTextures_Source:Texture 2D Object Reference)
   ) -> (then, OutTexture:Texture 2D Object Reference)
   n34 = Get SavePath()
   > n35 = CreateNewTexture2DAsset(
    FromTexture=@n33
    AssetPathAndName=Append!(A=Append!(A=@n34, B=/) -> (ReturnValue), B=T_, C:String=Get VehicleName(), D:String=_, E:String=ChassisTextures, F:String=_LOD) -> (ReturnValue)
    Options=Make Geometry Script Create New Texture 2DAsset Options(bOverwriteIfExists=true)
   )
   Set ChassisTextures_LOD(execute=@n35.Success, ChassisTextures_LOD:Texture 2D Object Reference=@n35) -> (Output_Get:Texture 2D Object Reference)
   > n36 = Branch(Condition=Get CompositeChassisTextures() -> (CompositeChassisTextures:bool))
   > n37 = Branch(Condition=IsValid(Object=Get ChassisCompositeMask_LOD()))
   Delay(execute=[@n43.then, @n37.then], Duration=3)
   n38 = Get DynamicMeshActor()
   > n39 = BakeTexture(
    Target=Get DynamicMeshLowChassis(self=@n38)
    Source=Get DynamicMeshHighTrim!(self:Vehicle Cosmetics LODBP5 Object Reference=@n38) -> (DynamicMeshHighTrim:Dynamic Mesh Object Reference)
    SourceTexture=Get BodyTextures_Source() -> (BodyTextures_Source:Texture 2D Object Reference)
   )
   > Set ChassisTextures_LOD_TRIM(ChassisTextures_LOD_TRIM:Texture 2D Object Reference=@n39) -> (Output_Get:Texture 2D Object Reference)
   > Delay(Duration=3)
   n40 = Get ChassisTextures_LOD() -> (ChassisTextures_LOD:Texture 2D Object Reference)
   > CreateCompositChassisTextures!(
    execute
    self
    A:Texture 2D Object Reference=@n40
    B:Texture 2D Object Reference=Get ChassisTextures_LOD_TRIM() -> (ChassisTextures_LOD_TRIM:Texture 2D Object Reference)
    OverrideTexture:Texture 2D Object Reference=@n40
   ) -> (then) {is=title K2Node_CallFunction}
   > n41 = Delay(Duration=2)
   n42 = Get ChassisTextures_LOD()
   Set CompressionSettings(execute=[@n41.then, @n44.then], CompressionSettings=TC_Default, self=@n42)
   > Set SRGB(SRGB=false, self=@n42)
   > Set LODGroup(LODGroup=TEXTUREGROUP_Project22, self=@n42)
   n43 = BakeCompositeMask_Initialize(execute=@n37.else)
   n44 = Delay(execute=@n36.else, Duration=1)
   ExportChassisTextures() {is=title K2Node_CustomEvent}
   > n45 = Get!(execute) -> (then, else, ChassisTextures_LOD:Texture 2D Object Reference)
   > SetExportAssetsName!(execute, self, Texture:Texture 2D Object Reference=@n45) -> (then)
   ExportCompositeMask() {is=title K2Node_CustomEvent}
   > n46 = Get!(execute) -> (then, else, ChassisCompositeMask_LOD:Texture 2D Object Reference)
   > SetExportAssetsName(Texture=@n46)
   ExportBodyNormal() {is=title K2Node_CustomEvent}
   > n47 = Get!(execute) -> (then, else, Body_N_LOD:Texture 2D Object Reference)
   > SetExportAssetsName(Texture=@n47)
   ExportChassisBaseColor() {is=title K2Node_CustomEvent}
   > n48 = Get!(execute) -> (then, else, Chassis_BC_LOD:Texture 2D Object Reference)
   > SetExportAssetsName(Texture=@n48)
   ExportChassisNormal() {is=title K2Node_CustomEvent}
   > n49 = Get!(execute) -> (then, else, Chassis_N_LOD:Texture 2D Object Reference)
   > SetExportAssetsName(Texture=@n49)
   ExportChassisE() {is=title K2Node_CustomEvent}
   > n50 = Get!(execute) -> (then, else, Chassis_E_LOD:Texture 2D Object Reference)
   > SetExportAssetsName(Texture=@n50)
   ExportChassisM() {is=title K2Node_CustomEvent}
   > n51 = Get!(execute) -> (then, else, Chassis_M_LOD:Texture 2D Object Reference)
   > SetExportAssetsName(Texture=@n51)
   GetDynamicActor() {is=title K2Node_CustomEvent}
   > n52 = Branch(Condition=Get ComputeNormals() -> (ComputeNormals:bool))
   > Set Mesh Build Settings(Mesh Build Settings:Mesh Build Settings Structure=Make Mesh Build Settings!(
     bUseMikkTSpace:bool=True
     bRecomputeNormals:bool=True
     bRecomputeTangents:bool=True
     bComputeWeightedNormals:bool=true
     bRemoveDegenerates:bool=True
     bBuildReversedIndexBuffer:bool=True
     bUseHighPrecisionTangentBasis:bool=False
     bUseFullPrecisionUVs:bool=False
     bUseBackwardsCompatibleF16TruncUVs:bool=False
     bGenerateLightmapUVs:bool=false
     bGenerateDistanceFieldAsIfTwoSided:bool=False
     bSupportFaceRemap:bool=False
     MinLightmapResolution:int=64
     SrcLightmapIndex:int=0
     DstLightmapIndex:int=1
     BuildScale3D:Vector="1,1,1"
     DistanceFieldResolutionScale:float=1
     DistanceFieldReplacementMesh:Static Mesh Object Reference
     MaxLumenMeshCards:int=12
    ) -> (MeshBuildSettings:Mesh Build Settings Structure) {is=title K2Node_MakeStruct}) -> (Output_Get:Mesh Build Settings Structure)
   > n53 = SetLodBuildSettings!(
    execute
    self:Static Mesh Editor Subsystem Object Reference=Get StaticMeshEditorSubsystem()
    StaticMesh:Static Mesh Object Reference=Get TargetMesh()
    LodIndex:int=0
    BuildOptions:"Mesh Build Settings Structure (by ref)"=Get Mesh Build Settings() -> (Mesh Build Settings:Mesh Build Settings Structure)
   ) -> (then)
   n54 = GetAllActorsOfClass!(
    execute=[@n53.then, @n58.then]
    WorldContextObject:Object Reference
    ActorClass:Actor Class Reference=/Game/NeverCook/TechArt/VehicleCosmeticsTools/VC_LOD_TOOL/BP/VehicleCosmetics_LODBP5.VehicleCosmetics_LODBP5_C
   ) -> (then, OutActors:Array of Vehicle Cosmetics LODBP5 Object References)
   > n55 = Set DynamicMeshActor(DynamicMeshActor:Vehicle Cosmetics LODBP5 Object Reference="Get (a copy)"(Array:Array of Vehicle Cosmetics LODBP5 Object References=@n54) -> (Output:Vehicle Cosmetics LODBP5 Object Reference)) -> (Output_Get:Vehicle Cosmetics LODBP5 Object Reference)
   > n56 = Branch(Condition=IsValid(Object=@n55))
   n57 = SpawnActor Vehicle Cosmetics LODBP5!(
    execute:Exec=@n56.else
    WorldContextObject:Object Reference
    Class:Actor Class Reference=&VehicleCosmetics_LODBP5_C
    SpawnTransform_Location:Vector="0, 0, 0"
    SpawnTransform_Rotation:Rotator="0, 0, 0"
    SpawnTransform_Scale:Vector="1,1,1"
    CollisionHandlingOverride:ESpawnActorCollisionHandlingMethod Enum=Undefined
    TransformScaleMethod:ESpawnActorScaleMethod Enum=MultiplyWithRoot
    Owner:Actor Object Reference
    Instigator:Pawn Object Reference
   ) -> (then:Exec, ReturnValue:Vehicle Cosmetics LODBP5 Object Reference) {is=title K2Node_SpawnActorFromClass}
   > Set DynamicMeshActor(DynamicMeshActor=@n57)
   Set Mesh Build Settings(
    execute=@n52.else
    Mesh Build Settings=Make Mesh Build Settings(
     bUseMikkTSpace=True
     bRecomputeNormals=false
     bRecomputeTangents=True
     bComputeWeightedNormals=False
     bRemoveDegenerates=True
     bBuildReversedIndexBuffer=True
     bUseHighPrecisionTangentBasis=False
     bUseFullPrecisionUVs=False
     bUseBackwardsCompatibleF16TruncUVs=False
     bGenerateLightmapUVs=false
     bGenerateDistanceFieldAsIfTwoSided=False
     bSupportFaceRemap=False
     MinLightmapResolution=64
     SrcLightmapIndex=0
     DstLightmapIndex=1
     BuildScale3D="1,1,1"
     DistanceFieldResolutionScale=1
     MaxLumenMeshCards=12
    )
   )
   > n58 = SetLodBuildSettings(self=Get StaticMeshEditorSubsystem(), StaticMesh=Get TargetMesh(), LodIndex=0, BuildOptions=Get Mesh Build Settings())
   BakeChassis_M() {is=title K2Node_CustomEvent}
   n59 = Get DynamicMeshActor()
   > n60 = BakeTexture(
    Target=Get DynamicMeshLowChassis(self=@n59)
    Source=Get DynamicMeshHighChassis(self=@n59)
    SourceTexture=Get Chassis_M_Source() -> (Chassis_M_Source:Texture 2D Object Reference)
   )
   n61 = Get SavePath()
   > n62 = CreateNewTexture2DAsset(
    FromTexture=@n60
    AssetPathAndName=Append!(A=Append!(A=@n61, B=/) -> (ReturnValue), B=T_, C:String=Get VehicleName(), D:String=_, E:String=Chassis_M, F:String=_LOD) -> (ReturnValue)
    Options=Make Geometry Script Create New Texture 2DAsset Options(bOverwriteIfExists=true)
   )
   Set Chassis_M_LOD(execute=@n62.Success, Chassis_M_LOD:Texture 2D Object Reference=@n62) -> (Output_Get:Texture 2D Object Reference)
   > Delay(Duration=1)
   > Set CompressionSettings(CompressionSettings=TC_Default, self=Get Chassis_M_LOD() -> (Chassis_M_LOD:Texture 2D Object Reference))
   n63 = Get Chassis_M_LOD()
   > Set CompressionNoAlpha(CompressionNoAlpha=true, self=@n63)
   > Set LODGroup(LODGroup=TEXTUREGROUP_Project22, self=@n63)
   n64 = BakeAll() {is=title K2Node_CustomEvent}
   n65 = Sequence!(execute=[@n102.then, @n64.then, @n104.else, @n101.then]) -> (then_0, then_1, then_2, then_3, then_4:Exec, then_5:Exec)
   > BakeBodyNormal()
   BakeChassisBC(execute=@n65.then_1)
   BakeChassisNormal(execute=@n65.then_2)
   BakeChassisTextures(execute=@n65.then_3) {is=title K2Node_CallFunction}
   BakeChassis_E(execute=@n65.then_4) {is=title K2Node_CallFunction}
   BakeChassis_M(execute=@n65.then_5) {is=title K2Node_CallFunction}
   BakeCompositeMask_Initialize() {is=title K2Node_CustomEvent}
   > Set Array Elem!(
    execute
    TargetArray:Array of Texture 2D Object References=Get CompositeArray()
    Index:int=Get Source_ChassisSlotID!(self:Vehicle Cosmetics LODBP5 Object Reference=Get DynamicMeshActor()) -> (Source_ChassisSlotID:int=0)
    Item:"Texture 2D Object Reference (by ref)"=Get TextureBlack() -> (TextureBlack:Texture 2D Object Reference)
    bSizeToFit:bool=false
   ) -> (then)
   > Set Array Elem(
    TargetArray=Get CompositeArray()
    Index=Get Source_TrimSlotID!(self:Vehicle Cosmetics LODBP5 Object Reference=Get DynamicMeshActor()) -> (Source_TrimSlotID:int=0)
    Item=Get TextureGreen() -> (TextureGreen:Texture 2D Object Reference)
    bSizeToFit=false
   )
   > n66 = Branch(Condition=Get HasCarbonFiberSlot!(self:Vehicle Cosmetics LODBP5 Object Reference=Get DynamicMeshActor()) -> (HasCarbonFiberSlot:bool))
   > n67 = Set Array Elem(
    TargetArray=Get CompositeArray()
    Index=Get Source_CarbonFiberSlotID!(self:Vehicle Cosmetics LODBP5 Object Reference=Get DynamicMeshActor()) -> (Source_CarbonFiberSlotID:int=0)
    Item=Get TextureRed() -> (TextureRed:Texture 2D Object Reference)
    bSizeToFit=false
   )
   n68 = Branch(
    execute=[@n67.then, @n66.else]
    Condition=Get HasPlasticSlot!(self:Vehicle Cosmetics LODBP5 Object Reference=Get DynamicMeshActor()) -> (HasPlasticSlot:bool)
   )
   > Set Array Elem(
    TargetArray=Get CompositeArray()
    Index=Get Source_PlasticSlotID!(self:Vehicle Cosmetics LODBP5 Object Reference=Get DynamicMeshActor()) -> (Source_PlasticSlotID:int=0)
    Item=Get TextureBlue() -> (TextureBlue:Texture 2D Object Reference)
    bSizeToFit=false
   )
   > BakeCompositeMask() {is=title K2Node_CallFunction}
   BakeCompositeMask(execute=@n68.else)
   ReleaseComputeMeshes() {is=title K2Node_CustomEvent}
   > ReleaseAllComputeMeshes(self:Dynamic Mesh Actor Object Reference=Get DynamicMeshActor())
   BakeBodyNormal() {is=title K2Node_CustomEvent}
   n69 = Get DynamicMeshActor()
   n70 = Get Target Transform()
   > n71 = BakeTexture!(
    execute
    TargetMesh:Dynamic Mesh Object Reference=Get DynamicMeshLowBody!(self:Vehicle Cosmetics LODBP5 Object Reference=@n69) -> (DynamicMeshLowBody:Dynamic Mesh Object Reference)
    TargetTransform:Transform=@n70
    TargetOptions:Geometry Script Bake Target Mesh Options Structure=Make Geometry Script Bake Target Mesh Options(TargetUVLayer=0)
    SourceMesh:Dynamic Mesh Object Reference=Get DynamicMeshHighBody!(self:Vehicle Cosmetics LODBP5 Object Reference=@n69) -> (DynamicMeshHighBody:Dynamic Mesh Object Reference)
    SourceTransform:Transform=@n70
    SourceOptions:Geometry Script Bake Source Mesh Options Structure=Make Geometry Script Bake Source Mesh Options(SourceNormalUVLayer=0, SourceNormalSpace=Tangent)
    BakeTypes:Array of Geometry Script Bake Type Options Structures=Make Array("[0]"=MakeBakeTypeTangentNormal!() -> (ReturnValue:Geometry Script Bake Type Options Structure))
    BakeOptions:Geometry Script Bake Texture Options Structure=Make Geometry Script Bake Texture Options(
     Resolution=Get Resolution() -> (Resolution=Resolution16)
     BitDepth=ChannelBits8
     GutterSize=4
     SamplesPerPixel=Get Samples Per Pixel() -> (Samples Per Pixel=Sample1)
     FilteringType=BSpline
     ProjectionDistance=Get Projection Distance()
     bProjectionInWorldSpace=False
    )
   ) -> (then, ReturnValue:Array of Texture 2D Object References)
   n72 = Get SavePath()
   > n73 = CreateNewTexture2DAsset(
    FromTexture="Get (a copy)"(Array:Array of Texture 2D Object References=@n71) -> (Output:Texture 2D Object Reference)
    AssetPathAndName=Append!(A=Append!(A=@n72, B=/) -> (ReturnValue), B=T_, C:String=Get VehicleName(), D:String=_, E:String=BodyNormal, F:String=_LOD) -> (ReturnValue)
    Options=Make Geometry Script Create New Texture 2DAsset Options(bOverwriteIfExists=true)
   )
   n74 = Set Body_N_LOD(execute=@n73.Success, Body_N_LOD:Texture 2D Object Reference=@n73) -> (Output_Get:Texture 2D Object Reference)
   > Delay(Duration=2)
   > Set CompressionSettings(CompressionSettings=TC_Normalmap, self=@n74)
   > Set CompressFinal!(execute, CompressFinal:bool=true, self:Texture Object Reference=@n74) -> (then, Output_Get:bool)
   > Set LODGroup(LODGroup=TEXTUREGROUP_Project21, self=@n74)
   BakeChassisNormal() {is=title K2Node_CustomEvent}
   > PrintString(InString=Baking Chassis Normal, bPrintToScreen=false)
   n75 = Get Target Transform()
   n76 = Get DynamicMeshActor()
   > n77 = BakeTexture(
    TargetMesh=Get DynamicMeshLowChassis(self=@n76)
    TargetTransform=@n75
    TargetOptions=Make Geometry Script Bake Target Mesh Options(TargetUVLayer=0)
    SourceMesh=Get DynamicMeshHighChassis(self=@n76)
    SourceTransform=@n75
    SourceOptions=Make Geometry Script Bake Source Mesh Options(SourceNormalMap=Get Chassis_N_Source() -> (Chassis_N_Source:Texture 2D Object Reference), SourceNormalUVLayer=0, SourceNormalSpace=Tangent)
    BakeTypes=Make Array("[0]"=MakeBakeTypeTangentNormal())
    BakeOptions=Make Geometry Script Bake Texture Options(
     Resolution=Get Resolution() -> (Resolution=Resolution16)
     BitDepth=ChannelBits8
     GutterSize=4
     SamplesPerPixel=Get Samples Per Pixel() -> (Samples Per Pixel=Sample1)
     FilteringType=BSpline
     ProjectionDistance=Get Projection Distance()
     bProjectionInWorldSpace=False
    )
   )
   n78 = Get SavePath()
   > n79 = CreateNewTexture2DAsset(
    FromTexture="Get (a copy)"(Array=@n77)
    AssetPathAndName=Append!(A=Append!(A=@n78, B=/) -> (ReturnValue), B=T_, C:String=Get VehicleName(), D:String=_, E:String=Chassis_N, F:String=_LOD) -> (ReturnValue)
    Options=Make Geometry Script Create New Texture 2DAsset Options(bOverwriteIfExists=true)
   )
   Set Chassis_N_LOD(execute=@n79.Success, Chassis_N_LOD:Texture 2D Object Reference=@n79) -> (Output_Get:Texture 2D Object Reference)
   > n80 = Branch(Condition=Get CompositeChassisN() -> (CompositeChassisN:bool))
   > n81 = Branch(Condition=IsValid(Object=Get ChassisCompositeMask_LOD()))
   Delay(execute=[@n88.then, @n81.then], Duration=2)
   n82 = Get Target Transform()
   n83 = Get DynamicMeshActor()
   > n84 = BakeTexture(
    TargetMesh=Get DynamicMeshLowChassis(self=@n83)
    TargetTransform=@n82
    TargetOptions=Make Geometry Script Bake Target Mesh Options(TargetUVLayer=0)
    SourceMesh=Get DynamicMeshHighTrim(self=@n83)
    SourceTransform=@n82
    SourceOptions=Make Geometry Script Bake Source Mesh Options(SourceNormalUVLayer=0, SourceNormalSpace=Tangent)
    BakeTypes=Make Array("[0]"=MakeBakeTypeTangentNormal())
    BakeOptions=Make Geometry Script Bake Texture Options(
     Resolution=Get Resolution() -> (Resolution=Resolution16)
     BitDepth=ChannelBits8
     GutterSize=4
     SamplesPerPixel=Get Samples Per Pixel() -> (Samples Per Pixel=Sample1)
     FilteringType=BSpline
     ProjectionDistance=Get Projection Distance()
     bProjectionInWorldSpace=False
    )
   )
   > Set ChassisN_LOD_TRIM(ChassisN_LOD_TRIM:Texture 2D Object Reference="Get (a copy)"(Array=@n84)) -> (Output_Get:Texture 2D Object Reference)
   > Delay(Duration=2)
   n85 = Get Chassis_N_LOD() -> (Chassis_N_LOD:Texture 2D Object Reference)
   > CreateCompositChassisNormal!(
    execute
    self
    A:Texture 2D Object Reference=@n85
    B:Texture 2D Object Reference=Get ChassisN_LOD_TRIM() -> (ChassisN_LOD_TRIM:Texture 2D Object Reference)
    OverrideTexture:Texture 2D Object Reference=@n85
   ) -> (then) {is=title K2Node_CallFunction}
   > n86 = Delay(Duration=2)
   Set CompressionSettings(execute=[@n89.then, @n86.then], CompressionSettings=TC_Normalmap, self=Get Chassis_N_LOD())
   n87 = Get Chassis_N_LOD()
   > Set CompressFinal(CompressFinal=true, self=@n87)
   > Set LODGroup(LODGroup=TEXTUREGROUP_Project21, self=@n87)
   n88 = BakeCompositeMask_Initialize(execute=@n81.else) {is=title K2Node_CallFunction}
   n89 = Delay(execute=@n80.else, Duration=1)
   BakeChassisBC() {is=title K2Node_CustomEvent}
   > PrintString(InString=Baking Chassis BC Texture, bPrintToScreen=false)
   n90 = Get DynamicMeshActor()
   > n91 = BakeTexture!(
    execute
    self:Self Object Reference
    Target:Dynamic Mesh Object Reference=Get DynamicMeshLowChassis(self=@n90)
    Source:Dynamic Mesh Object Reference=Get DynamicMeshHighChassis(self=@n90)
    SourceTexture:Texture 2D Object Reference=Get ChassisBC_Source() -> (ChassisBC_Source:Texture 2D Object Reference)
   ) -> (then, OutTexture:Texture 2D Object Reference)
   n92 = Get SavePath()
   > n93 = CreateNewTexture2DAsset(
    FromTexture=@n91
    AssetPathAndName=Append!(A=Append!(A=@n92, B=/) -> (ReturnValue), B=T_, C:String=Get VehicleName(), D:String=_, E:String=Chassis_BC, F:String=_LOD) -> (ReturnValue)
    Options=Make Geometry Script Create New Texture 2DAsset Options(bOverwriteIfExists=true)
   )
   Set Chassis_BC_LOD(execute=@n93.Success, Chassis_BC_LOD:Texture 2D Object Reference=@n93) -> (Output_Get:Texture 2D Object Reference)
   > n94 = Branch(Condition=Get CompositeChassisBC() -> (CompositeChassisBC:bool))
   > PrintString(InString=Compositing Chassis BC, bPrintToScreen=false)
   > n95 = Branch(Condition=IsValid(Object=Get ChassisCompositeMask_LOD()))
   Delay(execute=[@n98.then, @n95.then], Duration=2)
   > CreateCompositChassisBC() {is=title K2Node_CallFunction}
   > n96 = Delay(Duration=2)
   Set CompressionSettings(execute=[@n99.then, @n96.then], CompressionSettings=TC_Default, self=Get Chassis_BC_LOD())
   n97 = Get Chassis_BC_LOD()
   > Set SRGB(SRGB=true, self=@n97)
   > Set CompressionNoAlpha(CompressionNoAlpha=true, self=@n97)
   > Set LODGroup(LODGroup=TEXTUREGROUP_Project20, self=@n97)
   n98 = BakeCompositeMask_Initialize(execute=@n95.else)
   n99 = Delay(execute=@n94.else, Duration=1)
   Event PreConstruct()
   > SetObject(self=Get DetailsView_62() -> (DetailsView_62:Details View Object Reference), NewObject=Self-Reference())
   > SetObject(self=Get DetailsView_124() -> (DetailsView_124:Details View Object Reference), NewObject=Self-Reference())
   > SetObject(self=Get DetailsView_257() -> (DetailsView_257:Details View Object Reference), NewObject=Self-Reference())
   > SetObject(self=Get DetailsView_497() -> (DetailsView_497:Details View Object Reference), NewObject=Self-Reference())
   > SetObject(self=Get DetailsView_69() -> (DetailsView_69:Details View Object Reference), NewObject=Self-Reference())
   > SetObject(self=Get DetailsView_150() -> (DetailsView_150:Details View Object Reference), NewObject=Self-Reference())
   n100 = Branch(Condition=Get CompositeChassisN())
   n101 = Branch(execute=[@n104.then, @n100.then, @n103.then], Condition=IsValid(Object=Get ChassisCompositeMask_LOD()))
   n102 = BakeCompositeMask_Initialize(execute=@n101.else)
   n103 = Branch(execute=@n100.else, Condition=Get CompositeChassisTextures())
   n104 = Branch(execute=@n103.else, Condition=Get CompositeChassisBC())
   LoadAsset!(
    execute
    self:Editor Asset Subsystem Object Reference=Get EditorAssetSubsystem() -> (ReturnValue:Editor Asset Subsystem Object Reference) {is=title K2Node_GetEditorSubsystem}
    AssetPath=GetSystemPath!(Object:Object Reference=@n93) -> (ReturnValue:String)
   ) -> (then, ReturnValue)
   ReleaseComputeMeshes() {is=title K2Node_CallFunction}
   ExportAssets!(execute, self:Asset Tools Interface=@n1, AssetsToExport:Array of Strings=@n2, ExportPath:String) -> (then)
   Event Construct()
  ]}
 ExportEvents {
  path=EventGraph.K2Node_Composite_0.ExportEvents
  nodes [
   Inputs()
   > ExportCompositeMask() {is=title K2Node_CallFunction}
   > ExportBodyNormal() {is=title K2Node_CallFunction}
   > ExportChassisBaseColor() {is=title K2Node_CallFunction}
   > ExportChassisNormal() {is=title K2Node_CallFunction}
   > ExportChassisE() {is=title K2Node_CallFunction}
   > ExportChassisM() {is=title K2Node_CallFunction}
   > ExportChassisTextures() {is=title K2Node_CallFunction}
   > Outputs!(Out:Exec)
  ]
 }
 SetExportAssetsName {nodes [
   n1 = SetExportAssetsName!() -> (then, Texture:Texture 2D Object Reference) {is=title K2Node_FunctionEntry}
   > Add(TargetArray=Get Assets to Export(), NewItem=Get Object Path String(Object=@n1)) {is=title K2Node_CallArrayFunction}
  ]}
]
variables [
 Assets to Export {type="array<string>", default []}
 Body_N_LOD {type=Texture2D}
 BodyTextures_Source {type=Texture2D}
 Chassis_BC_LOD {type=Texture2D}
 Chassis_E_LOD {type=Texture2D}
 Chassis_E_Source {type=Texture2D}
 Chassis_M_LOD {type=Texture2D}
 Chassis_M_Source {type=Texture2D}
 Chassis_N_LOD {type=Texture2D}
 Chassis_N_Source {type=Texture2D}
 ChassisBC_Source {type=Texture2D}
 ChassisCompositeMask_LOD {type=Texture2D}
 ChassisN_LOD_TRIM {type=Texture2D}
 ChassisTextures_LOD {type=Texture2D}
 ChassisTextures_LOD_TRIM {type=Texture2D}
 ChassisTextures_Source {type=Texture2D}
 CompositeArray {type="array<Texture2D>", default [null,null,null,null,null,null]}
 CompositeChassisBC {type=boolean, default=true}
 CompositeChassisN {type=boolean, default=true}
 CompositeChassisTextures {type=boolean, default=true}
 CompositeMask_ChassisWhite {type=Texture2D}
 ComputeNormals {type=boolean, default=true}
 DynamicMeshActor {type=VehicleCosmetics_LODBP5_C, default=null}
 IsLoading {type=boolean}
 Mesh Build Settings {
  type=MeshBuildSettings
  default {
   bBuildReversedIndexBuffer=true
   bComputeWeightedNormals=false
   bGenerateDistanceFieldAsIfTwoSided=false
   bGenerateLightmapUVs=true
   bRecomputeNormals=true
   bRecomputeTangents=true
   bRemoveDegenerates=true
   bSupportFaceRemap=false
   bUseBackwardsCompatibleF16TruncUVs=false
   bUseFullPrecisionUVs=false
   bUseHighPrecisionTangentBasis=false
   bUseMikkTSpace=true
   distanceFieldReplacementMesh=null
   distanceFieldResolutionScale=1
   dstLightmapIndex=1
   maxLumenMeshCards=12
   minLightmapResolution=64
   srcLightmapIndex=0
   buildScale3D=<1 1 1>
  }
 }
 Projection Distance {type=number, default=3}
 Resolution {type=EGeometryScriptBakeResolution, default=Resolution2048}
 Samples Per Pixel {type=EGeometryScriptBakeSamplesPerPixel, default=Sample64}
 SavePath {type=string, default=/Game/NeverCook/TechArt/VehicleCosmeticsTools/VC_LOD_TOOL}
 SourceMesh {type=StaticMesh, default=null}
 Target Transform {type=Transform, default {location=<0 0 0>, rotation=<rot 0 0 0>, scale=<1 1 1>}}
 TargetMesh {type=StaticMesh, default=null}
 TextureBlack {type=Texture2D, default=/Game/NeverCook/TechArt/VehicleCosmeticsTools/VC_LOD_TOOL/Textures/T_Black}
 TextureBlue {type=Texture2D, default=/Game/NeverCook/TechArt/VehicleCosmeticsTools/VC_LOD_TOOL/Textures/T_Blue}
 TextureGreen {type=Texture2D, default=/Game/NeverCook/TechArt/VehicleCosmeticsTools/VC_LOD_TOOL/Textures/T_Green}
 TextureRed {type=Texture2D, default=/Game/NeverCook/TechArt/VehicleCosmeticsTools/VC_LOD_TOOL/Textures/T_Red}
 TextureRenderTarget16 {type=TextureRenderTarget2D, default=&RT_CompositA}
 TextureRenderTarget8 {type=TextureRenderTarget2D, default=/Game/NeverCook/TechArt/VehicleCosmeticsTools/VC_LOD_TOOL/Textures/RT_CompositB}
 VehicleName {type=string, default=VEHICLENAME}
]
widgets {
 info {rootWidgetClass=&Overlay, widgetCount=122}
 widgets [
  Overlay_423 {widgetClassPath=&Overlay, widget=Overlay_423}
  ThresholdScrollBox_2 {
   widgetClassPath=/Script/FortniteUI.ThresholdScrollBox
   parent=Overlay_423
   widget=ThresholdScrollBox_2
   properties {
    widgetBarStyle {
     thickness=8
     draggedThumbImage {
      bIsDynamicallyLoaded=false
      drawAs=RoundedBox
      imageType=NoImage
      mirroring=NoMirror
      resourceName=None
      resourceObject=null
      tiling=NoTile
      imageSize=<0 0>
      margin=<ltrb 0 0 0 0>
      outlineSettings {
       bUseBrushTransparency=false
       roundingType=FixedRadius
       width=0
       color {colorUseRule=UseColor_Specified, specifiedColor=<rgba 0 0 0 0>}
       cornerRadii=<4 4 4 4>
      }
      tintColor {colorUseRule=UseColor_Specified, specifiedColor=<rgba 0.2159 0.2159 0.2159 1>}
      uVRegion {bIsValid=false, max=<0 0>, min=<0 0>}
     }
     horizontalBackgroundImage {
      bIsDynamicallyLoaded=false
      drawAs=NoDrawType
      imageType=NoImage
      mirroring=NoMirror
      resourceName=None
      resourceObject=null
      tiling=NoTile
      imageSize=<0 0>
      margin=<ltrb 0 0 0 0>
      outlineSettings {
       bUseBrushTransparency=false
       roundingType=HalfHeightRadius
       width=0
       color {colorUseRule=UseColor_Specified, specifiedColor=<rgba 0 0 0 0>}
       cornerRadii=<0 0 0 0>
      }
      tintColor {colorUseRule=UseColor_Specified, specifiedColor=<rgba 1 1 1 1>}
      uVRegion {bIsValid=false, max=<0 0>, min=<0 0>}
     }
     horizontalBottomSlotImage {
      bIsDynamicallyLoaded=false
      drawAs=NoDrawType
      imageType=NoImage
      mirroring=NoMirror
      resourceName=None
      resourceObject=null
      tiling=NoTile
      imageSize=<0 0>
      margin=<ltrb 0 0 0 0>
      outlineSettings {
       bUseBrushTransparency=false
       roundingType=HalfHeightRadius
       width=0
       color {colorUseRule=UseColor_Specified, specifiedColor=<rgba 0 0 0 0>}
       cornerRadii=<0 0 0 0>
      }
      tintColor {colorUseRule=UseColor_Specified, specifiedColor=<rgba 1 1 1 1>}
      uVRegion {bIsValid=false, max=<0 0>, min=<0 0>}
     }
     horizontalTopSlotImage {
      bIsDynamicallyLoaded=false
      drawAs=NoDrawType
      imageType=NoImage
      mirroring=NoMirror
      resourceName=None
      resourceObject=null
      tiling=NoTile
      imageSize=<0 0>
      margin=<ltrb 0 0 0 0>
      outlineSettings {
       bUseBrushTransparency=false
       roundingType=HalfHeightRadius
       width=0
       color {colorUseRule=UseColor_Specified, specifiedColor=<rgba 0 0 0 0>}
       cornerRadii=<0 0 0 0>
      }
      tintColor {colorUseRule=UseColor_Specified, specifiedColor=<rgba 1 1 1 1>}
      uVRegion {bIsValid=false, max=<0 0>, min=<0 0>}
     }
     hoveredThumbImage {
      bIsDynamicallyLoaded=false
      drawAs=RoundedBox
      imageType=NoImage
      mirroring=NoMirror
      resourceName=None
      resourceObject=null
      tiling=NoTile
      imageSize=<0 0>
      margin=<ltrb 0 0 0 0>
      outlineSettings {
       bUseBrushTransparency=false
       roundingType=FixedRadius
       width=0
       color {colorUseRule=UseColor_Specified, specifiedColor=<rgba 0 0 0 0>}
       cornerRadii=<4 4 4 4>
      }
      tintColor {colorUseRule=UseColor_Specified, specifiedColor=<rgba 0.2159 0.2159 0.2159 1>}
      uVRegion {bIsValid=false, max=<0 0>, min=<0 0>}
     }
     normalThumbImage {
      bIsDynamicallyLoaded=false
      drawAs=RoundedBox
      imageType=NoImage
      mirroring=NoMirror
      resourceName=None
      resourceObject=null
      tiling=NoTile
      imageSize=<0 0>
      margin=<ltrb 0 0 0 0>
      outlineSettings {
       bUseBrushTransparency=false
       roundingType=FixedRadius
       width=0
       color {colorUseRule=UseColor_Specified, specifiedColor=<rgba 0 0 0 0>}
       cornerRadii=<4 4 4 4>
      }
      tintColor {colorUseRule=UseColor_Specified, specifiedColor=<rgba 0.0953 0.0953 0.0953 1>}
      uVRegion {bIsValid=false, max=<0 0>, min=<0 0>}
     }
     verticalBackgroundImage {
      bIsDynamicallyLoaded=false
      drawAs=NoDrawType
      imageType=NoImage
      mirroring=NoMirror
      resourceName=None
      resourceObject=null
      tiling=NoTile
      imageSize=<0 0>
      margin=<ltrb 0 0 0 0>
      outlineSettings {
       bUseBrushTransparency=false
       roundingType=HalfHeightRadius
       width=0
       color {colorUseRule=UseColor_Specified, specifiedColor=<rgba 0 0 0 0>}
       cornerRadii=<0 0 0 0>
      }
      tintColor {colorUseRule=UseColor_Specified, specifiedColor=<rgba 1 1 1 1>}
      uVRegion {bIsValid=false, max=<0 0>, min=<0 0>}
     }
     verticalBottomSlotImage {
      bIsDynamicallyLoaded=false
      drawAs=NoDrawType
      imageType=NoImage
      mirroring=NoMirror
      resourceName=None
      resourceObject=null
      tiling=NoTile
      imageSize=<0 0>
      margin=<ltrb 0 0 0 0>
      outlineSettings {
       bUseBrushTransparency=false
       roundingType=HalfHeightRadius
       width=0
       color {colorUseRule=UseColor_Specified, specifiedColor=<rgba 0 0 0 0>}
       cornerRadii=<0 0 0 0>
      }
      tintColor {colorUseRule=UseColor_Specified, specifiedColor=<rgba 1 1 1 1>}
      uVRegion {bIsValid=false, max=<0 0>, min=<0 0>}
     }
     verticalTopSlotImage {
      bIsDynamicallyLoaded=false
      drawAs=NoDrawType
      imageType=NoImage
      mirroring=NoMirror
      resourceName=None
      resourceObject=null
      tiling=NoTile
      imageSize=<0 0>
      margin=<ltrb 0 0 0 0>
      outlineSettings {
       bUseBrushTransparency=false
       roundingType=HalfHeightRadius
       width=0
       color {colorUseRule=UseColor_Specified, specifiedColor=<rgba 0 0 0 0>}
       cornerRadii=<0 0 0 0>
      }
      tintColor {colorUseRule=UseColor_Specified, specifiedColor=<rgba 1 1 1 1>}
      uVRegion {bIsValid=false, max=<0 0>, min=<0 0>}
     }
    }
    widgetStyle {
     barThickness=8
     bottomShadowBrush {
      bIsDynamicallyLoaded=false
      drawAs=Box
      imageType=FullColor
      mirroring=NoMirror
      resourceName=../../../Engine/Content/Slate/Common/ScrollBoxShadowBottom.png
      resourceObject=null
      tiling=NoTile
      imageSize=<16 8>
      margin=<ltrb 0.5 0 0.5 1>
      outlineSettings {
       bUseBrushTransparency=false
       roundingType=HalfHeightRadius
       width=0
       color {colorUseRule=UseColor_Specified, specifiedColor=<rgba 0 0 0 0>}
       cornerRadii=<0 0 0 0>
      }
      tintColor {colorUseRule=UseColor_Specified, specifiedColor=<rgba 1 1 1 1>}
      uVRegion {bIsValid=false, max=<0 0>, min=<0 0>}
     }
     horizontalScrolledContentPadding=<ltrb 0 0 1 0>
     leftShadowBrush {
      bIsDynamicallyLoaded=false
      drawAs=Box
      imageType=FullColor
      mirroring=NoMirror
      resourceName=../../../Engine/Content/Slate/Common/ScrollBoxShadowLeft.png
      resourceObject=null
      tiling=NoTile
      imageSize=<8 16>
      margin=<ltrb 1 0.5 0 0.5>
      outlineSettings {
       bUseBrushTransparency=false
       roundingType=HalfHeightRadius
       width=0
       color {colorUseRule=UseColor_Specified, specifiedColor=<rgba 0 0 0 0>}
       cornerRadii=<0 0 0 0>
      }
      tintColor {colorUseRule=UseColor_Specified, specifiedColor=<rgba 1 1 1 1>}
      uVRegion {bIsValid=false, max=<0 0>, min=<0 0>}
     }
     rightShadowBrush {
      bIsDynamicallyLoaded=false
      drawAs=Box
      imageType=FullColor
      mirroring=NoMirror
      resourceName=../../../Engine/Content/Slate/Common/ScrollBoxShadowRight.png
      resourceObject=null
      tiling=NoTile
      imageSize=<8 16>
      margin=<ltrb 0 0.5 1 0.5>
      outlineSettings {
       bUseBrushTransparency=false
       roundingType=HalfHeightRadius
       width=0
       color {colorUseRule=UseColor_Specified, specifiedColor=<rgba 0 0 0 0>}
       cornerRadii=<0 0 0 0>
      }
      tintColor {colorUseRule=UseColor_Specified, specifiedColor=<rgba 1 1 1 1>}
      uVRegion {bIsValid=false, max=<0 0>, min=<0 0>}
     }
     topShadowBrush {
      bIsDynamicallyLoaded=false
      drawAs=Box
      imageType=FullColor
      mirroring=NoMirror
      resourceName=../../../Engine/Content/Slate/Common/ScrollBoxShadowTop.png
      resourceObject=null
      tiling=NoTile
      imageSize=<16 8>
      margin=<ltrb 0.5 1 0.5 0>
      outlineSettings {
       bUseBrushTransparency=false
       roundingType=HalfHeightRadius
       width=0
       color {colorUseRule=UseColor_Specified, specifiedColor=<rgba 0 0 0 0>}
       cornerRadii=<0 0 0 0>
      }
      tintColor {colorUseRule=UseColor_Specified, specifiedColor=<rgba 1 1 1 1>}
      uVRegion {bIsValid=false, max=<0 0>, min=<0 0>}
     }
     verticalScrolledContentPadding=<ltrb 0 0 0 1>
    }
   }
   slot {class=OverlaySlot, properties {horizontalAlignment=HAlign_Fill, verticalAlignment=VAlign_Fill}}
  }
  VerticalBox_1 {
   widgetClassPath=&VerticalBox
   parent=ThresholdScrollBox_2
   widget=VerticalBox_1
   slot {class=ScrollBoxSlot, properties {size {sizeRule=Fill, value=1}}}
  }
  ScaleBox_7 {
   widgetClassPath=&ScaleBox
   parent=VerticalBox_1
   widget=ScaleBox_7
   slot {class=VerticalBoxSlot, properties {horizontalAlignment=HAlign_Center, padding=<ltrb 16 16 16 0>}}
  }
  VerticalBox_TopLogo {widgetClassPath=&VerticalBox, parent=ScaleBox_7, widget=VerticalBox_TopLogo, slot {class=ScaleBoxSlot}}
  HorizontalBox_13 {widgetClassPath=&HorizontalBox, parent=VerticalBox_TopLogo, widget=HorizontalBox_13, slot {class=VerticalBoxSlot}}
  SizeBox_701 {widgetClassPath=&SizeBox, parent=HorizontalBox_13, widget=SizeBox_701, slot {class=HorizontalBoxSlot}}
  Image_35 {
   widgetClassPath=&Image
   bIsVariable=true
   parent=SizeBox_701
   widget=Image_35
   properties {
    brush {
     bIsDynamicallyLoaded=false
     drawAs=Image
     imageType=NoImage
     mirroring=NoMirror
     resourceName=None
     resourceObject=/Game/NeverCook/TechArt/VehicleCosmeticsTools/VC_LOD_TOOL/Icons/UI_RL_Icon1
     tiling=NoTile
     imageSize=<64 64>
     margin=<ltrb 0 0 0 0>
     outlineSettings {
      bUseBrushTransparency=false
      roundingType=HalfHeightRadius
      width=0
      color {colorUseRule=UseColor_Specified, specifiedColor=<rgba 0 0 0 0>}
      cornerRadii=<0 0 0 0>
     }
     tintColor {colorUseRule=UseColor_Specified, specifiedColor=<rgba 1 1 1 1>}
     uVRegion {bIsValid=false, max=<0 0>, min=<0 0>}
    }
    colorAndOpacity=<rgba 0.665 0.665 0.665 1>
   }
   slot {class=SizeBoxSlot, properties {horizontalAlignment=HAlign_Left, verticalAlignment=VAlign_Center}}
  }
  SizeBox_754 {widgetClassPath=&SizeBox, parent=HorizontalBox_13, widget=SizeBox_754, slot {class=HorizontalBoxSlot}}
  TextBlock_310 {
   widgetClassPath=&TextBlock
   parent=SizeBox_754
   widget=TextBlock_310
   properties {
    text=Vehicle Cosmetics LOD Tool
    colorAndOpacity {colorUseRule=UseColor_Specified, specifiedColor=<rgba 0.599 0.599 0.599 1>}
    font {
     bForceMonospaced=false
     bMaterialIsStencil=false
     fontMaterial=null
     fontObject=null
     letterSpacing=0
     monospacedWidth=1
     size=18
     skewAmount=0
     typefaceFontName=Default
     outlineSettings {
      bApplyOutlineToDropShadows=false
      bMiteredCorners=false
      bSeparateFillAlpha=false
      outlineMaterial=null
      outlineSize=0
      outlineColor=<rgba 0 0 0 1>
     }
    }
   }
   slot {class=SizeBoxSlot, properties {padding=<ltrb 16 16 16 16>}}
  }
  Border_90 {
   widgetClassPath=&Border
   parent=VerticalBox_1
   widget=Border_90
   properties {brushColor=<rgba 1 1 1 0.25>}
   slot {class=VerticalBoxSlot, properties {padding=<ltrb 0 16 0 16>}}
  }
  ButtonsAndSettingsHorizontalBox {
   widgetClassPath=&HorizontalBox
   parent=VerticalBox_1
   widget=ButtonsAndSettingsHorizontalBox
   slot {class=VerticalBoxSlot, properties {padding=<ltrb 6 6 6 6>}}
  }
  Border_135 {
   widgetClassPath=&Border
   parent=ButtonsAndSettingsHorizontalBox
   widget=Border_135
   properties {
    background {
     bIsDynamicallyLoaded=false
     drawAs=Image
     imageType=NoImage
     mirroring=NoMirror
     resourceName=None
     resourceObject=null
     tiling=NoTile
     imageSize=<32 32>
     margin=<ltrb 0 0 0 0>
     outlineSettings {
      bUseBrushTransparency=false
      roundingType=HalfHeightRadius
      width=0
      color {colorUseRule=UseColor_Specified, specifiedColor=<rgba 0 0 0 0>}
      cornerRadii=<0 0 0 0>
     }
     tintColor {colorUseRule=UseColor_Specified, specifiedColor=<rgba 0 0 0 1>}
     uVRegion {bIsValid=false, max=<0 0>, min=<0 0>}
    }
    brushColor=<rgba 1 1 1 0.2>
    padding=<ltrb 3 3 3 3>
   }
   slot {class=HorizontalBoxSlot, properties {size {sizeRule=Fill, value=1}}}
  }
  VerticalBox_2 {
   widgetClassPath=&VerticalBox
   parent=Border_135
   widget=VerticalBox_2
   slot {class=BorderSlot, properties {padding=<ltrb 3 3 3 3>}}
  }
  ScaleBox_0 {widgetClassPath=&ScaleBox, parent=VerticalBox_2, widget=ScaleBox_0, slot {class=VerticalBoxSlot}}
  EditorUtilityButton_BakeAll {
   widgetClassPath=&EditorUtilityButton
   parent=ScaleBox_0
   widget=EditorUtilityButton_BakeAll
   properties {widgetStyle {
     clickedSlateSound {resourceObject=null}
     disabled {
      bIsDynamicallyLoaded=false
      drawAs=RoundedBox
      imageType=NoImage
      mirroring=NoMirror
      resourceName=None
      resourceObject=null
      tiling=NoTile
      imageSize=<0 0>
      margin=<ltrb 0 0 0 0>
      outlineSettings {
       bUseBrushTransparency=false
       roundingType=FixedRadius
       width=1
       color {colorUseRule=UseColor_ColorTable, specifiedColor=<rgba 0 0 0 0>}
       cornerRadii=<4 4 4 4>
      }
      tintColor {colorUseRule=UseColor_Specified, specifiedColor=<rgba 0.0395 0.0395 0.0395 1>}
      uVRegion {bIsValid=false, max=<0 0>, min=<0 0>}
     }
     disabledForeground {colorUseRule=UseColor_ColorTable, specifiedColor=<rgba 0 0 0 0>}
     hovered {
      bIsDynamicallyLoaded=false
      drawAs=RoundedBox
      imageType=NoImage
      mirroring=NoMirror
      resourceName=None
      resourceObject=null
      tiling=NoTile
      imageSize=<32 32>
      margin=<ltrb 0 0 0 0>
      outlineSettings {
       bUseBrushTransparency=false
       roundingType=FixedRadius
       width=1
       color {colorUseRule=UseColor_ColorTable, specifiedColor=<rgba 0 0 0 0>}
       cornerRadii=<4 4 4 4>
      }
      tintColor {colorUseRule=UseColor_Specified, specifiedColor=<rgba 0.9115 0.3654 0.1472 1>}
      uVRegion {bIsValid=false, max=<0 0>, min=<0 0>}
     }
     hoveredForeground {colorUseRule=UseColor_ColorTable, specifiedColor=<rgba 0 0 0 0>}
     hoveredSlateSound {resourceObject=null}
     normal {
      bIsDynamicallyLoaded=false
      drawAs=RoundedBox
      imageType=NoImage
      mirroring=NoMirror
      resourceName=None
      resourceObject=null
      tiling=NoTile
      imageSize=<32 32>
      margin=<ltrb 0 0 0 0>
      outlineSettings {
       bUseBrushTransparency=false
       roundingType=FixedRadius
       width=1
       color {colorUseRule=UseColor_ColorTable, specifiedColor=<rgba 0 0 0 0>}
       cornerRadii=<4 4 4 4>
      }
      tintColor {colorUseRule=UseColor_Specified, specifiedColor=<rgba 0.4167 0.119 0 1>}
      uVRegion {bIsValid=false, max=<0 0>, min=<0 0>}
     }
     normalForeground {colorUseRule=UseColor_ColorTable, specifiedColor=<rgba 0 0 0 0>}
     normalPadding=<ltrb 6 0 6 0>
     pressed {
      bIsDynamicallyLoaded=false
      drawAs=RoundedBox
      imageType=NoImage
      mirroring=NoMirror
      resourceName=None
      resourceObject=null
      tiling=NoTile
      imageSize=<32 32>
      margin=<ltrb 0 0 0 0>
      outlineSettings {
       bUseBrushTransparency=false
       roundingType=FixedRadius
       width=1
       color {colorUseRule=UseColor_ColorTable, specifiedColor=<rgba 0 0 0 0>}
       cornerRadii=<4 4 4 4>
      }
      tintColor {colorUseRule=UseColor_Specified, specifiedColor=<rgba 0.1302 0.0372 0 1>}
      uVRegion {bIsValid=false, max=<0 0>, min=<0 0>}
     }
     pressedForeground {colorUseRule=UseColor_ColorTable, specifiedColor=<rgba 0 0 0 0>}
     pressedPadding=<ltrb 6 0 6 0>
     pressedSlateSound {resourceObject=null}
    }}
   slot {class=ScaleBoxSlot, properties {horizontalAlignment=HAlign_Fill}}
  }
  HorizontalBox_125 {
   widgetClassPath=&HorizontalBox
   parent=EditorUtilityButton_BakeAll
   widget=HorizontalBox_125
   slot {class=ButtonSlot, properties {horizontalAlignment=HAlign_Fill, verticalAlignment=VAlign_Fill, padding=<ltrb 0 0 0 0>}}
  }
  SizeBox_49 {widgetClassPath=&SizeBox, parent=HorizontalBox_125, widget=SizeBox_49, slot {class=HorizontalBoxSlot}}
  Image_407 {
   widgetClassPath=&Image
   bIsVariable=true
   parent=SizeBox_49
   widget=Image_407
   properties {brush {
     bIsDynamicallyLoaded=false
     drawAs=Image
     imageType=NoImage
     mirroring=NoMirror
     resourceName=None
     resourceObject=/Game/NeverCook/TechArt/VehicleCosmeticsTools/VC_LOD_TOOL/Icons/Icon_Bread
     tiling=NoTile
     imageSize=<32 32>
     margin=<ltrb 0 0 0 0>
     outlineSettings {
      bUseBrushTransparency=false
      roundingType=HalfHeightRadius
      width=0
      color {colorUseRule=UseColor_Specified, specifiedColor=<rgba 0 0 0 0>}
      cornerRadii=<0 0 0 0>
     }
     tintColor {colorUseRule=UseColor_Specified, specifiedColor=<rgba 1 1 1 1>}
     uVRegion {bIsValid=false, max=<0 0>, min=<0 0>}
    }}
   slot {class=SizeBoxSlot, properties {horizontalAlignment=HAlign_Left, verticalAlignment=VAlign_Top, padding=<ltrb 4 4 4 4>}}
  }
  SizeBox_91 {widgetClassPath=&SizeBox, parent=HorizontalBox_125, widget=SizeBox_91, slot {class=HorizontalBoxSlot}}
  TextBlock_64 {
   widgetClassPath=&TextBlock
   parent=SizeBox_91
   widget=TextBlock_64
   properties {
    text=Bake All
    colorAndOpacity {colorUseRule=UseColor_Specified, specifiedColor=<rgba 0.6458 0.6458 0.6458 1>}
    font {
     bForceMonospaced=false
     bMaterialIsStencil=false
     fontMaterial=null
     fontObject=&Roboto
     letterSpacing=0
     monospacedWidth=1
     size=10.5
     skewAmount=0
     typefaceFontName=Bold
     outlineSettings {
      bApplyOutlineToDropShadows=false
      bMiteredCorners=false
      bSeparateFillAlpha=false
      outlineMaterial=null
      outlineSize=0
      outlineColor=<rgba 0 0 0 1>
     }
    }
   }
   slot {class=SizeBoxSlot, properties {horizontalAlignment=HAlign_Left, verticalAlignment=VAlign_Center, padding=<ltrb 4 2 4 2>}}
  }
  ScaleBox {widgetClassPath=&ScaleBox, parent=VerticalBox_2, widget=ScaleBox, slot {class=VerticalBoxSlot}}
  EditorUtilityButton_BakeCompositeMask {
   widgetClassPath=&EditorUtilityButton
   parent=ScaleBox
   widget=EditorUtilityButton_BakeCompositeMask
   properties {widgetStyle {
     clickedSlateSound {resourceObject=null}
     disabled {
      bIsDynamicallyLoaded=false
      drawAs=RoundedBox
      imageType=NoImage
      mirroring=NoMirror
      resourceName=None
      resourceObject=null
      tiling=NoTile
      imageSize=<0 0>
      margin=<ltrb 0 0 0 0>
      outlineSettings {
       bUseBrushTransparency=false
       roundingType=FixedRadius
       width=1
       color {colorUseRule=UseColor_ColorTable, specifiedColor=<rgba 0 0 0 0>}
       cornerRadii=<4 4 4 4>
      }
      tintColor {colorUseRule=UseColor_Specified, specifiedColor=<rgba 0.0395 0.0395 0.0395 1>}
      uVRegion {bIsValid=false, max=<0 0>, min=<0 0>}
     }
     disabledForeground {colorUseRule=UseColor_ColorTable, specifiedColor=<rgba 0 0 0 0>}
     hovered {
      bIsDynamicallyLoaded=false
      drawAs=RoundedBox
      imageType=NoImage
      mirroring=NoMirror
      resourceName=None
      resourceObject=null
      tiling=NoTile
      imageSize=<32 32>
      margin=<ltrb 0 0 0 0>
      outlineSettings {
       bUseBrushTransparency=false
       roundingType=FixedRadius
       width=1
       color {colorUseRule=UseColor_ColorTable, specifiedColor=<rgba 0 0 0 0>}
       cornerRadii=<4 4 4 4>
      }
      tintColor {colorUseRule=UseColor_Specified, specifiedColor=<rgba 0.0953 0.0953 0.0953 1>}
      uVRegion {bIsValid=false, max=<0 0>, min=<0 0>}
     }
     hoveredForeground {colorUseRule=UseColor_ColorTable, specifiedColor=<rgba 0 0 0 0>}
     hoveredSlateSound {resourceObject=null}
     normal {
      bIsDynamicallyLoaded=false
      drawAs=RoundedBox
      imageType=NoImage
      mirroring=NoMirror
      resourceName=None
      resourceObject=null
      tiling=NoTile
      imageSize=<32 32>
      margin=<ltrb 0 0 0 0>
      outlineSettings {
       bUseBrushTransparency=false
       roundingType=FixedRadius
       width=1
       color {colorUseRule=UseColor_ColorTable, specifiedColor=<rgba 0 0 0 0>}
       cornerRadii=<4 4 4 4>
      }
      tintColor {colorUseRule=UseColor_Specified, specifiedColor=<rgba 0.0395 0.0395 0.0395 1>}
      uVRegion {bIsValid=false, max=<0 0>, min=<0 0>}
     }
     normalForeground {colorUseRule=UseColor_ColorTable, specifiedColor=<rgba 0 0 0 0>}
     normalPadding=<ltrb 6 0 6 0>
     pressed {
      bIsDynamicallyLoaded=false
      drawAs=RoundedBox
      imageType=NoImage
      mirroring=NoMirror
      resourceName=None
      resourceObject=null
      tiling=NoTile
      imageSize=<32 32>
      margin=<ltrb 0 0 0 0>
      outlineSettings {
       bUseBrushTransparency=false
       roundingType=FixedRadius
       width=1
       color {colorUseRule=UseColor_ColorTable, specifiedColor=<rgba 0 0 0 0>}
       cornerRadii=<4 4 4 4>
      }
      tintColor {colorUseRule=UseColor_Specified, specifiedColor=<rgba 0.0284 0.0284 0.0284 1>}
      uVRegion {bIsValid=false, max=<0 0>, min=<0 0>}
     }
     pressedForeground {colorUseRule=UseColor_ColorTable, specifiedColor=<rgba 0 0 0 0>}
     pressedPadding=<ltrb 6 0 6 0>
     pressedSlateSound {resourceObject=null}
    }}
   slot {class=ScaleBoxSlot, properties {horizontalAlignment=HAlign_Fill}}
  }
  HorizontalBox_10 {
   widgetClassPath=&HorizontalBox
   parent=EditorUtilityButton_BakeCompositeMask
   widget=HorizontalBox_10
   slot {class=ButtonSlot, properties {horizontalAlignment=HAlign_Fill, verticalAlignment=VAlign_Fill, padding=<ltrb 0 0 0 0>}}
  }
  SizeBox_16 {widgetClassPath=&SizeBox, parent=HorizontalBox_10, widget=SizeBox_16, slot {class=HorizontalBoxSlot}}
  Image_6 {
   widgetClassPath=&Image
   bIsVariable=true
   parent=SizeBox_16
   widget=Image_6
   properties {brush {
     bIsDynamicallyLoaded=false
     drawAs=Image
     imageType=NoImage
     mirroring=NoMirror
     resourceName=None
     resourceObject=&Icon_Bread
     tiling=NoTile
     imageSize=<32 32>
     margin=<ltrb 0 0 0 0>
     outlineSettings {
      bUseBrushTransparency=false
      roundingType=HalfHeightRadius
      width=0
      color {colorUseRule=UseColor_Specified, specifiedColor=<rgba 0 0 0 0>}
      cornerRadii=<0 0 0 0>
     }
     tintColor {colorUseRule=UseColor_Specified, specifiedColor=<rgba 1 1 1 1>}
     uVRegion {bIsValid=false, max=<0 0>, min=<0 0>}
    }}
   slot {class=SizeBoxSlot, properties {horizontalAlignment=HAlign_Left, verticalAlignment=VAlign_Center, padding=<ltrb 4 4 4 4>}}
  }
  SizeBox_17 {
   widgetClassPath=&SizeBox
   parent=HorizontalBox_10
   widget=SizeBox_17
   slot {class=HorizontalBoxSlot, properties {size {sizeRule=Fill, value=1}}}
  }
  TextBlock_8 {
   widgetClassPath=&TextBlock
   parent=SizeBox_17
   widget=TextBlock_8
   properties {
    text="Bake Composite Mask                    "
    colorAndOpacity {colorUseRule=UseColor_Specified, specifiedColor=<rgba 0.6458 0.6458 0.6458 1>}
    font {
     bForceMonospaced=false
     bMaterialIsStencil=false
     fontMaterial=null
     fontObject=&Roboto
     letterSpacing=0
     monospacedWidth=1
     size=10.5
     skewAmount=0
     typefaceFontName=Regular
     outlineSettings {
      bApplyOutlineToDropShadows=false
      bMiteredCorners=false
      bSeparateFillAlpha=false
      outlineMaterial=null
      outlineSize=0
      outlineColor=<rgba 0 0 0 1>
     }
    }
   }
   slot {class=SizeBoxSlot, properties {horizontalAlignment=HAlign_Left, verticalAlignment=VAlign_Center, padding=<ltrb 4 2 4 2>}}
  }
  ScaleBox_1 {widgetClassPath=&ScaleBox, parent=VerticalBox_2, widget=ScaleBox_1, slot {class=VerticalBoxSlot}}
  EditorUtilityButton_BakeBodyNormal {
   widgetClassPath=&EditorUtilityButton
   parent=ScaleBox_1
   widget=EditorUtilityButton_BakeBodyNormal
   properties {widgetStyle {
     clickedSlateSound {resourceObject=null}
     disabled {
      bIsDynamicallyLoaded=false
      drawAs=RoundedBox
      imageType=NoImage
      mirroring=NoMirror
      resourceName=None
      resourceObject=null
      tiling=NoTile
      imageSize=<0 0>
      margin=<ltrb 0 0 0 0>
      outlineSettings {
       bUseBrushTransparency=false
       roundingType=FixedRadius
       width=1
       color {colorUseRule=UseColor_ColorTable, specifiedColor=<rgba 0 0 0 0>}
       cornerRadii=<4 4 4 4>
      }
      tintColor {colorUseRule=UseColor_Specified, specifiedColor=<rgba 0.0395 0.0395 0.0395 1>}
      uVRegion {bIsValid=false, max=<0 0>, min=<0 0>}
     }
     disabledForeground {colorUseRule=UseColor_ColorTable, specifiedColor=<rgba 0 0 0 0>}
     hovered {
      bIsDynamicallyLoaded=false
      drawAs=RoundedBox
      imageType=NoImage
      mirroring=NoMirror
      resourceName=None
      resourceObject=null
      tiling=NoTile
      imageSize=<32 32>
      margin=<ltrb 0 0 0 0>
      outlineSettings {
       bUseBrushTransparency=false
       roundingType=FixedRadius
       width=1
       color {colorUseRule=UseColor_ColorTable, specifiedColor=<rgba 0 0 0 0>}
       cornerRadii=<4 4 4 4>
      }
      tintColor {colorUseRule=UseColor_Specified, specifiedColor=<rgba 0.0953 0.0953 0.0953 1>}
      uVRegion {bIsValid=false, max=<0 0>, min=<0 0>}
     }
     hoveredForeground {colorUseRule=UseColor_ColorTable, specifiedColor=<rgba 0 0 0 0>}
     hoveredSlateSound {resourceObject=null}
     normal {
      bIsDynamicallyLoaded=false
      drawAs=RoundedBox
      imageType=NoImage
      mirroring=NoMirror
      resourceName=None
      resourceObject=null
      tiling=NoTile
      imageSize=<32 32>
      margin=<ltrb 0 0 0 0>
      outlineSettings {
       bUseBrushTransparency=false
       roundingType=FixedRadius
       width=1
       color {colorUseRule=UseColor_ColorTable, specifiedColor=<rgba 0 0 0 0>}
       cornerRadii=<4 4 4 4>
      }
      tintColor {colorUseRule=UseColor_Specified, specifiedColor=<rgba 0.0395 0.0395 0.0395 1>}
      uVRegion {bIsValid=false, max=<0 0>, min=<0 0>}
     }
     normalForeground {colorUseRule=UseColor_ColorTable, specifiedColor=<rgba 0 0 0 0>}
     normalPadding=<ltrb 6 0 6 0>
     pressed {
      bIsDynamicallyLoaded=false
      drawAs=RoundedBox
      imageType=NoImage
      mirroring=NoMirror
      resourceName=None
      resourceObject=null
      tiling=NoTile
      imageSize=<32 32>
      margin=<ltrb 0 0 0 0>
      outlineSettings {
       bUseBrushTransparency=false
       roundingType=FixedRadius
       width=1
       color {colorUseRule=UseColor_ColorTable, specifiedColor=<rgba 0 0 0 0>}
       cornerRadii=<4 4 4 4>
      }
      tintColor {colorUseRule=UseColor_Specified, specifiedColor=<rgba 0.0284 0.0284 0.0284 1>}
      uVRegion {bIsValid=false, max=<0 0>, min=<0 0>}
     }
     pressedForeground {colorUseRule=UseColor_ColorTable, specifiedColor=<rgba 0 0 0 0>}
     pressedPadding=<ltrb 6 0 6 0>
     pressedSlateSound {resourceObject=null}
    }}
   slot {class=ScaleBoxSlot, properties {horizontalAlignment=HAlign_Fill}}
  }
  HorizontalBox {
   widgetClassPath=&HorizontalBox
   parent=EditorUtilityButton_BakeBodyNormal
   widget=HorizontalBox
   slot {class=ButtonSlot, properties {horizontalAlignment=HAlign_Fill, verticalAlignment=VAlign_Fill, padding=<ltrb 0 0 0 0>}}
  }
  SizeBox {widgetClassPath=&SizeBox, parent=HorizontalBox, widget=SizeBox, slot {class=HorizontalBoxSlot}}
  Image {
   widgetClassPath=&Image
   bIsVariable=true
   parent=SizeBox
   widget=Image
   properties {brush {
     bIsDynamicallyLoaded=false
     drawAs=Image
     imageType=NoImage
     mirroring=NoMirror
     resourceName=None
     resourceObject=&Icon_Bread
     tiling=NoTile
     imageSize=<32 32>
     margin=<ltrb 0 0 0 0>
     outlineSettings {
      bUseBrushTransparency=false
      roundingType=HalfHeightRadius
      width=0
      color {colorUseRule=UseColor_Specified, specifiedColor=<rgba 0 0 0 0>}
      cornerRadii=<0 0 0 0>
     }
     tintColor {colorUseRule=UseColor_Specified, specifiedColor=<rgba 1 1 1 1>}
     uVRegion {bIsValid=false, max=<0 0>, min=<0 0>}
    }}
   slot {class=SizeBoxSlot, properties {horizontalAlignment=HAlign_Left, verticalAlignment=VAlign_Center, padding=<ltrb 4 4 4 4>}}
  }
  SizeBox_1 {
   widgetClassPath=&SizeBox
   parent=HorizontalBox
   widget=SizeBox_1
   slot {class=HorizontalBoxSlot, properties {size {sizeRule=Fill, value=1}}}
  }
  TextBlock {
   widgetClassPath=&TextBlock
   parent=SizeBox_1
   widget=TextBlock
   properties {
    text=Bake Body N
    colorAndOpacity {colorUseRule=UseColor_Specified, specifiedColor=<rgba 0.6458 0.6458 0.6458 1>}
    font {
     bForceMonospaced=false
     bMaterialIsStencil=false
     fontMaterial=null
     fontObject=&Roboto
     letterSpacing=0
     monospacedWidth=1
     size=10.5
     skewAmount=0
     typefaceFontName=Regular
     outlineSettings {
      bApplyOutlineToDropShadows=false
      bMiteredCorners=false
      bSeparateFillAlpha=false
      outlineMaterial=null
      outlineSize=0
      outlineColor=<rgba 0 0 0 1>
     }
    }
   }
   slot {class=SizeBoxSlot, properties {horizontalAlignment=HAlign_Left, verticalAlignment=VAlign_Center, padding=<ltrb 4 2 4 2>}}
  }
  ScaleBox_2 {widgetClassPath=&ScaleBox, parent=VerticalBox_2, widget=ScaleBox_2, slot {class=VerticalBoxSlot}}
  EditorUtilityButton_BakeChassisBC {
   widgetClassPath=&EditorUtilityButton
   parent=ScaleBox_2
   widget=EditorUtilityButton_BakeChassisBC
   properties {widgetStyle {
     clickedSlateSound {resourceObject=null}
     disabled {
      bIsDynamicallyLoaded=false
      drawAs=RoundedBox
      imageType=NoImage
      mirroring=NoMirror
      resourceName=None
      resourceObject=null
      tiling=NoTile
      imageSize=<0 0>
      margin=<ltrb 0 0 0 0>
      outlineSettings {
       bUseBrushTransparency=false
       roundingType=FixedRadius
       width=1
       color {colorUseRule=UseColor_ColorTable, specifiedColor=<rgba 0 0 0 0>}
       cornerRadii=<4 4 4 4>
      }
      tintColor {colorUseRule=UseColor_Specified, specifiedColor=<rgba 0.0395 0.0395 0.0395 1>}
      uVRegion {bIsValid=false, max=<0 0>, min=<0 0>}
     }
     disabledForeground {colorUseRule=UseColor_ColorTable, specifiedColor=<rgba 0 0 0 0>}
     hovered {
      bIsDynamicallyLoaded=false
      drawAs=RoundedBox
      imageType=NoImage
      mirroring=NoMirror
      resourceName=None
      resourceObject=null
      tiling=NoTile
      imageSize=<32 32>
      margin=<ltrb 0 0 0 0>
      outlineSettings {
       bUseBrushTransparency=false
       roundingType=FixedRadius
       width=1
       color {colorUseRule=UseColor_ColorTable, specifiedColor=<rgba 0 0 0 0>}
       cornerRadii=<4 4 4 4>
      }
      tintColor {colorUseRule=UseColor_Specified, specifiedColor=<rgba 0.0953 0.0953 0.0953 1>}
      uVRegion {bIsValid=false, max=<0 0>, min=<0 0>}
     }
     hoveredForeground {colorUseRule=UseColor_ColorTable, specifiedColor=<rgba 0 0 0 0>}
     hoveredSlateSound {resourceObject=null}
     normal {
      bIsDynamicallyLoaded=false
      drawAs=RoundedBox
      imageType=NoImage
      mirroring=NoMirror
      resourceName=None
      resourceObject=null
      tiling=NoTile
      imageSize=<32 32>
      margin=<ltrb 0 0 0 0>
      outlineSettings {
       bUseBrushTransparency=false
       roundingType=FixedRadius
       width=1
       color {colorUseRule=UseColor_ColorTable, specifiedColor=<rgba 0 0 0 0>}
       cornerRadii=<4 4 4 4>
      }
      tintColor {colorUseRule=UseColor_Specified, specifiedColor=<rgba 0.0395 0.0395 0.0395 1>}
      uVRegion {bIsValid=false, max=<0 0>, min=<0 0>}
     }
     normalForeground {colorUseRule=UseColor_ColorTable, specifiedColor=<rgba 0 0 0 0>}
     normalPadding=<ltrb 6 0 6 0>
     pressed {
      bIsDynamicallyLoaded=false
      drawAs=RoundedBox
      imageType=NoImage
      mirroring=NoMirror
      resourceName=None
      resourceObject=null
      tiling=NoTile
      imageSize=<32 32>
      margin=<ltrb 0 0 0 0>
      outlineSettings {
       bUseBrushTransparency=false
       roundingType=FixedRadius
       width=1
       color {colorUseRule=UseColor_ColorTable, specifiedColor=<rgba 0 0 0 0>}
       cornerRadii=<4 4 4 4>
      }
      tintColor {colorUseRule=UseColor_Specified, specifiedColor=<rgba 0.0284 0.0284 0.0284 1>}
      uVRegion {bIsValid=false, max=<0 0>, min=<0 0>}
     }
     pressedForeground {colorUseRule=UseColor_ColorTable, specifiedColor=<rgba 0 0 0 0>}
     pressedPadding=<ltrb 6 0 6 0>
     pressedSlateSound {resourceObject=null}
    }}
   slot {class=ScaleBoxSlot, properties {horizontalAlignment=HAlign_Fill}}
  }
  HorizontalBox_1 {
   widgetClassPath=&HorizontalBox
   parent=EditorUtilityButton_BakeChassisBC
   widget=HorizontalBox_1
   slot {class=ButtonSlot, properties {horizontalAlignment=HAlign_Fill, verticalAlignment=VAlign_Fill, padding=<ltrb 0 0 0 0>}}
  }
  SizeBox_2 {widgetClassPath=&SizeBox, parent=HorizontalBox_1, widget=SizeBox_2, slot {class=HorizontalBoxSlot}}
  Image_1 {
   widgetClassPath=&Image
   bIsVariable=true
   parent=SizeBox_2
   widget=Image_1
   properties {brush {
     bIsDynamicallyLoaded=false
     drawAs=Image
     imageType=NoImage
     mirroring=NoMirror
     resourceName=None
     resourceObject=&Icon_Bread
     tiling=NoTile
     imageSize=<32 32>
     margin=<ltrb 0 0 0 0>
     outlineSettings {
      bUseBrushTransparency=false
      roundingType=HalfHeightRadius
      width=0
      color {colorUseRule=UseColor_Specified, specifiedColor=<rgba 0 0 0 0>}
      cornerRadii=<0 0 0 0>
     }
     tintColor {colorUseRule=UseColor_Specified, specifiedColor=<rgba 1 1 1 1>}
     uVRegion {bIsValid=false, max=<0 0>, min=<0 0>}
    }}
   slot {class=SizeBoxSlot, properties {horizontalAlignment=HAlign_Left, verticalAlignment=VAlign_Center, padding=<ltrb 4 4 4 4>}}
  }
  SizeBox_3 {
   widgetClassPath=&SizeBox
   parent=HorizontalBox_1
   widget=SizeBox_3
   slot {class=HorizontalBoxSlot, properties {size {sizeRule=Fill, value=1}}}
  }
  TextBlock_1 {
   widgetClassPath=&TextBlock
   parent=SizeBox_3
   widget=TextBlock_1
   properties {
    text=Bake Chassis BC
    colorAndOpacity {colorUseRule=UseColor_Specified, specifiedColor=<rgba 0.6458 0.6458 0.6458 1>}
    font {
     bForceMonospaced=false
     bMaterialIsStencil=false
     fontMaterial=null
     fontObject=&Roboto
     letterSpacing=0
     monospacedWidth=1
     size=10.5
     skewAmount=0
     typefaceFontName=Regular
     outlineSettings {
      bApplyOutlineToDropShadows=false
      bMiteredCorners=false
      bSeparateFillAlpha=false
      outlineMaterial=null
      outlineSize=0
      outlineColor=<rgba 0 0 0 1>
     }
    }
   }
   slot {class=SizeBoxSlot, properties {horizontalAlignment=HAlign_Left, verticalAlignment=VAlign_Center, padding=<ltrb 4 2 4 2>}}
  }
  ScaleBox_3 {widgetClassPath=&ScaleBox, parent=VerticalBox_2, widget=ScaleBox_3, slot {class=VerticalBoxSlot}}
  EditorUtilityButton_BakeChassisNormal {
   widgetClassPath=&EditorUtilityButton
   parent=ScaleBox_3
   widget=EditorUtilityButton_BakeChassisNormal
   properties {widgetStyle {
     clickedSlateSound {resourceObject=null}
     disabled {
      bIsDynamicallyLoaded=false
      drawAs=RoundedBox
      imageType=NoImage
      mirroring=NoMirror
      resourceName=None
      resourceObject=null
      tiling=NoTile
      imageSize=<0 0>
      margin=<ltrb 0 0 0 0>
      outlineSettings {
       bUseBrushTransparency=false
       roundingType=FixedRadius
       width=1
       color {colorUseRule=UseColor_ColorTable, specifiedColor=<rgba 0 0 0 0>}
       cornerRadii=<4 4 4 4>
      }
      tintColor {colorUseRule=UseColor_Specified, specifiedColor=<rgba 0.0395 0.0395 0.0395 1>}
      uVRegion {bIsValid=false, max=<0 0>, min=<0 0>}
     }
     disabledForeground {colorUseRule=UseColor_ColorTable, specifiedColor=<rgba 0 0 0 0>}
     hovered {
      bIsDynamicallyLoaded=false
      drawAs=RoundedBox
      imageType=NoImage
      mirroring=NoMirror
      resourceName=None
      resourceObject=null
      tiling=NoTile
      imageSize=<32 32>
      margin=<ltrb 0 0 0 0>
      outlineSettings {
       bUseBrushTransparency=false
       roundingType=FixedRadius
       width=1
       color {colorUseRule=UseColor_ColorTable, specifiedColor=<rgba 0 0 0 0>}
       cornerRadii=<4 4 4 4>
      }
      tintColor {colorUseRule=UseColor_Specified, specifiedColor=<rgba 0.0953 0.0953 0.0953 1>}
      uVRegion {bIsValid=false, max=<0 0>, min=<0 0>}
     }
     hoveredForeground {colorUseRule=UseColor_ColorTable, specifiedColor=<rgba 0 0 0 0>}
     hoveredSlateSound {resourceObject=null}
     normal {
      bIsDynamicallyLoaded=false
      drawAs=RoundedBox
      imageType=NoImage
      mirroring=NoMirror
      resourceName=None
      resourceObject=null
      tiling=NoTile
      imageSize=<32 32>
      margin=<ltrb 0 0 0 0>
      outlineSettings {
       bUseBrushTransparency=false
       roundingType=FixedRadius
       width=1
       color {colorUseRule=UseColor_ColorTable, specifiedColor=<rgba 0 0 0 0>}
       cornerRadii=<4 4 4 4>
      }
      tintColor {colorUseRule=UseColor_Specified, specifiedColor=<rgba 0.0395 0.0395 0.0395 1>}
      uVRegion {bIsValid=false, max=<0 0>, min=<0 0>}
     }
     normalForeground {colorUseRule=UseColor_ColorTable, specifiedColor=<rgba 0 0 0 0>}
     normalPadding=<ltrb 6 0 6 0>
     pressed {
      bIsDynamicallyLoaded=false
      drawAs=RoundedBox
      imageType=NoImage
      mirroring=NoMirror
      resourceName=None
      resourceObject=null
      tiling=NoTile
      imageSize=<32 32>
      margin=<ltrb 0 0 0 0>
      outlineSettings {
       bUseBrushTransparency=false
       roundingType=FixedRadius
       width=1
       color {colorUseRule=UseColor_ColorTable, specifiedColor=<rgba 0 0 0 0>}
       cornerRadii=<4 4 4 4>
      }
      tintColor {colorUseRule=UseColor_Specified, specifiedColor=<rgba 0.0284 0.0284 0.0284 1>}
      uVRegion {bIsValid=false, max=<0 0>, min=<0 0>}
     }
     pressedForeground {colorUseRule=UseColor_ColorTable, specifiedColor=<rgba 0 0 0 0>}
     pressedPadding=<ltrb 6 0 6 0>
     pressedSlateSound {resourceObject=null}
    }}
   slot {class=ScaleBoxSlot, properties {horizontalAlignment=HAlign_Fill}}
  }
  HorizontalBox_2 {
   widgetClassPath=&HorizontalBox
   parent=EditorUtilityButton_BakeChassisNormal
   widget=HorizontalBox_2
   slot {class=ButtonSlot, properties {horizontalAlignment=HAlign_Fill, verticalAlignment=VAlign_Fill, padding=<ltrb 0 0 0 0>}}
  }
  SizeBox_4 {widgetClassPath=&SizeBox, parent=HorizontalBox_2, widget=SizeBox_4, slot {class=HorizontalBoxSlot}}
  Image_2 {
   widgetClassPath=&Image
   bIsVariable=true
   parent=SizeBox_4
   widget=Image_2
   properties {brush {
     bIsDynamicallyLoaded=false
     drawAs=Image
     imageType=NoImage
     mirroring=NoMirror
     resourceName=None
     resourceObject=&Icon_Bread
     tiling=NoTile
     imageSize=<32 32>
     margin=<ltrb 0 0 0 0>
     outlineSettings {
      bUseBrushTransparency=false
      roundingType=HalfHeightRadius
      width=0
      color {colorUseRule=UseColor_Specified, specifiedColor=<rgba 0 0 0 0>}
      cornerRadii=<0 0 0 0>
     }
     tintColor {colorUseRule=UseColor_Specified, specifiedColor=<rgba 1 1 1 1>}
     uVRegion {bIsValid=false, max=<0 0>, min=<0 0>}
    }}
   slot {class=SizeBoxSlot, properties {horizontalAlignment=HAlign_Left, verticalAlignment=VAlign_Center, padding=<ltrb 4 4 4 4>}}
  }
  SizeBox_5 {
   widgetClassPath=&SizeBox
   parent=HorizontalBox_2
   widget=SizeBox_5
   slot {class=HorizontalBoxSlot, properties {size {sizeRule=Fill, value=1}}}
  }
  TextBlock_2 {
   widgetClassPath=&TextBlock
   parent=SizeBox_5
   widget=TextBlock_2
   properties {
    text=Bake Chassis N
    colorAndOpacity {colorUseRule=UseColor_Specified, specifiedColor=<rgba 0.6458 0.6458 0.6458 1>}
    font {
     bForceMonospaced=false
     bMaterialIsStencil=false
     fontMaterial=null
     fontObject=&Roboto
     letterSpacing=0
     monospacedWidth=1
     size=10.5
     skewAmount=0
     typefaceFontName=Regular
     outlineSettings {
      bApplyOutlineToDropShadows=false
      bMiteredCorners=false
      bSeparateFillAlpha=false
      outlineMaterial=null
      outlineSize=0
      outlineColor=<rgba 0 0 0 1>
     }
    }
   }
   slot {class=SizeBoxSlot, properties {horizontalAlignment=HAlign_Left, verticalAlignment=VAlign_Center, padding=<ltrb 4 2 4 2>}}
  }
  ScaleBox_4 {widgetClassPath=&ScaleBox, parent=VerticalBox_2, widget=ScaleBox_4, slot {class=VerticalBoxSlot}}
  EditorUtilityButton_BakeChassis_E {
   widgetClassPath=&EditorUtilityButton
   parent=ScaleBox_4
   widget=EditorUtilityButton_BakeChassis_E
   properties {widgetStyle {
     clickedSlateSound {resourceObject=null}
     disabled {
      bIsDynamicallyLoaded=false
      drawAs=RoundedBox
      imageType=NoImage
      mirroring=NoMirror
      resourceName=None
      resourceObject=null
      tiling=NoTile
      imageSize=<0 0>
      margin=<ltrb 0 0 0 0>
      outlineSettings {
       bUseBrushTransparency=false
       roundingType=FixedRadius
       width=1
       color {colorUseRule=UseColor_ColorTable, specifiedColor=<rgba 0 0 0 0>}
       cornerRadii=<4 4 4 4>
      }
      tintColor {colorUseRule=UseColor_Specified, specifiedColor=<rgba 0.0395 0.0395 0.0395 1>}
      uVRegion {bIsValid=false, max=<0 0>, min=<0 0>}
     }
     disabledForeground {colorUseRule=UseColor_ColorTable, specifiedColor=<rgba 0 0 0 0>}
     hovered {
      bIsDynamicallyLoaded=false
      drawAs=RoundedBox
      imageType=NoImage
      mirroring=NoMirror
      resourceName=None
      resourceObject=null
      tiling=NoTile
      imageSize=<32 32>
      margin=<ltrb 0 0 0 0>
      outlineSettings {
       bUseBrushTransparency=false
       roundingType=FixedRadius
       width=1
       color {colorUseRule=UseColor_ColorTable, specifiedColor=<rgba 0 0 0 0>}
       cornerRadii=<4 4 4 4>
      }
      tintColor {colorUseRule=UseColor_Specified, specifiedColor=<rgba 0.0953 0.0953 0.0953 1>}
      uVRegion {bIsValid=false, max=<0 0>, min=<0 0>}
     }
     hoveredForeground {colorUseRule=UseColor_ColorTable, specifiedColor=<rgba 0 0 0 0>}
     hoveredSlateSound {resourceObject=null}
     normal {
      bIsDynamicallyLoaded=false
      drawAs=RoundedBox
      imageType=NoImage
      mirroring=NoMirror
      resourceName=None
      resourceObject=null
      tiling=NoTile
      imageSize=<32 32>
      margin=<ltrb 0 0 0 0>
      outlineSettings {
       bUseBrushTransparency=false
       roundingType=FixedRadius
       width=1
       color {colorUseRule=UseColor_ColorTable, specifiedColor=<rgba 0 0 0 0>}
       cornerRadii=<4 4 4 4>
      }
      tintColor {colorUseRule=UseColor_Specified, specifiedColor=<rgba 0.0395 0.0395 0.0395 1>}
      uVRegion {bIsValid=false, max=<0 0>, min=<0 0>}
     }
     normalForeground {colorUseRule=UseColor_ColorTable, specifiedColor=<rgba 0 0 0 0>}
     normalPadding=<ltrb 6 0 6 0>
     pressed {
      bIsDynamicallyLoaded=false
      drawAs=RoundedBox
      imageType=NoImage
      mirroring=NoMirror
      resourceName=None
      resourceObject=null
      tiling=NoTile
      imageSize=<32 32>
      margin=<ltrb 0 0 0 0>
      outlineSettings {
       bUseBrushTransparency=false
       roundingType=FixedRadius
       width=1
       color {colorUseRule=UseColor_ColorTable, specifiedColor=<rgba 0 0 0 0>}
       cornerRadii=<4 4 4 4>
      }
      tintColor {colorUseRule=UseColor_Specified, specifiedColor=<rgba 0.0284 0.0284 0.0284 1>}
      uVRegion {bIsValid=false, max=<0 0>, min=<0 0>}
     }
     pressedForeground {colorUseRule=UseColor_ColorTable, specifiedColor=<rgba 0 0 0 0>}
     pressedPadding=<ltrb 6 0 6 0>
     pressedSlateSound {resourceObject=null}
    }}
   slot {class=ScaleBoxSlot, properties {horizontalAlignment=HAlign_Fill}}
  }
  HorizontalBox_4 {
   widgetClassPath=&HorizontalBox
   parent=EditorUtilityButton_BakeChassis_E
   widget=HorizontalBox_4
   slot {class=ButtonSlot, properties {horizontalAlignment=HAlign_Fill, verticalAlignment=VAlign_Fill, padding=<ltrb 0 0 0 0>}}
  }
  SizeBox_8 {widgetClassPath=&SizeBox, parent=HorizontalBox_4, widget=SizeBox_8, slot {class=HorizontalBoxSlot}}
  Image_4 {
   widgetClassPath=&Image
   bIsVariable=true
   parent=SizeBox_8
   widget=Image_4
   properties {brush {
     bIsDynamicallyLoaded=false
     drawAs=Image
     imageType=NoImage
     mirroring=NoMirror
     resourceName=None
     resourceObject=&Icon_Bread
     tiling=NoTile
     imageSize=<32 32>
     margin=<ltrb 0 0 0 0>
     outlineSettings {
      bUseBrushTransparency=false
      roundingType=HalfHeightRadius
      width=0
      color {colorUseRule=UseColor_Specified, specifiedColor=<rgba 0 0 0 0>}
      cornerRadii=<0 0 0 0>
     }
     tintColor {colorUseRule=UseColor_Specified, specifiedColor=<rgba 1 1 1 1>}
     uVRegion {bIsValid=false, max=<0 0>, min=<0 0>}
    }}
   slot {class=SizeBoxSlot, properties {horizontalAlignment=HAlign_Left, verticalAlignment=VAlign_Center, padding=<ltrb 4 4 4 4>}}
  }
  SizeBox_9 {
   widgetClassPath=&SizeBox
   parent=HorizontalBox_4
   widget=SizeBox_9
   slot {class=HorizontalBoxSlot, properties {size {sizeRule=Fill, value=1}}}
  }
  TextBlock_4 {
   widgetClassPath=&TextBlock
   parent=SizeBox_9
   widget=TextBlock_4
   properties {
    text=Bake Chassis E
    colorAndOpacity {colorUseRule=UseColor_Specified, specifiedColor=<rgba 0.6458 0.6458 0.6458 1>}
    font {
     bForceMonospaced=false
     bMaterialIsStencil=false
     fontMaterial=null
     fontObject=&Roboto
     letterSpacing=0
     monospacedWidth=1
     size=10.5
     skewAmount=0
     typefaceFontName=Regular
     outlineSettings {
      bApplyOutlineToDropShadows=false
      bMiteredCorners=false
      bSeparateFillAlpha=false
      outlineMaterial=null
      outlineSize=0
      outlineColor=<rgba 0 0 0 1>
     }
    }
   }
   slot {class=SizeBoxSlot, properties {horizontalAlignment=HAlign_Left, verticalAlignment=VAlign_Center, padding=<ltrb 4 2 4 2>}}
  }
  ScaleBox_5 {widgetClassPath=&ScaleBox, parent=VerticalBox_2, widget=ScaleBox_5, slot {class=VerticalBoxSlot}}
  EditorUtilityButton_BakeChassis_M {
   widgetClassPath=&EditorUtilityButton
   parent=ScaleBox_5
   widget=EditorUtilityButton_BakeChassis_M
   properties {widgetStyle {
     clickedSlateSound {resourceObject=null}
     disabled {
      bIsDynamicallyLoaded=false
      drawAs=RoundedBox
      imageType=NoImage
      mirroring=NoMirror
      resourceName=None
      resourceObject=null
      tiling=NoTile
      imageSize=<0 0>
      margin=<ltrb 0 0 0 0>
      outlineSettings {
       bUseBrushTransparency=false
       roundingType=FixedRadius
       width=1
       color {colorUseRule=UseColor_ColorTable, specifiedColor=<rgba 0 0 0 0>}
       cornerRadii=<4 4 4 4>
      }
      tintColor {colorUseRule=UseColor_Specified, specifiedColor=<rgba 0.0395 0.0395 0.0395 1>}
      uVRegion {bIsValid=false, max=<0 0>, min=<0 0>}
     }
     disabledForeground {colorUseRule=UseColor_ColorTable, specifiedColor=<rgba 0 0 0 0>}
     hovered {
      bIsDynamicallyLoaded=false
      drawAs=RoundedBox
      imageType=NoImage
      mirroring=NoMirror
      resourceName=None
      resourceObject=null
      tiling=NoTile
      imageSize=<32 32>
      margin=<ltrb 0 0 0 0>
      outlineSettings {
       bUseBrushTransparency=false
       roundingType=FixedRadius
       width=1
       color {colorUseRule=UseColor_ColorTable, specifiedColor=<rgba 0 0 0 0>}
       cornerRadii=<4 4 4 4>
      }
      tintColor {colorUseRule=UseColor_Specified, specifiedColor=<rgba 0.0953 0.0953 0.0953 1>}
      uVRegion {bIsValid=false, max=<0 0>, min=<0 0>}
     }
     hoveredForeground {colorUseRule=UseColor_ColorTable, specifiedColor=<rgba 0 0 0 0>}
     hoveredSlateSound {resourceObject=null}
     normal {
      bIsDynamicallyLoaded=false
      drawAs=RoundedBox
      imageType=NoImage
      mirroring=NoMirror
      resourceName=None
      resourceObject=null
      tiling=NoTile
      imageSize=<32 32>
      margin=<ltrb 0 0 0 0>
      outlineSettings {
       bUseBrushTransparency=false
       roundingType=FixedRadius
       width=1
       color {colorUseRule=UseColor_ColorTable, specifiedColor=<rgba 0 0 0 0>}
       cornerRadii=<4 4 4 4>
      }
      tintColor {colorUseRule=UseColor_Specified, specifiedColor=<rgba 0.0395 0.0395 0.0395 1>}
      uVRegion {bIsValid=false, max=<0 0>, min=<0 0>}
     }
     normalForeground {colorUseRule=UseColor_ColorTable, specifiedColor=<rgba 0 0 0 0>}
     normalPadding=<ltrb 6 0 6 0>
     pressed {
      bIsDynamicallyLoaded=false
      drawAs=RoundedBox
      imageType=NoImage
      mirroring=NoMirror
      resourceName=None
      resourceObject=null
      tiling=NoTile
      imageSize=<32 32>
      margin=<ltrb 0 0 0 0>
      outlineSettings {
       bUseBrushTransparency=false
       roundingType=FixedRadius
       width=1
       color {colorUseRule=UseColor_ColorTable, specifiedColor=<rgba 0 0 0 0>}
       cornerRadii=<4 4 4 4>
      }
      tintColor {colorUseRule=UseColor_Specified, specifiedColor=<rgba 0.0284 0.0284 0.0284 1>}
      uVRegion {bIsValid=false, max=<0 0>, min=<0 0>}
     }
     pressedForeground {colorUseRule=UseColor_ColorTable, specifiedColor=<rgba 0 0 0 0>}
     pressedPadding=<ltrb 6 0 6 0>
     pressedSlateSound {resourceObject=null}
    }}
   slot {class=ScaleBoxSlot, properties {horizontalAlignment=HAlign_Fill}}
  }
  HorizontalBox_5 {
   widgetClassPath=&HorizontalBox
   parent=EditorUtilityButton_BakeChassis_M
   widget=HorizontalBox_5
   slot {class=ButtonSlot, properties {horizontalAlignment=HAlign_Fill, verticalAlignment=VAlign_Fill, padding=<ltrb 0 0 0 0>}}
  }
  SizeBox_10 {widgetClassPath=&SizeBox, parent=HorizontalBox_5, widget=SizeBox_10, slot {class=HorizontalBoxSlot}}
  Image_5 {
   widgetClassPath=&Image
   bIsVariable=true
   parent=SizeBox_10
   widget=Image_5
   properties {brush {
     bIsDynamicallyLoaded=false
     drawAs=Image
     imageType=NoImage
     mirroring=NoMirror
     resourceName=None
     resourceObject=&Icon_Bread
     tiling=NoTile
     imageSize=<32 32>
     margin=<ltrb 0 0 0 0>
     outlineSettings {
      bUseBrushTransparency=false
      roundingType=HalfHeightRadius
      width=0
      color {colorUseRule=UseColor_Specified, specifiedColor=<rgba 0 0 0 0>}
      cornerRadii=<0 0 0 0>
     }
     tintColor {colorUseRule=UseColor_Specified, specifiedColor=<rgba 1 1 1 1>}
     uVRegion {bIsValid=false, max=<0 0>, min=<0 0>}
    }}
   slot {class=SizeBoxSlot, properties {horizontalAlignment=HAlign_Left, verticalAlignment=VAlign_Center, padding=<ltrb 4 4 4 4>}}
  }
  SizeBox_11 {
   widgetClassPath=&SizeBox
   parent=HorizontalBox_5
   widget=SizeBox_11
   slot {class=HorizontalBoxSlot, properties {size {sizeRule=Fill, value=1}}}
  }
  TextBlock_5 {
   widgetClassPath=&TextBlock
   parent=SizeBox_11
   widget=TextBlock_5
   properties {
    text=Bake Chassis M
    colorAndOpacity {colorUseRule=UseColor_Specified, specifiedColor=<rgba 0.6458 0.6458 0.6458 1>}
    font {
     bForceMonospaced=false
     bMaterialIsStencil=false
     fontMaterial=null
     fontObject=&Roboto
     letterSpacing=0
     monospacedWidth=1
     size=10.5
     skewAmount=0
     typefaceFontName=Regular
     outlineSettings {
      bApplyOutlineToDropShadows=false
      bMiteredCorners=false
      bSeparateFillAlpha=false
      outlineMaterial=null
      outlineSize=0
      outlineColor=<rgba 0 0 0 1>
     }
    }
   }
   slot {class=SizeBoxSlot, properties {horizontalAlignment=HAlign_Left, verticalAlignment=VAlign_Center, padding=<ltrb 4 2 4 2>}}
  }
  ScaleBox_6 {widgetClassPath=&ScaleBox, parent=VerticalBox_2, widget=ScaleBox_6, slot {class=VerticalBoxSlot}}
  EditorUtilityButton_BakeChassisTextures {
   widgetClassPath=&EditorUtilityButton
   parent=ScaleBox_6
   widget=EditorUtilityButton_BakeChassisTextures
   properties {widgetStyle {
     clickedSlateSound {resourceObject=null}
     disabled {
      bIsDynamicallyLoaded=false
      drawAs=RoundedBox
      imageType=NoImage
      mirroring=NoMirror
      resourceName=None
      resourceObject=null
      tiling=NoTile
      imageSize=<0 0>
      margin=<ltrb 0 0 0 0>
      outlineSettings {
       bUseBrushTransparency=false
       roundingType=FixedRadius
       width=1
       color {colorUseRule=UseColor_ColorTable, specifiedColor=<rgba 0 0 0 0>}
       cornerRadii=<4 4 4 4>
      }
      tintColor {colorUseRule=UseColor_Specified, specifiedColor=<rgba 0.0395 0.0395 0.0395 1>}
      uVRegion {bIsValid=false, max=<0 0>, min=<0 0>}
     }
     disabledForeground {colorUseRule=UseColor_ColorTable, specifiedColor=<rgba 0 0 0 0>}
     hovered {
      bIsDynamicallyLoaded=false
      drawAs=RoundedBox
      imageType=NoImage
      mirroring=NoMirror
      resourceName=None
      resourceObject=null
      tiling=NoTile
      imageSize=<32 32>
      margin=<ltrb 0 0 0 0>
      outlineSettings {
       bUseBrushTransparency=false
       roundingType=FixedRadius
       width=1
       color {colorUseRule=UseColor_ColorTable, specifiedColor=<rgba 0 0 0 0>}
       cornerRadii=<4 4 4 4>
      }
      tintColor {colorUseRule=UseColor_Specified, specifiedColor=<rgba 0.0953 0.0953 0.0953 1>}
      uVRegion {bIsValid=false, max=<0 0>, min=<0 0>}
     }
     hoveredForeground {colorUseRule=UseColor_ColorTable, specifiedColor=<rgba 0 0 0 0>}
     hoveredSlateSound {resourceObject=null}
     normal {
      bIsDynamicallyLoaded=false
      drawAs=RoundedBox
      imageType=NoImage
      mirroring=NoMirror
      resourceName=None
      resourceObject=null
      tiling=NoTile
      imageSize=<32 32>
      margin=<ltrb 0 0 0 0>
      outlineSettings {
       bUseBrushTransparency=false
       roundingType=FixedRadius
       width=1
       color {colorUseRule=UseColor_ColorTable, specifiedColor=<rgba 0 0 0 0>}
       cornerRadii=<4 4 4 4>
      }
      tintColor {colorUseRule=UseColor_Specified, specifiedColor=<rgba 0.0395 0.0395 0.0395 1>}
      uVRegion {bIsValid=false, max=<0 0>, min=<0 0>}
     }
     normalForeground {colorUseRule=UseColor_ColorTable, specifiedColor=<rgba 0 0 0 0>}
     normalPadding=<ltrb 6 0 6 0>
     pressed {
      bIsDynamicallyLoaded=false
      drawAs=RoundedBox
      imageType=NoImage
      mirroring=NoMirror
      resourceName=None
      resourceObject=null
      tiling=NoTile
      imageSize=<32 32>
      margin=<ltrb 0 0 0 0>
      outlineSettings {
       bUseBrushTransparency=false
       roundingType=FixedRadius
       width=1
       color {colorUseRule=UseColor_ColorTable, specifiedColor=<rgba 0 0 0 0>}
       cornerRadii=<4 4 4 4>
      }
      tintColor {colorUseRule=UseColor_Specified, specifiedColor=<rgba 0.0284 0.0284 0.0284 1>}
      uVRegion {bIsValid=false, max=<0 0>, min=<0 0>}
     }
     pressedForeground {colorUseRule=UseColor_ColorTable, specifiedColor=<rgba 0 0 0 0>}
     pressedPadding=<ltrb 6 0 6 0>
     pressedSlateSound {resourceObject=null}
    }}
   slot {class=ScaleBoxSlot, properties {horizontalAlignment=HAlign_Fill}}
  }
  HorizontalBox_3 {
   widgetClassPath=&HorizontalBox
   parent=EditorUtilityButton_BakeChassisTextures
   widget=HorizontalBox_3
   slot {class=ButtonSlot, properties {horizontalAlignment=HAlign_Fill, verticalAlignment=VAlign_Fill, padding=<ltrb 0 0 0 0>}}
  }
  SizeBox_6 {widgetClassPath=&SizeBox, parent=HorizontalBox_3, widget=SizeBox_6, slot {class=HorizontalBoxSlot}}
  Image_3 {
   widgetClassPath=&Image
   bIsVariable=true
   parent=SizeBox_6
   widget=Image_3
   properties {brush {
     bIsDynamicallyLoaded=false
     drawAs=Image
     imageType=NoImage
     mirroring=NoMirror
     resourceName=None
     resourceObject=&Icon_Bread
     tiling=NoTile
     imageSize=<32 32>
     margin=<ltrb 0 0 0 0>
     outlineSettings {
      bUseBrushTransparency=false
      roundingType=HalfHeightRadius
      width=0
      color {colorUseRule=UseColor_Specified, specifiedColor=<rgba 0 0 0 0>}
      cornerRadii=<0 0 0 0>
     }
     tintColor {colorUseRule=UseColor_Specified, specifiedColor=<rgba 1 1 1 1>}
     uVRegion {bIsValid=false, max=<0 0>, min=<0 0>}
    }}
   slot {class=SizeBoxSlot, properties {horizontalAlignment=HAlign_Left, verticalAlignment=VAlign_Center, padding=<ltrb 4 4 4 4>}}
  }
  SizeBox_7 {
   widgetClassPath=&SizeBox
   parent=HorizontalBox_3
   widget=SizeBox_7
   slot {class=HorizontalBoxSlot, properties {size {sizeRule=Fill, value=1}}}
  }
  TextBlock_3 {
   widgetClassPath=&TextBlock
   parent=SizeBox_7
   widget=TextBlock_3
   properties {
    text=Bake ChassisTextures
    colorAndOpacity {colorUseRule=UseColor_Specified, specifiedColor=<rgba 0.6458 0.6458 0.6458 1>}
    font {
     bForceMonospaced=false
     bMaterialIsStencil=false
     fontMaterial=null
     fontObject=&Roboto
     letterSpacing=0
     monospacedWidth=1
     size=10.5
     skewAmount=0
     typefaceFontName=Regular
     outlineSettings {
      bApplyOutlineToDropShadows=false
      bMiteredCorners=false
      bSeparateFillAlpha=false
      outlineMaterial=null
      outlineSize=0
      outlineColor=<rgba 0 0 0 1>
     }
    }
   }
   slot {class=SizeBoxSlot, properties {horizontalAlignment=HAlign_Left, verticalAlignment=VAlign_Center, padding=<ltrb 4 2 4 2>}}
  }
  Border_136 {
   widgetClassPath=&Border
   parent=ButtonsAndSettingsHorizontalBox
   widget=Border_136
   properties {
    horizontalAlignment=HAlign_Left
    verticalAlignment=VAlign_Top
    background {
     bIsDynamicallyLoaded=false
     drawAs=Image
     imageType=NoImage
     mirroring=NoMirror
     resourceName=None
     resourceObject=null
     tiling=NoTile
     imageSize=<32 32>
     margin=<ltrb 0 0 0 0>
     outlineSettings {
      bUseBrushTransparency=false
      roundingType=HalfHeightRadius
      width=0
      color {colorUseRule=UseColor_Specified, specifiedColor=<rgba 0 0 0 0>}
      cornerRadii=<0 0 0 0>
     }
     tintColor {colorUseRule=UseColor_Specified, specifiedColor=<rgba 0 0 0 1>}
     uVRegion {bIsValid=false, max=<0 0>, min=<0 0>}
    }
    brushColor=<rgba 1 1 1 0.2>
    padding=<ltrb 3 3 3 3>
   }
   slot {class=HorizontalBoxSlot}
  }
  ScaleBox_75 {
   widgetClassPath=&ScaleBox
   parent=Border_136
   widget=ScaleBox_75
   slot {class=BorderSlot, properties {horizontalAlignment=HAlign_Left, verticalAlignment=VAlign_Top, padding=<ltrb 3 3 3 3>}}
  }
  VerticalBox_95 {widgetClassPath=&VerticalBox, parent=ScaleBox_75, widget=VerticalBox_95, slot {class=ScaleBoxSlot}}
  Border_213 {
   widgetClassPath=&Border
   parent=VerticalBox_95
   widget=Border_213
   properties {brushColor=<rgba 0.0284 0.0284 0.0284 1>, padding=<ltrb 16 6 4 6>}
   slot {class=VerticalBoxSlot}
  }
  TextBlock_74 {
   widgetClassPath=&TextBlock
   parent=Border_213
   widget=TextBlock_74
   properties {
    text=Baker Parameters
    colorAndOpacity {colorUseRule=UseColor_Specified, specifiedColor=<rgba 0.6719 0.6719 0.6719 1>}
    font {
     bForceMonospaced=false
     bMaterialIsStencil=false
     fontMaterial=null
     fontObject=&Roboto
     letterSpacing=0
     monospacedWidth=1
     size=9
     skewAmount=0
     typefaceFontName=Regular
     outlineSettings {
      bApplyOutlineToDropShadows=false
      bMiteredCorners=false
      bSeparateFillAlpha=false
      outlineMaterial=null
      outlineSize=0
      outlineColor=<rgba 0 0 0 1>
     }
    }
   }
   slot {class=BorderSlot, properties {padding=<ltrb 16 6 4 6>}}
  }
  UniformGridPanel_0 {
   widgetClassPath=/Script/UMG.UniformGridPanel
   parent=VerticalBox_95
   widget=UniformGridPanel_0
   properties {slotPadding=<ltrb 1 1 1 1>}
   slot {class=VerticalBoxSlot, properties {padding=<ltrb 3 3 0 0>}}
  }
  Border_215 {
   widgetClassPath=&Border
   parent=UniformGridPanel_0
   widget=Border_215
   properties {brushColor=<rgba 0.0176 0.0176 0.0176 1>}
   slot {class=UniformGridSlot, properties {horizontalAlignment=HAlign_Fill, verticalAlignment=VAlign_Fill}}
  }
  TextBlock_9 {
   widgetClassPath=&TextBlock
   parent=Border_215
   widget=TextBlock_9
   properties {
    text=Compute Normal
    toolTipText=Will check on recompute normal and compute weighted normal in the SM details
    colorAndOpacity {colorUseRule=UseColor_Specified, specifiedColor=<rgba 0.6719 0.6719 0.6719 1>}
    font {
     bForceMonospaced=false
     bMaterialIsStencil=false
     fontMaterial=null
     fontObject=&Roboto
     letterSpacing=0
     monospacedWidth=1
     size=9
     skewAmount=0
     typefaceFontName=Regular
     outlineSettings {
      bApplyOutlineToDropShadows=false
      bMiteredCorners=false
      bSeparateFillAlpha=false
      outlineMaterial=null
      outlineSize=0
      outlineColor=<rgba 0 0 0 1>
     }
    }
   }
   slot {class=BorderSlot}
  }
  Border_216 {
   widgetClassPath=&Border
   parent=UniformGridPanel_0
   widget=Border_216
   properties {brushColor=<rgba 0.0176 0.0176 0.0176 1>}
   slot {class=UniformGridSlot, properties {column=1, horizontalAlignment=HAlign_Fill, verticalAlignment=VAlign_Fill}}
  }
  EditorUtilityCheckBox_N_Compute {
   widgetClassPath=&EditorUtilityCheckBox
   bIsVariable=true
   parent=Border_216
   widget=EditorUtilityCheckBox_N_Compute
   properties {
    checkedState=Checked
    toolTipText=Will check on recompute normal and compute weighted normal in the SM details
    widgetStyle {
     checkBoxType=CheckBox
     backgroundHoveredImage {
      bIsDynamicallyLoaded=false
      drawAs=RoundedBox
      imageType=NoImage
      mirroring=NoMirror
      resourceName=None
      resourceObject=null
      tiling=NoTile
      imageSize=<18 18>
      margin=<ltrb 0 0 0 0>
      outlineSettings {
       bUseBrushTransparency=false
       roundingType=FixedRadius
       width=1
       color {colorUseRule=UseColor_ColorTable, specifiedColor=<rgba 0 0 0 0>}
       cornerRadii=<3 3 3 3>
      }
      tintColor {colorUseRule=UseColor_Specified, specifiedColor=<rgba 0.0048 0.0048 0.0048 1>}
      uVRegion {bIsValid=false, max=<0 0>, min=<0 0>}
     }
     backgroundImage {
      bIsDynamicallyLoaded=false
      drawAs=RoundedBox
      imageType=NoImage
      mirroring=NoMirror
      resourceName=None
      resourceObject=null
      tiling=NoTile
      imageSize=<18 18>
      margin=<ltrb 0 0 0 0>
      outlineSettings {
       bUseBrushTransparency=false
       roundingType=FixedRadius
       width=1
       color {colorUseRule=UseColor_ColorTable, specifiedColor=<rgba 0 0 0 0>}
       cornerRadii=<3 3 3 3>
      }
      tintColor {colorUseRule=UseColor_Specified, specifiedColor=<rgba 0.0048 0.0048 0.0048 1>}
      uVRegion {bIsValid=false, max=<0 0>, min=<0 0>}
     }
     backgroundPressedImage {
      bIsDynamicallyLoaded=false
      drawAs=RoundedBox
      imageType=NoImage
      mirroring=NoMirror
      resourceName=None
      resourceObject=null
      tiling=NoTile
      imageSize=<18 18>
      margin=<ltrb 0 0 0 0>
      outlineSettings {
       bUseBrushTransparency=false
       roundingType=FixedRadius
       width=1
       color {colorUseRule=UseColor_ColorTable, specifiedColor=<rgba 0 0 0 0>}
       cornerRadii=<3 3 3 3>
      }
      tintColor {colorUseRule=UseColor_Specified, specifiedColor=<rgba 0.0048 0.0048 0.0048 1>}
      uVRegion {bIsValid=false, max=<0 0>, min=<0 0>}
     }
     borderBackgroundColor {colorUseRule=UseColor_Specified, specifiedColor=<rgba 1 1 1 1>}
     checkedForeground {colorUseRule=UseColor_Specified, specifiedColor=<rgba 1 1 1 1>}
     checkedHoveredForeground {colorUseRule=UseColor_Specified, specifiedColor=<rgba 1 1 1 1>}
     checkedHoveredImage {
      bIsDynamicallyLoaded=false
      drawAs=Image
      imageType=Vector
      mirroring=NoMirror
      resourceName=../../../Engine/Content/Slate/Starship/CoreWidgets/CheckBox/check.svg
      resourceObject=null
      tiling=NoTile
      imageSize=<16 16>
      margin=<ltrb 0 0 0 0>
      outlineSettings {
       bUseBrushTransparency=false
       roundingType=HalfHeightRadius
       width=0
       color {colorUseRule=UseColor_Specified, specifiedColor=<rgba 0 0 0 0>}
       cornerRadii=<0 0 0 0>
      }
      tintColor {colorUseRule=UseColor_Specified, specifiedColor=<rgba 0.901 0.378 0.1689 1>}
      uVRegion {bIsValid=false, max=<0 0>, min=<0 0>}
     }
     checkedImage {
      bIsDynamicallyLoaded=false
      drawAs=Image
      imageType=Vector
      mirroring=NoMirror
      resourceName=../../../Engine/Content/Slate/Starship/CoreWidgets/CheckBox/check.svg
      resourceObject=null
      tiling=NoTile
      imageSize=<16 16>
      margin=<ltrb 0 0 0 0>
      outlineSettings {
       bUseBrushTransparency=false
       roundingType=HalfHeightRadius
       width=0
       color {colorUseRule=UseColor_Specified, specifiedColor=<rgba 0 0 0 0>}
       cornerRadii=<0 0 0 0>
      }
      tintColor {colorUseRule=UseColor_Specified, specifiedColor=<rgba 0.7454 0.2129 0 1>}
      uVRegion {bIsValid=false, max=<0 0>, min=<0 0>}
     }
     checkedPressedForeground {colorUseRule=UseColor_Specified, specifiedColor=<rgba 1 1 1 1>}
     checkedPressedImage {
      bIsDynamicallyLoaded=false
      drawAs=Image
      imageType=Vector
      mirroring=NoMirror
      resourceName=../../../Engine/Content/Slate/Starship/CoreWidgets/CheckBox/check.svg
      resourceObject=null
      tiling=NoTile
      imageSize=<16 16>
      margin=<ltrb 0 0 0 0>
      outlineSettings {
       bUseBrushTransparency=false
       roundingType=HalfHeightRadius
       width=0
       color {colorUseRule=UseColor_Specified, specifiedColor=<rgba 0 0 0 0>}
       cornerRadii=<0 0 0 0>
      }
      tintColor {colorUseRule=UseColor_Specified, specifiedColor=<rgba 0.4219 0.1205 0 1>}
      uVRegion {bIsValid=false, max=<0 0>, min=<0 0>}
     }
     checkedSlateSound {resourceObject=null}
     foregroundColor {colorUseRule=UseColor_Specified, specifiedColor=<rgba 1 1 1 1>}
     hoveredForeground {colorUseRule=UseColor_Specified, specifiedColor=<rgba 1 1 1 1>}
     hoveredSlateSound {resourceObject=null}
     padding=<ltrb 2 0 0 0>
     pressedForeground {colorUseRule=UseColor_Specified, specifiedColor=<rgba 1 1 1 1>}
     uncheckedHoveredImage {
      bIsDynamicallyLoaded=false
      drawAs=NoDrawType
      imageType=NoImage
      mirroring=NoMirror
      resourceName=None
      resourceObject=null
      tiling=NoTile
      imageSize=<0 0>
      margin=<ltrb 0 0 0 0>
      outlineSettings {
       bUseBrushTransparency=false
       roundingType=HalfHeightRadius
       width=0
       color {colorUseRule=UseColor_Specified, specifiedColor=<rgba 0 0 0 0>}
       cornerRadii=<0 0 0 0>
      }
      tintColor {colorUseRule=UseColor_Specified, specifiedColor=<rgba 1 1 1 1>}
      uVRegion {bIsValid=false, max=<0 0>, min=<0 0>}
     }
     uncheckedImage {
      bIsDynamicallyLoaded=false
      drawAs=NoDrawType
      imageType=NoImage
      mirroring=NoMirror
      resourceName=None
      resourceObject=null
      tiling=NoTile
      imageSize=<0 0>
      margin=<ltrb 0 0 0 0>
      outlineSettings {
       bUseBrushTransparency=false
       roundingType=HalfHeightRadius
       width=0
       color {colorUseRule=UseColor_Specified, specifiedColor=<rgba 0 0 0 0>}
       cornerRadii=<0 0 0 0>
      }
      tintColor {colorUseRule=UseColor_Specified, specifiedColor=<rgba 1 1 1 1>}
      uVRegion {bIsValid=false, max=<0 0>, min=<0 0>}
     }
     uncheckedPressedImage {
      bIsDynamicallyLoaded=false
      drawAs=NoDrawType
      imageType=NoImage
      mirroring=NoMirror
      resourceName=None
      resourceObject=null
      tiling=NoTile
      imageSize=<0 0>
      margin=<ltrb 0 0 0 0>
      outlineSettings {
       bUseBrushTransparency=false
       roundingType=HalfHeightRadius
       width=0
       color {colorUseRule=UseColor_Specified, specifiedColor=<rgba 0 0 0 0>}
       cornerRadii=<0 0 0 0>
      }
      tintColor {colorUseRule=UseColor_Specified, specifiedColor=<rgba 1 1 1 1>}
      uVRegion {bIsValid=false, max=<0 0>, min=<0 0>}
     }
     uncheckedSlateSound {resourceObject=null}
     undeterminedForeground {colorUseRule=UseColor_Specified, specifiedColor=<rgba 1 1 1 1>}
     undeterminedHoveredImage {
      bIsDynamicallyLoaded=false
      drawAs=Image
      imageType=Vector
      mirroring=NoMirror
      resourceName=../../../Engine/Content/Slate/Starship/CoreWidgets/CheckBox/indeterminate.svg
      resourceObject=null
      tiling=NoTile
      imageSize=<16 16>
      margin=<ltrb 0 0 0 0>
      outlineSettings {
       bUseBrushTransparency=false
       roundingType=HalfHeightRadius
       width=0
       color {colorUseRule=UseColor_Specified, specifiedColor=<rgba 0 0 0 0>}
       cornerRadii=<0 0 0 0>
      }
      tintColor {colorUseRule=UseColor_Specified, specifiedColor=<rgba 0.0044 0.2384 1 1>}
      uVRegion {bIsValid=false, max=<0 0>, min=<0 0>}
     }
     undeterminedImage {
      bIsDynamicallyLoaded=false
      drawAs=Image
      imageType=Vector
      mirroring=NoMirror
      resourceName=../../../Engine/Content/Slate/Starship/CoreWidgets/CheckBox/indeterminate.svg
      resourceObject=null
      tiling=NoTile
      imageSize=<16 16>
      margin=<ltrb 0 0 0 0>
      outlineSettings {
       bUseBrushTransparency=false
       roundingType=HalfHeightRadius
       width=0
       color {colorUseRule=UseColor_Specified, specifiedColor=<rgba 0 0 0 0>}
       cornerRadii=<0 0 0 0>
      }
      tintColor {colorUseRule=UseColor_Specified, specifiedColor=<rgba 0 0.162 0.7454 1>}
      uVRegion {bIsValid=false, max=<0 0>, min=<0 0>}
     }
     undeterminedPressedImage {
      bIsDynamicallyLoaded=false
      drawAs=Image
      imageType=Vector
      mirroring=NoMirror
      resourceName=../../../Engine/Content/Slate/Starship/CoreWidgets/CheckBox/indeterminate.svg
      resourceObject=null
      tiling=NoTile
      imageSize=<16 16>
      margin=<ltrb 0 0 0 0>
      outlineSettings {
       bUseBrushTransparency=false
       roundingType=HalfHeightRadius
       width=0
       color {colorUseRule=UseColor_Specified, specifiedColor=<rgba 0 0 0 0>}
       cornerRadii=<0 0 0 0>
      }
      tintColor {colorUseRule=UseColor_Specified, specifiedColor=<rgba 0 0.0802 0.3515 1>}
      uVRegion {bIsValid=false, max=<0 0>, min=<0 0>}
     }
    }
   }
   slot {class=BorderSlot}
  }
  Border_217 {
   widgetClassPath=&Border
   parent=UniformGridPanel_0
   widget=Border_217
   properties {brushColor=<rgba 0.0176 0.0176 0.0176 1>}
   slot {class=UniformGridSlot, properties {horizontalAlignment=HAlign_Fill, row=1, verticalAlignment=VAlign_Fill}}
  }
  TextBlock_83 {
   widgetClassPath=&TextBlock
   parent=Border_217
   widget=TextBlock_83
   properties {
    text="Composite Chassis BaseColor         "
    colorAndOpacity {colorUseRule=UseColor_Specified, specifiedColor=<rgba 0.6719 0.6719 0.6719 1>}
    font {
     bForceMonospaced=false
     bMaterialIsStencil=false
     fontMaterial=null
     fontObject=&Roboto
     letterSpacing=0
     monospacedWidth=1
     size=9
     skewAmount=0
     typefaceFontName=Regular
     outlineSettings {
      bApplyOutlineToDropShadows=false
      bMiteredCorners=false
      bSeparateFillAlpha=false
      outlineMaterial=null
      outlineSize=0
      outlineColor=<rgba 0 0 0 1>
     }
    }
   }
   slot {class=BorderSlot}
  }
  Border_218 {
   widgetClassPath=&Border
   parent=UniformGridPanel_0
   widget=Border_218
   properties {brushColor=<rgba 0.0176 0.0176 0.0176 1>}
   slot {class=UniformGridSlot, properties {column=1, horizontalAlignment=HAlign_Fill, row=1, verticalAlignment=VAlign_Fill}}
  }
  EditorUtilityCheckBox_BC_Composit {
   widgetClassPath=&EditorUtilityCheckBox
   bIsVariable=true
   parent=Border_218
   widget=EditorUtilityCheckBox_BC_Composit
   properties {
    checkedState=Checked
    widgetStyle {
     checkBoxType=CheckBox
     backgroundHoveredImage {
      bIsDynamicallyLoaded=false
      drawAs=RoundedBox
      imageType=NoImage
      mirroring=NoMirror
      resourceName=None
      resourceObject=null
      tiling=NoTile
      imageSize=<18 18>
      margin=<ltrb 0 0 0 0>
      outlineSettings {
       bUseBrushTransparency=false
       roundingType=FixedRadius
       width=1
       color {colorUseRule=UseColor_ColorTable, specifiedColor=<rgba 0 0 0 0>}
       cornerRadii=<3 3 3 3>
      }
      tintColor {colorUseRule=UseColor_Specified, specifiedColor=<rgba 0.0048 0.0048 0.0048 1>}
      uVRegion {bIsValid=false, max=<0 0>, min=<0 0>}
     }
     backgroundImage {
      bIsDynamicallyLoaded=false
      drawAs=RoundedBox
      imageType=NoImage
      mirroring=NoMirror
      resourceName=None
      resourceObject=null
      tiling=NoTile
      imageSize=<18 18>
      margin=<ltrb 0 0 0 0>
      outlineSettings {
       bUseBrushTransparency=false
       roundingType=FixedRadius
       width=1
       color {colorUseRule=UseColor_ColorTable, specifiedColor=<rgba 0 0 0 0>}
       cornerRadii=<3 3 3 3>
      }
      tintColor {colorUseRule=UseColor_Specified, specifiedColor=<rgba 0.0048 0.0048 0.0048 1>}
      uVRegion {bIsValid=false, max=<0 0>, min=<0 0>}
     }
     backgroundPressedImage {
      bIsDynamicallyLoaded=false
      drawAs=RoundedBox
      imageType=NoImage
      mirroring=NoMirror
      resourceName=None
      resourceObject=null
      tiling=NoTile
      imageSize=<18 18>
      margin=<ltrb 0 0 0 0>
      outlineSettings {
       bUseBrushTransparency=false
       roundingType=FixedRadius
       width=1
       color {colorUseRule=UseColor_ColorTable, specifiedColor=<rgba 0 0 0 0>}
       cornerRadii=<3 3 3 3>
      }
      tintColor {colorUseRule=UseColor_Specified, specifiedColor=<rgba 0.0048 0.0048 0.0048 1>}
      uVRegion {bIsValid=false, max=<0 0>, min=<0 0>}
     }
     borderBackgroundColor {colorUseRule=UseColor_Specified, specifiedColor=<rgba 1 1 1 1>}
     checkedForeground {colorUseRule=UseColor_Specified, specifiedColor=<rgba 1 1 1 1>}
     checkedHoveredForeground {colorUseRule=UseColor_Specified, specifiedColor=<rgba 1 1 1 1>}
     checkedHoveredImage {
      bIsDynamicallyLoaded=false
      drawAs=Image
      imageType=Vector
      mirroring=NoMirror
      resourceName=../../../Engine/Content/Slate/Starship/CoreWidgets/CheckBox/check.svg
      resourceObject=null
      tiling=NoTile
      imageSize=<16 16>
      margin=<ltrb 0 0 0 0>
      outlineSettings {
       bUseBrushTransparency=false
       roundingType=HalfHeightRadius
       width=0
       color {colorUseRule=UseColor_Specified, specifiedColor=<rgba 0 0 0 0>}
       cornerRadii=<0 0 0 0>
      }
      tintColor {colorUseRule=UseColor_Specified, specifiedColor=<rgba 0.901 0.378 0.1689 1>}
      uVRegion {bIsValid=false, max=<0 0>, min=<0 0>}
     }
     checkedImage {
      bIsDynamicallyLoaded=false
      drawAs=Image
      imageType=Vector
      mirroring=NoMirror
      resourceName=../../../Engine/Content/Slate/Starship/CoreWidgets/CheckBox/check.svg
      resourceObject=null
      tiling=NoTile
      imageSize=<16 16>
      margin=<ltrb 0 0 0 0>
      outlineSettings {
       bUseBrushTransparency=false
       roundingType=HalfHeightRadius
       width=0
       color {colorUseRule=UseColor_Specified, specifiedColor=<rgba 0 0 0 0>}
       cornerRadii=<0 0 0 0>
      }
      tintColor {colorUseRule=UseColor_Specified, specifiedColor=<rgba 0.7454 0.2129 0 1>}
      uVRegion {bIsValid=false, max=<0 0>, min=<0 0>}
     }
     checkedPressedForeground {colorUseRule=UseColor_Specified, specifiedColor=<rgba 1 1 1 1>}
     checkedPressedImage {
      bIsDynamicallyLoaded=false
      drawAs=Image
      imageType=Vector
      mirroring=NoMirror
      resourceName=../../../Engine/Content/Slate/Starship/CoreWidgets/CheckBox/check.svg
      resourceObject=null
      tiling=NoTile
      imageSize=<16 16>
      margin=<ltrb 0 0 0 0>
      outlineSettings {
       bUseBrushTransparency=false
       roundingType=HalfHeightRadius
       width=0
       color {colorUseRule=UseColor_Specified, specifiedColor=<rgba 0 0 0 0>}
       cornerRadii=<0 0 0 0>
      }
      tintColor {colorUseRule=UseColor_Specified, specifiedColor=<rgba 0.4219 0.1205 0 1>}
      uVRegion {bIsValid=false, max=<0 0>, min=<0 0>}
     }
     checkedSlateSound {resourceObject=null}
     foregroundColor {colorUseRule=UseColor_Specified, specifiedColor=<rgba 1 1 1 1>}
     hoveredForeground {colorUseRule=UseColor_Specified, specifiedColor=<rgba 1 1 1 1>}
     hoveredSlateSound {resourceObject=null}
     padding=<ltrb 2 0 0 0>
     pressedForeground {colorUseRule=UseColor_Specified, specifiedColor=<rgba 1 1 1 1>}
     uncheckedHoveredImage {
      bIsDynamicallyLoaded=false
      drawAs=NoDrawType
      imageType=NoImage
      mirroring=NoMirror
      resourceName=None
      resourceObject=null
      tiling=NoTile
      imageSize=<0 0>
      margin=<ltrb 0 0 0 0>
      outlineSettings {
       bUseBrushTransparency=false
       roundingType=HalfHeightRadius
       width=0
       color {colorUseRule=UseColor_Specified, specifiedColor=<rgba 0 0 0 0>}
       cornerRadii=<0 0 0 0>
      }
      tintColor {colorUseRule=UseColor_Specified, specifiedColor=<rgba 1 1 1 1>}
      uVRegion {bIsValid=false, max=<0 0>, min=<0 0>}
     }
     uncheckedImage {
      bIsDynamicallyLoaded=false
      drawAs=NoDrawType
      imageType=NoImage
      mirroring=NoMirror
      resourceName=None
      resourceObject=null
      tiling=NoTile
      imageSize=<0 0>
      margin=<ltrb 0 0 0 0>
      outlineSettings {
       bUseBrushTransparency=false
       roundingType=HalfHeightRadius
       width=0
       color {colorUseRule=UseColor_Specified, specifiedColor=<rgba 0 0 0 0>}
       cornerRadii=<0 0 0 0>
      }
      tintColor {colorUseRule=UseColor_Specified, specifiedColor=<rgba 1 1 1 1>}
      uVRegion {bIsValid=false, max=<0 0>, min=<0 0>}
     }
     uncheckedPressedImage {
      bIsDynamicallyLoaded=false
      drawAs=NoDrawType
      imageType=NoImage
      mirroring=NoMirror
      resourceName=None
      resourceObject=null
      tiling=NoTile
      imageSize=<0 0>
      margin=<ltrb 0 0 0 0>
      outlineSettings {
       bUseBrushTransparency=false
       roundingType=HalfHeightRadius
       width=0
       color {colorUseRule=UseColor_Specified, specifiedColor=<rgba 0 0 0 0>}
       cornerRadii=<0 0 0 0>
      }
      tintColor {colorUseRule=UseColor_Specified, specifiedColor=<rgba 1 1 1 1>}
      uVRegion {bIsValid=false, max=<0 0>, min=<0 0>}
     }
     uncheckedSlateSound {resourceObject=null}
     undeterminedForeground {colorUseRule=UseColor_Specified, specifiedColor=<rgba 1 1 1 1>}
     undeterminedHoveredImage {
      bIsDynamicallyLoaded=false
      drawAs=Image
      imageType=Vector
      mirroring=NoMirror
      resourceName=../../../Engine/Content/Slate/Starship/CoreWidgets/CheckBox/indeterminate.svg
      resourceObject=null
      tiling=NoTile
      imageSize=<16 16>
      margin=<ltrb 0 0 0 0>
      outlineSettings {
       bUseBrushTransparency=false
       roundingType=HalfHeightRadius
       width=0
       color {colorUseRule=UseColor_Specified, specifiedColor=<rgba 0 0 0 0>}
       cornerRadii=<0 0 0 0>
      }
      tintColor {colorUseRule=UseColor_Specified, specifiedColor=<rgba 0.0044 0.2384 1 1>}
      uVRegion {bIsValid=false, max=<0 0>, min=<0 0>}
     }
     undeterminedImage {
      bIsDynamicallyLoaded=false
      drawAs=Image
      imageType=Vector
      mirroring=NoMirror
      resourceName=../../../Engine/Content/Slate/Starship/CoreWidgets/CheckBox/indeterminate.svg
      resourceObject=null
      tiling=NoTile
      imageSize=<16 16>
      margin=<ltrb 0 0 0 0>
      outlineSettings {
       bUseBrushTransparency=false
       roundingType=HalfHeightRadius
       width=0
       color {colorUseRule=UseColor_Specified, specifiedColor=<rgba 0 0 0 0>}
       cornerRadii=<0 0 0 0>
      }
      tintColor {colorUseRule=UseColor_Specified, specifiedColor=<rgba 0 0.162 0.7454 1>}
      uVRegion {bIsValid=false, max=<0 0>, min=<0 0>}
     }
     undeterminedPressedImage {
      bIsDynamicallyLoaded=false
      drawAs=Image
      imageType=Vector
      mirroring=NoMirror
      resourceName=../../../Engine/Content/Slate/Starship/CoreWidgets/CheckBox/indeterminate.svg
      resourceObject=null
      tiling=NoTile
      imageSize=<16 16>
      margin=<ltrb 0 0 0 0>
      outlineSettings {
       bUseBrushTransparency=false
       roundingType=HalfHeightRadius
       width=0
       color {colorUseRule=UseColor_Specified, specifiedColor=<rgba 0 0 0 0>}
       cornerRadii=<0 0 0 0>
      }
      tintColor {colorUseRule=UseColor_Specified, specifiedColor=<rgba 0 0.0802 0.3515 1>}
      uVRegion {bIsValid=false, max=<0 0>, min=<0 0>}
     }
    }
   }
   slot {class=BorderSlot}
  }
  Border_219 {
   widgetClassPath=&Border
   parent=UniformGridPanel_0
   widget=Border_219
   properties {brushColor=<rgba 0.0176 0.0176 0.0176 1>}
   slot {class=UniformGridSlot, properties {horizontalAlignment=HAlign_Fill, row=2, verticalAlignment=VAlign_Fill}}
  }
  TextBlock_6 {
   widgetClassPath=&TextBlock
   parent=Border_219
   widget=TextBlock_6
   properties {
    text=Composite Chassis Normal
    colorAndOpacity {colorUseRule=UseColor_Specified, specifiedColor=<rgba 0.6719 0.6719 0.6719 1>}
    font {
     bForceMonospaced=false
     bMaterialIsStencil=false
     fontMaterial=null
     fontObject=&Roboto
     letterSpacing=0
     monospacedWidth=1
     size=9
     skewAmount=0
     typefaceFontName=Regular
     outlineSettings {
      bApplyOutlineToDropShadows=false
      bMiteredCorners=false
      bSeparateFillAlpha=false
      outlineMaterial=null
      outlineSize=0
      outlineColor=<rgba 0 0 0 1>
     }
    }
   }
   slot {class=BorderSlot}
  }
  Border_220 {
   widgetClassPath=&Border
   parent=UniformGridPanel_0
   widget=Border_220
   properties {brushColor=<rgba 0.0176 0.0176 0.0176 1>}
   slot {class=UniformGridSlot, properties {column=1, horizontalAlignment=HAlign_Fill, row=2, verticalAlignment=VAlign_Fill}}
  }
  EditorUtilityCheckBox_N_Composit {
   widgetClassPath=&EditorUtilityCheckBox
   bIsVariable=true
   parent=Border_220
   widget=EditorUtilityCheckBox_N_Composit
   properties {
    checkedState=Checked
    widgetStyle {
     checkBoxType=CheckBox
     backgroundHoveredImage {
      bIsDynamicallyLoaded=false
      drawAs=RoundedBox
      imageType=NoImage
      mirroring=NoMirror
      resourceName=None
      resourceObject=null
      tiling=NoTile
      imageSize=<18 18>
      margin=<ltrb 0 0 0 0>
      outlineSettings {
       bUseBrushTransparency=false
       roundingType=FixedRadius
       width=1
       color {colorUseRule=UseColor_ColorTable, specifiedColor=<rgba 0 0 0 0>}
       cornerRadii=<3 3 3 3>
      }
      tintColor {colorUseRule=UseColor_Specified, specifiedColor=<rgba 0.0048 0.0048 0.0048 1>}
      uVRegion {bIsValid=false, max=<0 0>, min=<0 0>}
     }
     backgroundImage {
      bIsDynamicallyLoaded=false
      drawAs=RoundedBox
      imageType=NoImage
      mirroring=NoMirror
      resourceName=None
      resourceObject=null
      tiling=NoTile
      imageSize=<18 18>
      margin=<ltrb 0 0 0 0>
      outlineSettings {
       bUseBrushTransparency=false
       roundingType=FixedRadius
       width=1
       color {colorUseRule=UseColor_ColorTable, specifiedColor=<rgba 0 0 0 0>}
       cornerRadii=<3 3 3 3>
      }
      tintColor {colorUseRule=UseColor_Specified, specifiedColor=<rgba 0.0048 0.0048 0.0048 1>}
      uVRegion {bIsValid=false, max=<0 0>, min=<0 0>}
     }
     backgroundPressedImage {
      bIsDynamicallyLoaded=false
      drawAs=RoundedBox
      imageType=NoImage
      mirroring=NoMirror
      resourceName=None
      resourceObject=null
      tiling=NoTile
      imageSize=<18 18>
      margin=<ltrb 0 0 0 0>
      outlineSettings {
       bUseBrushTransparency=false
       roundingType=FixedRadius
       width=1
       color {colorUseRule=UseColor_ColorTable, specifiedColor=<rgba 0 0 0 0>}
       cornerRadii=<3 3 3 3>
      }
      tintColor {colorUseRule=UseColor_Specified, specifiedColor=<rgba 0.0048 0.0048 0.0048 1>}
      uVRegion {bIsValid=false, max=<0 0>, min=<0 0>}
     }
     borderBackgroundColor {colorUseRule=UseColor_Specified, specifiedColor=<rgba 1 1 1 1>}
     checkedForeground {colorUseRule=UseColor_Specified, specifiedColor=<rgba 1 1 1 1>}
     checkedHoveredForeground {colorUseRule=UseColor_Specified, specifiedColor=<rgba 1 1 1 1>}
     checkedHoveredImage {
      bIsDynamicallyLoaded=false
      drawAs=Image
      imageType=Vector
      mirroring=NoMirror
      resourceName=../../../Engine/Content/Slate/Starship/CoreWidgets/CheckBox/check.svg
      resourceObject=null
      tiling=NoTile
      imageSize=<16 16>
      margin=<ltrb 0 0 0 0>
      outlineSettings {
       bUseBrushTransparency=false
       roundingType=HalfHeightRadius
       width=0
       color {colorUseRule=UseColor_Specified, specifiedColor=<rgba 0 0 0 0>}
       cornerRadii=<0 0 0 0>
      }
      tintColor {colorUseRule=UseColor_Specified, specifiedColor=<rgba 0.901 0.378 0.1689 1>}
      uVRegion {bIsValid=false, max=<0 0>, min=<0 0>}
     }
     checkedImage {
      bIsDynamicallyLoaded=false
      drawAs=Image
      imageType=Vector
      mirroring=NoMirror
      resourceName=../../../Engine/Content/Slate/Starship/CoreWidgets/CheckBox/check.svg
      resourceObject=null
      tiling=NoTile
      imageSize=<16 16>
      margin=<ltrb 0 0 0 0>
      outlineSettings {
       bUseBrushTransparency=false
       roundingType=HalfHeightRadius
       width=0
       color {colorUseRule=UseColor_Specified, specifiedColor=<rgba 0 0 0 0>}
       cornerRadii=<0 0 0 0>
      }
      tintColor {colorUseRule=UseColor_Specified, specifiedColor=<rgba 0.7454 0.2129 0 1>}
      uVRegion {bIsValid=false, max=<0 0>, min=<0 0>}
     }
     checkedPressedForeground {colorUseRule=UseColor_Specified, specifiedColor=<rgba 1 1 1 1>}
     checkedPressedImage {
      bIsDynamicallyLoaded=false
      drawAs=Image
      imageType=Vector
      mirroring=NoMirror
      resourceName=../../../Engine/Content/Slate/Starship/CoreWidgets/CheckBox/check.svg
      resourceObject=null
      tiling=NoTile
      imageSize=<16 16>
      margin=<ltrb 0 0 0 0>
      outlineSettings {
       bUseBrushTransparency=false
       roundingType=HalfHeightRadius
       width=0
       color {colorUseRule=UseColor_Specified, specifiedColor=<rgba 0 0 0 0>}
       cornerRadii=<0 0 0 0>
      }
      tintColor {colorUseRule=UseColor_Specified, specifiedColor=<rgba 0.4219 0.1205 0 1>}
      uVRegion {bIsValid=false, max=<0 0>, min=<0 0>}
     }
     checkedSlateSound {resourceObject=null}
     foregroundColor {colorUseRule=UseColor_Specified, specifiedColor=<rgba 1 1 1 1>}
     hoveredForeground {colorUseRule=UseColor_Specified, specifiedColor=<rgba 1 1 1 1>}
     hoveredSlateSound {resourceObject=null}
     padding=<ltrb 2 0 0 0>
     pressedForeground {colorUseRule=UseColor_Specified, specifiedColor=<rgba 1 1 1 1>}
     uncheckedHoveredImage {
      bIsDynamicallyLoaded=false
      drawAs=NoDrawType
      imageType=NoImage
      mirroring=NoMirror
      resourceName=None
      resourceObject=null
      tiling=NoTile
      imageSize=<0 0>
      margin=<ltrb 0 0 0 0>
      outlineSettings {
       bUseBrushTransparency=false
       roundingType=HalfHeightRadius
       width=0
       color {colorUseRule=UseColor_Specified, specifiedColor=<rgba 0 0 0 0>}
       cornerRadii=<0 0 0 0>
      }
      tintColor {colorUseRule=UseColor_Specified, specifiedColor=<rgba 1 1 1 1>}
      uVRegion {bIsValid=false, max=<0 0>, min=<0 0>}
     }
     uncheckedImage {
      bIsDynamicallyLoaded=false
      drawAs=NoDrawType
      imageType=NoImage
      mirroring=NoMirror
      resourceName=None
      resourceObject=null
      tiling=NoTile
      imageSize=<0 0>
      margin=<ltrb 0 0 0 0>
      outlineSettings {
       bUseBrushTransparency=false
       roundingType=HalfHeightRadius
       width=0
       color {colorUseRule=UseColor_Specified, specifiedColor=<rgba 0 0 0 0>}
       cornerRadii=<0 0 0 0>
      }
      tintColor {colorUseRule=UseColor_Specified, specifiedColor=<rgba 1 1 1 1>}
      uVRegion {bIsValid=false, max=<0 0>, min=<0 0>}
     }
     uncheckedPressedImage {
      bIsDynamicallyLoaded=false
      drawAs=NoDrawType
      imageType=NoImage
      mirroring=NoMirror
      resourceName=None
      resourceObject=null
      tiling=NoTile
      imageSize=<0 0>
      margin=<ltrb 0 0 0 0>
      outlineSettings {
       bUseBrushTransparency=false
       roundingType=HalfHeightRadius
       width=0
       color {colorUseRule=UseColor_Specified, specifiedColor=<rgba 0 0 0 0>}
       cornerRadii=<0 0 0 0>
      }
      tintColor {colorUseRule=UseColor_Specified, specifiedColor=<rgba 1 1 1 1>}
      uVRegion {bIsValid=false, max=<0 0>, min=<0 0>}
     }
     uncheckedSlateSound {resourceObject=null}
     undeterminedForeground {colorUseRule=UseColor_Specified, specifiedColor=<rgba 1 1 1 1>}
     undeterminedHoveredImage {
      bIsDynamicallyLoaded=false
      drawAs=Image
      imageType=Vector
      mirroring=NoMirror
      resourceName=../../../Engine/Content/Slate/Starship/CoreWidgets/CheckBox/indeterminate.svg
      resourceObject=null
      tiling=NoTile
      imageSize=<16 16>
      margin=<ltrb 0 0 0 0>
      outlineSettings {
       bUseBrushTransparency=false
       roundingType=HalfHeightRadius
       width=0
       color {colorUseRule=UseColor_Specified, specifiedColor=<rgba 0 0 0 0>}
       cornerRadii=<0 0 0 0>
      }
      tintColor {colorUseRule=UseColor_Specified, specifiedColor=<rgba 0.0044 0.2384 1 1>}
      uVRegion {bIsValid=false, max=<0 0>, min=<0 0>}
     }
     undeterminedImage {
      bIsDynamicallyLoaded=false
      drawAs=Image
      imageType=Vector
      mirroring=NoMirror
      resourceName=../../../Engine/Content/Slate/Starship/CoreWidgets/CheckBox/indeterminate.svg
      resourceObject=null
      tiling=NoTile
      imageSize=<16 16>
      margin=<ltrb 0 0 0 0>
      outlineSettings {
       bUseBrushTransparency=false
       roundingType=HalfHeightRadius
       width=0
       color {colorUseRule=UseColor_Specified, specifiedColor=<rgba 0 0 0 0>}
       cornerRadii=<0 0 0 0>
      }
      tintColor {colorUseRule=UseColor_Specified, specifiedColor=<rgba 0 0.162 0.7454 1>}
      uVRegion {bIsValid=false, max=<0 0>, min=<0 0>}
     }
     undeterminedPressedImage {
      bIsDynamicallyLoaded=false
      drawAs=Image
      imageType=Vector
      mirroring=NoMirror
      resourceName=../../../Engine/Content/Slate/Starship/CoreWidgets/CheckBox/indeterminate.svg
      resourceObject=null
      tiling=NoTile
      imageSize=<16 16>
      margin=<ltrb 0 0 0 0>
      outlineSettings {
       bUseBrushTransparency=false
       roundingType=HalfHeightRadius
       width=0
       color {colorUseRule=UseColor_Specified, specifiedColor=<rgba 0 0 0 0>}
       cornerRadii=<0 0 0 0>
      }
      tintColor {colorUseRule=UseColor_Specified, specifiedColor=<rgba 0 0.0802 0.3515 1>}
      uVRegion {bIsValid=false, max=<0 0>, min=<0 0>}
     }
    }
   }
   slot {class=BorderSlot}
  }
  Border_221 {
   widgetClassPath=&Border
   parent=UniformGridPanel_0
   widget=Border_221
   properties {brushColor=<rgba 0.0176 0.0176 0.0176 1>}
   slot {class=UniformGridSlot, properties {horizontalAlignment=HAlign_Fill, row=3, verticalAlignment=VAlign_Fill}}
  }
  TextBlock_7 {
   widgetClassPath=&TextBlock
   parent=Border_221
   widget=TextBlock_7
   properties {
    text=Composite Chassis Textures
    colorAndOpacity {colorUseRule=UseColor_Specified, specifiedColor=<rgba 0.6719 0.6719 0.6719 1>}
    font {
     bForceMonospaced=false
     bMaterialIsStencil=false
     fontMaterial=null
     fontObject=&Roboto
     letterSpacing=0
     monospacedWidth=1
     size=9
     skewAmount=0
     typefaceFontName=Regular
     outlineSettings {
      bApplyOutlineToDropShadows=false
      bMiteredCorners=false
      bSeparateFillAlpha=false
      outlineMaterial=null
      outlineSize=0
      outlineColor=<rgba 0 0 0 1>
     }
    }
   }
   slot {class=BorderSlot}
  }
  Border_222 {
   widgetClassPath=&Border
   parent=UniformGridPanel_0
   widget=Border_222
   properties {brushColor=<rgba 0.0176 0.0176 0.0176 1>}
   slot {class=UniformGridSlot, properties {column=1, horizontalAlignment=HAlign_Fill, row=3, verticalAlignment=VAlign_Fill}}
  }
  EditorUtilityCheckBox_CTextures_Composit {
   widgetClassPath=&EditorUtilityCheckBox
   bIsVariable=true
   parent=Border_222
   widget=EditorUtilityCheckBox_CTextures_Composit
   properties {
    checkedState=Checked
    widgetStyle {
     checkBoxType=CheckBox
     backgroundHoveredImage {
      bIsDynamicallyLoaded=false
      drawAs=RoundedBox
      imageType=NoImage
      mirroring=NoMirror
      resourceName=None
      resourceObject=null
      tiling=NoTile
      imageSize=<18 18>
      margin=<ltrb 0 0 0 0>
      outlineSettings {
       bUseBrushTransparency=false
       roundingType=FixedRadius
       width=1
       color {colorUseRule=UseColor_ColorTable, specifiedColor=<rgba 0 0 0 0>}
       cornerRadii=<3 3 3 3>
      }
      tintColor {colorUseRule=UseColor_Specified, specifiedColor=<rgba 0.0048 0.0048 0.0048 1>}
      uVRegion {bIsValid=false, max=<0 0>, min=<0 0>}
     }
     backgroundImage {
      bIsDynamicallyLoaded=false
      drawAs=RoundedBox
      imageType=NoImage
      mirroring=NoMirror
      resourceName=None
      resourceObject=null
      tiling=NoTile
      imageSize=<18 18>
      margin=<ltrb 0 0 0 0>
      outlineSettings {
       bUseBrushTransparency=false
       roundingType=FixedRadius
       width=1
       color {colorUseRule=UseColor_ColorTable, specifiedColor=<rgba 0 0 0 0>}
       cornerRadii=<3 3 3 3>
      }
      tintColor {colorUseRule=UseColor_Specified, specifiedColor=<rgba 0.0048 0.0048 0.0048 1>}
      uVRegion {bIsValid=false, max=<0 0>, min=<0 0>}
     }
     backgroundPressedImage {
      bIsDynamicallyLoaded=false
      drawAs=RoundedBox
      imageType=NoImage
      mirroring=NoMirror
      resourceName=None
      resourceObject=null
      tiling=NoTile
      imageSize=<18 18>
      margin=<ltrb 0 0 0 0>
      outlineSettings {
       bUseBrushTransparency=false
       roundingType=FixedRadius
       width=1
       color {colorUseRule=UseColor_ColorTable, specifiedColor=<rgba 0 0 0 0>}
       cornerRadii=<3 3 3 3>
      }
      tintColor {colorUseRule=UseColor_Specified, specifiedColor=<rgba 0.0048 0.0048 0.0048 1>}
      uVRegion {bIsValid=false, max=<0 0>, min=<0 0>}
     }
     borderBackgroundColor {colorUseRule=UseColor_Specified, specifiedColor=<rgba 1 1 1 1>}
     checkedForeground {colorUseRule=UseColor_Specified, specifiedColor=<rgba 1 1 1 1>}
     checkedHoveredForeground {colorUseRule=UseColor_Specified, specifiedColor=<rgba 1 1 1 1>}
     checkedHoveredImage {
      bIsDynamicallyLoaded=false
      drawAs=Image
      imageType=Vector
      mirroring=NoMirror
      resourceName=../../../Engine/Content/Slate/Starship/CoreWidgets/CheckBox/check.svg
      resourceObject=null
      tiling=NoTile
      imageSize=<16 16>
      margin=<ltrb 0 0 0 0>
      outlineSettings {
       bUseBrushTransparency=false
       roundingType=HalfHeightRadius
       width=0
       color {colorUseRule=UseColor_Specified, specifiedColor=<rgba 0 0 0 0>}
       cornerRadii=<0 0 0 0>
      }
      tintColor {colorUseRule=UseColor_Specified, specifiedColor=<rgba 0.901 0.378 0.1689 1>}
      uVRegion {bIsValid=false, max=<0 0>, min=<0 0>}
     }
     checkedImage {
      bIsDynamicallyLoaded=false
      drawAs=Image
      imageType=Vector
      mirroring=NoMirror
      resourceName=../../../Engine/Content/Slate/Starship/CoreWidgets/CheckBox/check.svg
      resourceObject=null
      tiling=NoTile
      imageSize=<16 16>
      margin=<ltrb 0 0 0 0>
      outlineSettings {
       bUseBrushTransparency=false
       roundingType=HalfHeightRadius
       width=0
       color {colorUseRule=UseColor_Specified, specifiedColor=<rgba 0 0 0 0>}
       cornerRadii=<0 0 0 0>
      }
      tintColor {colorUseRule=UseColor_Specified, specifiedColor=<rgba 0.7454 0.2129 0 1>}
      uVRegion {bIsValid=false, max=<0 0>, min=<0 0>}
     }
     checkedPressedForeground {colorUseRule=UseColor_Specified, specifiedColor=<rgba 1 1 1 1>}
     checkedPressedImage {
      bIsDynamicallyLoaded=false
      drawAs=Image
      imageType=Vector
      mirroring=NoMirror
      resourceName=../../../Engine/Content/Slate/Starship/CoreWidgets/CheckBox/check.svg
      resourceObject=null
      tiling=NoTile
      imageSize=<16 16>
      margin=<ltrb 0 0 0 0>
      outlineSettings {
       bUseBrushTransparency=false
       roundingType=HalfHeightRadius
       width=0
       color {colorUseRule=UseColor_Specified, specifiedColor=<rgba 0 0 0 0>}
       cornerRadii=<0 0 0 0>
      }
      tintColor {colorUseRule=UseColor_Specified, specifiedColor=<rgba 0.4219 0.1205 0 1>}
      uVRegion {bIsValid=false, max=<0 0>, min=<0 0>}
     }
     checkedSlateSound {resourceObject=null}
     foregroundColor {colorUseRule=UseColor_Specified, specifiedColor=<rgba 1 1 1 1>}
     hoveredForeground {colorUseRule=UseColor_Specified, specifiedColor=<rgba 1 1 1 1>}
     hoveredSlateSound {resourceObject=null}
     padding=<ltrb 2 0 0 0>
     pressedForeground {colorUseRule=UseColor_Specified, specifiedColor=<rgba 1 1 1 1>}
     uncheckedHoveredImage {
      bIsDynamicallyLoaded=false
      drawAs=NoDrawType
      imageType=NoImage
      mirroring=NoMirror
      resourceName=None
      resourceObject=null
      tiling=NoTile
      imageSize=<0 0>
      margin=<ltrb 0 0 0 0>
      outlineSettings {
       bUseBrushTransparency=false
       roundingType=HalfHeightRadius
       width=0
       color {colorUseRule=UseColor_Specified, specifiedColor=<rgba 0 0 0 0>}
       cornerRadii=<0 0 0 0>
      }
      tintColor {colorUseRule=UseColor_Specified, specifiedColor=<rgba 1 1 1 1>}
      uVRegion {bIsValid=false, max=<0 0>, min=<0 0>}
     }
     uncheckedImage {
      bIsDynamicallyLoaded=false
      drawAs=NoDrawType
      imageType=NoImage
      mirroring=NoMirror
      resourceName=None
      resourceObject=null
      tiling=NoTile
      imageSize=<0 0>
      margin=<ltrb 0 0 0 0>
      outlineSettings {
       bUseBrushTransparency=false
       roundingType=HalfHeightRadius
       width=0
       color {colorUseRule=UseColor_Specified, specifiedColor=<rgba 0 0 0 0>}
       cornerRadii=<0 0 0 0>
      }
      tintColor {colorUseRule=UseColor_Specified, specifiedColor=<rgba 1 1 1 1>}
      uVRegion {bIsValid=false, max=<0 0>, min=<0 0>}
     }
     uncheckedPressedImage {
      bIsDynamicallyLoaded=false
      drawAs=NoDrawType
      imageType=NoImage
      mirroring=NoMirror
      resourceName=None
      resourceObject=null
      tiling=NoTile
      imageSize=<0 0>
      margin=<ltrb 0 0 0 0>
      outlineSettings {
       bUseBrushTransparency=false
       roundingType=HalfHeightRadius
       width=0
       color {colorUseRule=UseColor_Specified, specifiedColor=<rgba 0 0 0 0>}
       cornerRadii=<0 0 0 0>
      }
      tintColor {colorUseRule=UseColor_Specified, specifiedColor=<rgba 1 1 1 1>}
      uVRegion {bIsValid=false, max=<0 0>, min=<0 0>}
     }
     uncheckedSlateSound {resourceObject=null}
     undeterminedForeground {colorUseRule=UseColor_Specified, specifiedColor=<rgba 1 1 1 1>}
     undeterminedHoveredImage {
      bIsDynamicallyLoaded=false
      drawAs=Image
      imageType=Vector
      mirroring=NoMirror
      resourceName=../../../Engine/Content/Slate/Starship/CoreWidgets/CheckBox/indeterminate.svg
      resourceObject=null
      tiling=NoTile
      imageSize=<16 16>
      margin=<ltrb 0 0 0 0>
      outlineSettings {
       bUseBrushTransparency=false
       roundingType=HalfHeightRadius
       width=0
       color {colorUseRule=UseColor_Specified, specifiedColor=<rgba 0 0 0 0>}
       cornerRadii=<0 0 0 0>
      }
      tintColor {colorUseRule=UseColor_Specified, specifiedColor=<rgba 0.0044 0.2384 1 1>}
      uVRegion {bIsValid=false, max=<0 0>, min=<0 0>}
     }
     undeterminedImage {
      bIsDynamicallyLoaded=false
      drawAs=Image
      imageType=Vector
      mirroring=NoMirror
      resourceName=../../../Engine/Content/Slate/Starship/CoreWidgets/CheckBox/indeterminate.svg
      resourceObject=null
      tiling=NoTile
      imageSize=<16 16>
      margin=<ltrb 0 0 0 0>
      outlineSettings {
       bUseBrushTransparency=false
       roundingType=HalfHeightRadius
       width=0
       color {colorUseRule=UseColor_Specified, specifiedColor=<rgba 0 0 0 0>}
       cornerRadii=<0 0 0 0>
      }
      tintColor {colorUseRule=UseColor_Specified, specifiedColor=<rgba 0 0.162 0.7454 1>}
      uVRegion {bIsValid=false, max=<0 0>, min=<0 0>}
     }
     undeterminedPressedImage {
      bIsDynamicallyLoaded=false
      drawAs=Image
      imageType=Vector
      mirroring=NoMirror
      resourceName=../../../Engine/Content/Slate/Starship/CoreWidgets/CheckBox/indeterminate.svg
      resourceObject=null
      tiling=NoTile
      imageSize=<16 16>
      margin=<ltrb 0 0 0 0>
      outlineSettings {
       bUseBrushTransparency=false
       roundingType=HalfHeightRadius
       width=0
       color {colorUseRule=UseColor_Specified, specifiedColor=<rgba 0 0 0 0>}
       cornerRadii=<0 0 0 0>
      }
      tintColor {colorUseRule=UseColor_Specified, specifiedColor=<rgba 0 0.0802 0.3515 1>}
      uVRegion {bIsValid=false, max=<0 0>, min=<0 0>}
     }
    }
   }
   slot {class=BorderSlot}
  }
  SizeBox_366 {
   widgetClassPath=&SizeBox
   parent=ButtonsAndSettingsHorizontalBox
   widget=SizeBox_366
   slot {class=HorizontalBoxSlot, properties {size {sizeRule=Fill, value=1}}}
  }
  VerticalBox_0 {
   widgetClassPath=&VerticalBox
   parent=SizeBox_366
   widget=VerticalBox_0
   slot {class=SizeBoxSlot, properties {padding=<ltrb 3 3 3 3>}}
  }
  SizeBox_370 {
   widgetClassPath=&SizeBox
   parent=VerticalBox_0
   widget=SizeBox_370
   slot {class=VerticalBoxSlot, properties {padding=<ltrb 3 3 3 3>}}
  }
  DetailsView_257 {
   widgetClassPath=&DetailsView
   parent=SizeBox_370
   widget=DetailsView_257
   properties {columnWidth=0.5, categoriesToShow [BakerSettings]}
   slot {class=SizeBoxSlot}
  }
  SizeBox_371 {
   widgetClassPath=&SizeBox
   parent=VerticalBox_0
   widget=SizeBox_371
   slot {class=VerticalBoxSlot, properties {padding=<ltrb 3 3 3 3>}}
  }
  DetailsView_497 {
   widgetClassPath=&DetailsView
   parent=SizeBox_371
   widget=DetailsView_497
   properties {columnWidth=0.7, categoriesToShow [OutputSettings]}
   slot {class=SizeBoxSlot}
  }
  SizeBox_864 {widgetClassPath=&SizeBox, parent=VerticalBox_0, widget=SizeBox_864, slot {class=VerticalBoxSlot}}
  SetPathButton {
   widgetClassPath=&EditorUtilityButton
   parent=SizeBox_864
   widget=SetPathButton
   properties {widgetStyle {
     clickedSlateSound {resourceObject=null}
     disabled {
      bIsDynamicallyLoaded=false
      drawAs=RoundedBox
      imageType=NoImage
      mirroring=NoMirror
      resourceName=None
      resourceObject=null
      tiling=NoTile
      imageSize=<0 0>
      margin=<ltrb 0 0 0 0>
      outlineSettings {
       bUseBrushTransparency=false
       roundingType=FixedRadius
       width=1
       color {colorUseRule=UseColor_ColorTable, specifiedColor=<rgba 0 0 0 0>}
       cornerRadii=<4 4 4 4>
      }
      tintColor {colorUseRule=UseColor_Specified, specifiedColor=<rgba 0.0395 0.0395 0.0395 1>}
      uVRegion {bIsValid=false, max=<0 0>, min=<0 0>}
     }
     disabledForeground {colorUseRule=UseColor_ColorTable, specifiedColor=<rgba 0 0 0 0>}
     hovered {
      bIsDynamicallyLoaded=false
      drawAs=RoundedBox
      imageType=NoImage
      mirroring=NoMirror
      resourceName=None
      resourceObject=null
      tiling=NoTile
      imageSize=<32 32>
      margin=<ltrb 0 0 0 0>
      outlineSettings {
       bUseBrushTransparency=false
       roundingType=FixedRadius
       width=1
       color {colorUseRule=UseColor_ColorTable, specifiedColor=<rgba 0 0 0 0>}
       cornerRadii=<4 4 4 4>
      }
      tintColor {colorUseRule=UseColor_Specified, specifiedColor=<rgba 0.9115 0.3654 0.1472 1>}
      uVRegion {bIsValid=false, max=<0 0>, min=<0 0>}
     }
     hoveredForeground {colorUseRule=UseColor_ColorTable, specifiedColor=<rgba 0 0 0 0>}
     hoveredSlateSound {resourceObject=null}
     normal {
      bIsDynamicallyLoaded=false
      drawAs=RoundedBox
      imageType=NoImage
      mirroring=NoMirror
      resourceName=None
      resourceObject=null
      tiling=NoTile
      imageSize=<32 32>
      margin=<ltrb 0 0 0 0>
      outlineSettings {
       bUseBrushTransparency=false
       roundingType=FixedRadius
       width=1
       color {colorUseRule=UseColor_ColorTable, specifiedColor=<rgba 0 0 0 0>}
       cornerRadii=<4 4 4 4>
      }
      tintColor {colorUseRule=UseColor_Specified, specifiedColor=<rgba 0.4167 0.119 0 1>}
      uVRegion {bIsValid=false, max=<0 0>, min=<0 0>}
     }
     normalForeground {colorUseRule=UseColor_ColorTable, specifiedColor=<rgba 0 0 0 0>}
     normalPadding=<ltrb 12 1.5 12 1.5>
     pressed {
      bIsDynamicallyLoaded=false
      drawAs=RoundedBox
      imageType=NoImage
      mirroring=NoMirror
      resourceName=None
      resourceObject=null
      tiling=NoTile
      imageSize=<32 32>
      margin=<ltrb 0 0 0 0>
      outlineSettings {
       bUseBrushTransparency=false
       roundingType=FixedRadius
       width=1
       color {colorUseRule=UseColor_ColorTable, specifiedColor=<rgba 0 0 0 0>}
       cornerRadii=<4 4 4 4>
      }
      tintColor {colorUseRule=UseColor_Specified, specifiedColor=<rgba 0.1302 0.0372 0 1>}
      uVRegion {bIsValid=false, max=<0 0>, min=<0 0>}
     }
     pressedForeground {colorUseRule=UseColor_ColorTable, specifiedColor=<rgba 0 0 0 0>}
     pressedPadding=<ltrb 12 2.5 12 0.5>
     pressedSlateSound {resourceObject=null}
    }}
   slot {class=SizeBoxSlot}
  }
  TextBlock_85 {
   widgetClassPath=&TextBlock
   parent=SetPathButton
   widget=TextBlock_85
   properties {
    text=Set Save Path to selected asset
    font {
     bForceMonospaced=false
     bMaterialIsStencil=false
     fontMaterial=null
     fontObject=&Roboto
     letterSpacing=0
     monospacedWidth=1
     size=9
     skewAmount=0
     typefaceFontName=Bold
     outlineSettings {
      bApplyOutlineToDropShadows=false
      bMiteredCorners=false
      bSeparateFillAlpha=false
      outlineMaterial=null
      outlineSize=0
      outlineColor=<rgba 0 0 0 1>
     }
    }
   }
   slot {class=ButtonSlot, properties {horizontalAlignment=HAlign_Left}}
  }
  SizeBox_21 {widgetClassPath=&SizeBox, parent=VerticalBox_0, widget=SizeBox_21, slot {class=VerticalBoxSlot}}
  SizeBox_216 {
   widgetClassPath=&SizeBox
   parent=VerticalBox_1
   widget=SizeBox_216
   properties {bOverride_HeightOverride=true, heightOverride=16}
   slot {class=VerticalBoxSlot}
  }
  Border_59 {
   widgetClassPath=&Border
   parent=SizeBox_216
   widget=Border_59
   properties {brushColor=<rgba 0.0176 0.0176 0.0176 1>}
   slot {class=SizeBoxSlot}
  }
  HorizontalBox_6 {widgetClassPath=&HorizontalBox, parent=VerticalBox_1, widget=HorizontalBox_6, slot {class=VerticalBoxSlot}}
  VerticalBox_4 {
   widgetClassPath=&VerticalBox
   parent=HorizontalBox_6
   widget=VerticalBox_4
   slot {class=HorizontalBoxSlot, properties {padding=<ltrb 3 3 3 3>, size {sizeRule=Fill, value=1}}}
  }
  SizeBox_369 {widgetClassPath=&SizeBox, parent=VerticalBox_4, widget=SizeBox_369, slot {class=VerticalBoxSlot}}
  DetailsView_150 {
   widgetClassPath=&DetailsView
   parent=SizeBox_369
   widget=DetailsView_150
   properties {categoriesToShow [Target]}
   slot {class=SizeBoxSlot}
  }
  SizeBox_19 {
   widgetClassPath=&SizeBox
   parent=VerticalBox_4
   widget=SizeBox_19
   properties {bOverride_HeightOverride=true, heightOverride=16}
   slot {class=VerticalBoxSlot}
  }
  Border {
   widgetClassPath=&Border
   parent=SizeBox_19
   widget=Border
   properties {brushColor=<rgba 0.0176 0.0176 0.0176 1>}
   slot {class=SizeBoxSlot}
  }
  SizeBox_138 {widgetClassPath=&SizeBox, parent=VerticalBox_4, widget=SizeBox_138, slot {class=VerticalBoxSlot}}
  VerticalBox_5 {
   widgetClassPath=&VerticalBox
   parent=HorizontalBox_6
   widget=VerticalBox_5
   slot {class=HorizontalBoxSlot, properties {padding=<ltrb 3 3 3 3>, size {sizeRule=Fill, value=1}}}
  }
  SizeBox_140 {
   widgetClassPath=&SizeBox
   parent=VerticalBox_5
   widget=SizeBox_140
   slot {class=VerticalBoxSlot, properties {size {sizeRule=Fill, value=1}}}
  }
  DetailsView_62 {
   widgetClassPath=&DetailsView
   parent=SizeBox_140
   widget=DetailsView_62
   properties {categoriesToShow [Source]}
   slot {class=SizeBoxSlot}
  }
  SizeBox_20 {
   widgetClassPath=&SizeBox
   parent=VerticalBox_5
   widget=SizeBox_20
   properties {bOverride_HeightOverride=true, heightOverride=16}
   slot {class=VerticalBoxSlot}
  }
  Border_1 {
   widgetClassPath=&Border
   parent=SizeBox_20
   widget=Border_1
   properties {brushColor=<rgba 0.0176 0.0176 0.0176 1>}
   slot {class=SizeBoxSlot}
  }
  SizeBox_18 {widgetClassPath=&SizeBox, parent=VerticalBox_5, widget=SizeBox_18, slot {class=VerticalBoxSlot}}
  HorizontalBox_196 {widgetClassPath=&HorizontalBox, parent=VerticalBox_1, widget=HorizontalBox_196, slot {class=VerticalBoxSlot}}
  ThresholdScrollBox_0 {
   widgetClassPath=&ThresholdScrollBox
   parent=HorizontalBox_196
   widget=ThresholdScrollBox_0
   properties {
    widgetBarStyle {
     thickness=8
     draggedThumbImage {
      bIsDynamicallyLoaded=false
      drawAs=RoundedBox
      imageType=NoImage
      mirroring=NoMirror
      resourceName=None
      resourceObject=null
      tiling=NoTile
      imageSize=<0 0>
      margin=<ltrb 0 0 0 0>
      outlineSettings {
       bUseBrushTransparency=false
       roundingType=FixedRadius
       width=0
       color {colorUseRule=UseColor_Specified, specifiedColor=<rgba 0 0 0 0>}
       cornerRadii=<4 4 4 4>
      }
      tintColor {colorUseRule=UseColor_Specified, specifiedColor=<rgba 0.2159 0.2159 0.2159 1>}
      uVRegion {bIsValid=false, max=<0 0>, min=<0 0>}
     }
     horizontalBackgroundImage {
      bIsDynamicallyLoaded=false
      drawAs=NoDrawType
      imageType=NoImage
      mirroring=NoMirror
      resourceName=None
      resourceObject=null
      tiling=NoTile
      imageSize=<0 0>
      margin=<ltrb 0 0 0 0>
      outlineSettings {
       bUseBrushTransparency=false
       roundingType=HalfHeightRadius
       width=0
       color {colorUseRule=UseColor_Specified, specifiedColor=<rgba 0 0 0 0>}
       cornerRadii=<0 0 0 0>
      }
      tintColor {colorUseRule=UseColor_Specified, specifiedColor=<rgba 1 1 1 1>}
      uVRegion {bIsValid=false, max=<0 0>, min=<0 0>}
     }
     horizontalBottomSlotImage {
      bIsDynamicallyLoaded=false
      drawAs=NoDrawType
      imageType=NoImage
      mirroring=NoMirror
      resourceName=None
      resourceObject=null
      tiling=NoTile
      imageSize=<0 0>
      margin=<ltrb 0 0 0 0>
      outlineSettings {
       bUseBrushTransparency=false
       roundingType=HalfHeightRadius
       width=0
       color {colorUseRule=UseColor_Specified, specifiedColor=<rgba 0 0 0 0>}
       cornerRadii=<0 0 0 0>
      }
      tintColor {colorUseRule=UseColor_Specified, specifiedColor=<rgba 1 1 1 1>}
      uVRegion {bIsValid=false, max=<0 0>, min=<0 0>}
     }
     horizontalTopSlotImage {
      bIsDynamicallyLoaded=false
      drawAs=NoDrawType
      imageType=NoImage
      mirroring=NoMirror
      resourceName=None
      resourceObject=null
      tiling=NoTile
      imageSize=<0 0>
      margin=<ltrb 0 0 0 0>
      outlineSettings {
       bUseBrushTransparency=false
       roundingType=HalfHeightRadius
       width=0
       color {colorUseRule=UseColor_Specified, specifiedColor=<rgba 0 0 0 0>}
       cornerRadii=<0 0 0 0>
      }
      tintColor {colorUseRule=UseColor_Specified, specifiedColor=<rgba 1 1 1 1>}
      uVRegion {bIsValid=false, max=<0 0>, min=<0 0>}
     }
     hoveredThumbImage {
      bIsDynamicallyLoaded=false
      drawAs=RoundedBox
      imageType=NoImage
      mirroring=NoMirror
      resourceName=None
      resourceObject=null
      tiling=NoTile
      imageSize=<0 0>
      margin=<ltrb 0 0 0 0>
      outlineSettings {
       bUseBrushTransparency=false
       roundingType=FixedRadius
       width=0
       color {colorUseRule=UseColor_Specified, specifiedColor=<rgba 0 0 0 0>}
       cornerRadii=<4 4 4 4>
      }
      tintColor {colorUseRule=UseColor_Specified, specifiedColor=<rgba 0.2159 0.2159 0.2159 1>}
      uVRegion {bIsValid=false, max=<0 0>, min=<0 0>}
     }
     normalThumbImage {
      bIsDynamicallyLoaded=false
      drawAs=RoundedBox
      imageType=NoImage
      mirroring=NoMirror
      resourceName=None
      resourceObject=null
      tiling=NoTile
      imageSize=<0 0>
      margin=<ltrb 0 0 0 0>
      outlineSettings {
       bUseBrushTransparency=false
       roundingType=FixedRadius
       width=0
       color {colorUseRule=UseColor_Specified, specifiedColor=<rgba 0 0 0 0>}
       cornerRadii=<4 4 4 4>
      }
      tintColor {colorUseRule=UseColor_Specified, specifiedColor=<rgba 0.0953 0.0953 0.0953 1>}
      uVRegion {bIsValid=false, max=<0 0>, min=<0 0>}
     }
     verticalBackgroundImage {
      bIsDynamicallyLoaded=false
      drawAs=NoDrawType
      imageType=NoImage
      mirroring=NoMirror
      resourceName=None
      resourceObject=null
      tiling=NoTile
      imageSize=<0 0>
      margin=<ltrb 0 0 0 0>
      outlineSettings {
       bUseBrushTransparency=false
       roundingType=HalfHeightRadius
       width=0
       color {colorUseRule=UseColor_Specified, specifiedColor=<rgba 0 0 0 0>}
       cornerRadii=<0 0 0 0>
      }
      tintColor {colorUseRule=UseColor_Specified, specifiedColor=<rgba 1 1 1 1>}
      uVRegion {bIsValid=false, max=<0 0>, min=<0 0>}
     }
     verticalBottomSlotImage {
      bIsDynamicallyLoaded=false
      drawAs=NoDrawType
      imageType=NoImage
      mirroring=NoMirror
      resourceName=None
      resourceObject=null
      tiling=NoTile
      imageSize=<0 0>
      margin=<ltrb 0 0 0 0>
      outlineSettings {
       bUseBrushTransparency=false
       roundingType=HalfHeightRadius
       width=0
       color {colorUseRule=UseColor_Specified, specifiedColor=<rgba 0 0 0 0>}
       cornerRadii=<0 0 0 0>
      }
      tintColor {colorUseRule=UseColor_Specified, specifiedColor=<rgba 1 1 1 1>}
      uVRegion {bIsValid=false, max=<0 0>, min=<0 0>}
     }
     verticalTopSlotImage {
      bIsDynamicallyLoaded=false
      drawAs=NoDrawType
      imageType=NoImage
      mirroring=NoMirror
      resourceName=None
      resourceObject=null
      tiling=NoTile
      imageSize=<0 0>
      margin=<ltrb 0 0 0 0>
      outlineSettings {
       bUseBrushTransparency=false
       roundingType=HalfHeightRadius
       width=0
       color {colorUseRule=UseColor_Specified, specifiedColor=<rgba 0 0 0 0>}
       cornerRadii=<0 0 0 0>
      }
      tintColor {colorUseRule=UseColor_Specified, specifiedColor=<rgba 1 1 1 1>}
      uVRegion {bIsValid=false, max=<0 0>, min=<0 0>}
     }
    }
    widgetStyle {
     barThickness=8
     bottomShadowBrush {
      bIsDynamicallyLoaded=false
      drawAs=Box
      imageType=FullColor
      mirroring=NoMirror
      resourceName=../../../Engine/Content/Slate/Common/ScrollBoxShadowBottom.png
      resourceObject=null
      tiling=NoTile
      imageSize=<16 8>
      margin=<ltrb 0.5 0 0.5 1>
      outlineSettings {
       bUseBrushTransparency=false
       roundingType=HalfHeightRadius
       width=0
       color {colorUseRule=UseColor_Specified, specifiedColor=<rgba 0 0 0 0>}
       cornerRadii=<0 0 0 0>
      }
      tintColor {colorUseRule=UseColor_Specified, specifiedColor=<rgba 1 1 1 1>}
      uVRegion {bIsValid=false, max=<0 0>, min=<0 0>}
     }
     horizontalScrolledContentPadding=<ltrb 0 0 1 0>
     leftShadowBrush {
      bIsDynamicallyLoaded=false
      drawAs=Box
      imageType=FullColor
      mirroring=NoMirror
      resourceName=../../../Engine/Content/Slate/Common/ScrollBoxShadowLeft.png
      resourceObject=null
      tiling=NoTile
      imageSize=<8 16>
      margin=<ltrb 1 0.5 0 0.5>
      outlineSettings {
       bUseBrushTransparency=false
       roundingType=HalfHeightRadius
       width=0
       color {colorUseRule=UseColor_Specified, specifiedColor=<rgba 0 0 0 0>}
       cornerRadii=<0 0 0 0>
      }
      tintColor {colorUseRule=UseColor_Specified, specifiedColor=<rgba 1 1 1 1>}
      uVRegion {bIsValid=false, max=<0 0>, min=<0 0>}
     }
     rightShadowBrush {
      bIsDynamicallyLoaded=false
      drawAs=Box
      imageType=FullColor
      mirroring=NoMirror
      resourceName=../../../Engine/Content/Slate/Common/ScrollBoxShadowRight.png
      resourceObject=null
      tiling=NoTile
      imageSize=<8 16>
      margin=<ltrb 0 0.5 1 0.5>
      outlineSettings {
       bUseBrushTransparency=false
       roundingType=HalfHeightRadius
       width=0
       color {colorUseRule=UseColor_Specified, specifiedColor=<rgba 0 0 0 0>}
       cornerRadii=<0 0 0 0>
      }
      tintColor {colorUseRule=UseColor_Specified, specifiedColor=<rgba 1 1 1 1>}
      uVRegion {bIsValid=false, max=<0 0>, min=<0 0>}
     }
     topShadowBrush {
      bIsDynamicallyLoaded=false
      drawAs=Box
      imageType=FullColor
      mirroring=NoMirror
      resourceName=../../../Engine/Content/Slate/Common/ScrollBoxShadowTop.png
      resourceObject=null
      tiling=NoTile
      imageSize=<16 8>
      margin=<ltrb 0.5 1 0.5 0>
      outlineSettings {
       bUseBrushTransparency=false
       roundingType=HalfHeightRadius
       width=0
       color {colorUseRule=UseColor_Specified, specifiedColor=<rgba 0 0 0 0>}
       cornerRadii=<0 0 0 0>
      }
      tintColor {colorUseRule=UseColor_Specified, specifiedColor=<rgba 1 1 1 1>}
      uVRegion {bIsValid=false, max=<0 0>, min=<0 0>}
     }
     verticalScrolledContentPadding=<ltrb 0 0 0 1>
    }
   }
   slot {class=HorizontalBoxSlot, properties {padding=<ltrb 6 6 6 6>, size {sizeRule=Fill, value=1}}}
  }
  DetailsView_124 {
   widgetClassPath=&DetailsView
   parent=ThresholdScrollBox_0
   widget=DetailsView_124
   properties {columnWidth=0.55, categoriesToShow [Bakes]}
   slot {class=ScrollBoxSlot, properties {size {sizeRule=Fill, value=1}}}
  }
  ThresholdScrollBox_1 {
   widgetClassPath=&ThresholdScrollBox
   parent=HorizontalBox_196
   widget=ThresholdScrollBox_1
   properties {
    alwaysShowScrollbarTrack=true
    widgetBarStyle {
     thickness=8
     draggedThumbImage {
      bIsDynamicallyLoaded=false
      drawAs=RoundedBox
      imageType=NoImage
      mirroring=NoMirror
      resourceName=None
      resourceObject=null
      tiling=NoTile
      imageSize=<0 0>
      margin=<ltrb 0 0 0 0>
      outlineSettings {
       bUseBrushTransparency=false
       roundingType=FixedRadius
       width=0
       color {colorUseRule=UseColor_Specified, specifiedColor=<rgba 0 0 0 0>}
       cornerRadii=<4 4 4 4>
      }
      tintColor {colorUseRule=UseColor_Specified, specifiedColor=<rgba 0.2159 0.2159 0.2159 1>}
      uVRegion {bIsValid=false, max=<0 0>, min=<0 0>}
     }
     horizontalBackgroundImage {
      bIsDynamicallyLoaded=false
      drawAs=NoDrawType
      imageType=NoImage
      mirroring=NoMirror
      resourceName=None
      resourceObject=null
      tiling=NoTile
      imageSize=<0 0>
      margin=<ltrb 0 0 0 0>
      outlineSettings {
       bUseBrushTransparency=false
       roundingType=HalfHeightRadius
       width=0
       color {colorUseRule=UseColor_Specified, specifiedColor=<rgba 0 0 0 0>}
       cornerRadii=<0 0 0 0>
      }
      tintColor {colorUseRule=UseColor_Specified, specifiedColor=<rgba 1 1 1 1>}
      uVRegion {bIsValid=false, max=<0 0>, min=<0 0>}
     }
     horizontalBottomSlotImage {
      bIsDynamicallyLoaded=false
      drawAs=NoDrawType
      imageType=NoImage
      mirroring=NoMirror
      resourceName=None
      resourceObject=null
      tiling=NoTile
      imageSize=<0 0>
      margin=<ltrb 0 0 0 0>
      outlineSettings {
       bUseBrushTransparency=false
       roundingType=HalfHeightRadius
       width=0
       color {colorUseRule=UseColor_Specified, specifiedColor=<rgba 0 0 0 0>}
       cornerRadii=<0 0 0 0>
      }
      tintColor {colorUseRule=UseColor_Specified, specifiedColor=<rgba 1 1 1 1>}
      uVRegion {bIsValid=false, max=<0 0>, min=<0 0>}
     }
     horizontalTopSlotImage {
      bIsDynamicallyLoaded=false
      drawAs=NoDrawType
      imageType=NoImage
      mirroring=NoMirror
      resourceName=None
      resourceObject=null
      tiling=NoTile
      imageSize=<0 0>
      margin=<ltrb 0 0 0 0>
      outlineSettings {
       bUseBrushTransparency=false
       roundingType=HalfHeightRadius
       width=0
       color {colorUseRule=UseColor_Specified, specifiedColor=<rgba 0 0 0 0>}
       cornerRadii=<0 0 0 0>
      }
      tintColor {colorUseRule=UseColor_Specified, specifiedColor=<rgba 1 1 1 1>}
      uVRegion {bIsValid=false, max=<0 0>, min=<0 0>}
     }
     hoveredThumbImage {
      bIsDynamicallyLoaded=false
      drawAs=RoundedBox
      imageType=NoImage
      mirroring=NoMirror
      resourceName=None
      resourceObject=null
      tiling=NoTile
      imageSize=<0 0>
      margin=<ltrb 0 0 0 0>
      outlineSettings {
       bUseBrushTransparency=false
       roundingType=FixedRadius
       width=0
       color {colorUseRule=UseColor_Specified, specifiedColor=<rgba 0 0 0 0>}
       cornerRadii=<4 4 4 4>
      }
      tintColor {colorUseRule=UseColor_Specified, specifiedColor=<rgba 0.2159 0.2159 0.2159 1>}
      uVRegion {bIsValid=false, max=<0 0>, min=<0 0>}
     }
     normalThumbImage {
      bIsDynamicallyLoaded=false
      drawAs=RoundedBox
      imageType=NoImage
      mirroring=NoMirror
      resourceName=None
      resourceObject=null
      tiling=NoTile
      imageSize=<0 0>
      margin=<ltrb 0 0 0 0>
      outlineSettings {
       bUseBrushTransparency=false
       roundingType=FixedRadius
       width=0
       color {colorUseRule=UseColor_Specified, specifiedColor=<rgba 0 0 0 0>}
       cornerRadii=<4 4 4 4>
      }
      tintColor {colorUseRule=UseColor_Specified, specifiedColor=<rgba 0.0953 0.0953 0.0953 1>}
      uVRegion {bIsValid=false, max=<0 0>, min=<0 0>}
     }
     verticalBackgroundImage {
      bIsDynamicallyLoaded=false
      drawAs=NoDrawType
      imageType=NoImage
      mirroring=NoMirror
      resourceName=None
      resourceObject=null
      tiling=NoTile
      imageSize=<0 0>
      margin=<ltrb 0 0 0 0>
      outlineSettings {
       bUseBrushTransparency=false
       roundingType=HalfHeightRadius
       width=0
       color {colorUseRule=UseColor_Specified, specifiedColor=<rgba 0 0 0 0>}
       cornerRadii=<0 0 0 0>
      }
      tintColor {colorUseRule=UseColor_Specified, specifiedColor=<rgba 1 1 1 1>}
      uVRegion {bIsValid=false, max=<0 0>, min=<0 0>}
     }
     verticalBottomSlotImage {
      bIsDynamicallyLoaded=false
      drawAs=NoDrawType
      imageType=NoImage
      mirroring=NoMirror
      resourceName=None
      resourceObject=null
      tiling=NoTile
      imageSize=<0 0>
      margin=<ltrb 0 0 0 0>
      outlineSettings {
       bUseBrushTransparency=false
       roundingType=HalfHeightRadius
       width=0
       color {colorUseRule=UseColor_Specified, specifiedColor=<rgba 0 0 0 0>}
       cornerRadii=<0 0 0 0>
      }
      tintColor {colorUseRule=UseColor_Specified, specifiedColor=<rgba 1 1 1 1>}
      uVRegion {bIsValid=false, max=<0 0>, min=<0 0>}
     }
     verticalTopSlotImage {
      bIsDynamicallyLoaded=false
      drawAs=NoDrawType
      imageType=NoImage
      mirroring=NoMirror
      resourceName=None
      resourceObject=null
      tiling=NoTile
      imageSize=<0 0>
      margin=<ltrb 0 0 0 0>
      outlineSettings {
       bUseBrushTransparency=false
       roundingType=HalfHeightRadius
       width=0
       color {colorUseRule=UseColor_Specified, specifiedColor=<rgba 0 0 0 0>}
       cornerRadii=<0 0 0 0>
      }
      tintColor {colorUseRule=UseColor_Specified, specifiedColor=<rgba 1 1 1 1>}
      uVRegion {bIsValid=false, max=<0 0>, min=<0 0>}
     }
    }
    widgetStyle {
     barThickness=8
     bottomShadowBrush {
      bIsDynamicallyLoaded=false
      drawAs=Box
      imageType=FullColor
      mirroring=NoMirror
      resourceName=../../../Engine/Content/Slate/Common/ScrollBoxShadowBottom.png
      resourceObject=null
      tiling=NoTile
      imageSize=<16 8>
      margin=<ltrb 0.5 0 0.5 1>
      outlineSettings {
       bUseBrushTransparency=false
       roundingType=HalfHeightRadius
       width=0
       color {colorUseRule=UseColor_Specified, specifiedColor=<rgba 0 0 0 0>}
       cornerRadii=<0 0 0 0>
      }
      tintColor {colorUseRule=UseColor_Specified, specifiedColor=<rgba 1 1 1 1>}
      uVRegion {bIsValid=false, max=<0 0>, min=<0 0>}
     }
     horizontalScrolledContentPadding=<ltrb 0 0 1 0>
     leftShadowBrush {
      bIsDynamicallyLoaded=false
      drawAs=Box
      imageType=FullColor
      mirroring=NoMirror
      resourceName=../../../Engine/Content/Slate/Common/ScrollBoxShadowLeft.png
      resourceObject=null
      tiling=NoTile
      imageSize=<8 16>
      margin=<ltrb 1 0.5 0 0.5>
      outlineSettings {
       bUseBrushTransparency=false
       roundingType=HalfHeightRadius
       width=0
       color {colorUseRule=UseColor_Specified, specifiedColor=<rgba 0 0 0 0>}
       cornerRadii=<0 0 0 0>
      }
      tintColor {colorUseRule=UseColor_Specified, specifiedColor=<rgba 1 1 1 1>}
      uVRegion {bIsValid=false, max=<0 0>, min=<0 0>}
     }
     rightShadowBrush {
      bIsDynamicallyLoaded=false
      drawAs=Box
      imageType=FullColor
      mirroring=NoMirror
      resourceName=../../../Engine/Content/Slate/Common/ScrollBoxShadowRight.png
      resourceObject=null
      tiling=NoTile
      imageSize=<8 16>
      margin=<ltrb 0 0.5 1 0.5>
      outlineSettings {
       bUseBrushTransparency=false
       roundingType=HalfHeightRadius
       width=0
       color {colorUseRule=UseColor_Specified, specifiedColor=<rgba 0 0 0 0>}
       cornerRadii=<0 0 0 0>
      }
      tintColor {colorUseRule=UseColor_Specified, specifiedColor=<rgba 1 1 1 1>}
      uVRegion {bIsValid=false, max=<0 0>, min=<0 0>}
     }
     topShadowBrush {
      bIsDynamicallyLoaded=false
      drawAs=Box
      imageType=FullColor
      mirroring=NoMirror
      resourceName=../../../Engine/Content/Slate/Common/ScrollBoxShadowTop.png
      resourceObject=null
      tiling=NoTile
      imageSize=<16 8>
      margin=<ltrb 0.5 1 0.5 0>
      outlineSettings {
       bUseBrushTransparency=false
       roundingType=HalfHeightRadius
       width=0
       color {colorUseRule=UseColor_Specified, specifiedColor=<rgba 0 0 0 0>}
       cornerRadii=<0 0 0 0>
      }
      tintColor {colorUseRule=UseColor_Specified, specifiedColor=<rgba 1 1 1 1>}
      uVRegion {bIsValid=false, max=<0 0>, min=<0 0>}
     }
     verticalScrolledContentPadding=<ltrb 0 0 0 1>
    }
   }
   slot {class=HorizontalBoxSlot, properties {padding=<ltrb 6 6 6 6>, size {sizeRule=Fill, value=1}}}
  }
  DetailsView_69 {
   widgetClassPath=&DetailsView
   parent=ThresholdScrollBox_1
   widget=DetailsView_69
   properties {categoriesToShow [Source Textures]}
   slot {class=ScrollBoxSlot}
  }
 ]
}
